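# Physics asset for the F_MED_OutletVerve female body cosmetic, setting up constrained bodies for its left and right fluff pieces and ribbons.
# Capsules on spine_04 and both upper-torso bones form the torso volume, with the torso bones and small fluff and ribbon base anchors angularly locked to spine_04.
# Each fluff piece swings and twists within narrow limits off its anchor, while each ribbon hangs as a tightly limited three-link chain.
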PhysicsAsset /BRCosmetics/Characters/Player/Female/Medium/Bodies/F_MED_OutletVerve/Meshes/F_MED_OutletVerve_Physics
bodies {
 base_fluff_l [{shapeType=Capsule, length=0.505, radius=0.505, shapeName=base_fluff_l_capsule, rotation=<rot 0 -90 0>}]
 base_fluff_r [{shapeType=Capsule, length=0.505, radius=0.505, shapeName=base_fluff_r_capsule, rotation=<rot 0 -90 0>}]
 base_ribbon_a_l [{shapeType=Capsule, length=0.505, radius=0.505, shapeName=base_ribbon_a_l_capsule, rotation=<rot 0 -90 0>}]
 base_ribbon_a_r [{shapeType=Capsule, length=0.505, radius=0.505, shapeName=base_ribbon_a_r_capsule, rotation=<rot 0 -90 0>}]
 dyn_ribbon_a_l_01 [{shapeType=Capsule, length=4.081, radius=1.25, shapeName=dyn_hair_06_cr_capsule, center=<2.0405 -1.2014 0>, rotation=<rot 90 -5.0001 0>}]
 dyn_ribbon_a_l_02 [{shapeType=Capsule, length=3.818, radius=1.25, shapeName=dyn_hair_06_cr_capsule, center=<1.909 -1.3209 0>, rotation=<rot 90 0 0>}]
 dyn_ribbon_a_l_03 [{shapeType=Capsule, length=3, radius=1.25, shapeName=dyn_hair_06_cr_capsule, center=<1.5 -1.3209 0>, rotation=<rot 90 0 0>}]
 dyn_ribbon_a_r_01 [{shapeType=Capsule, length=4.081, radius=1.25, shapeName=dyn_hair_06_cr_capsule, center=<2.0405 -1.2014 0>, rotation=<rot 90 -5.0001 0>}]
 dyn_ribbon_a_r_02 [{shapeType=Capsule, length=3.818, radius=1.25, shapeName=dyn_hair_06_cr_capsule, center=<1.909 -1.3209 0>, rotation=<rot 90 0 0>}]
 dyn_ribbon_a_r_03 [{shapeType=Capsule, length=3, radius=1.25, shapeName=dyn_hair_06_cr_capsule, center=<1.5 -1.3209 0>, rotation=<rot 90 0 0>}]
 dyn_upperTorso_l [{
   shapeType=Capsule
   length=8.3193
   radius=6.9732
   shapeName=dyn_upperTorso_l_capsule
   center=<-1.4935 2.5868 -0.7>
   rotation=<rot 0.0003 30 90>
  }]
 dyn_upperTorso_r [{shapeType=Capsule, length=8.2261, radius=6.5864, shapeName=dyn_upperTorso_r_capsule, center=<0 -2.9978 -0.7>, rotation=<rot 0 0 -90>}]
 fluff_l [{shapeType=Capsule, length=3.1309, radius=5.0056, shapeName=dyn_hair_06_cr_capsule, center=<3.5907 0 0>, rotation=<rot 90 0 0>}]
 fluff_r [{shapeType=Capsule, length=3.1309, radius=5.0056, shapeName=dyn_hair_06_cr_capsule, center=<-3.2522 0 0>, rotation=<rot 90 0 0>}]
 spine_04 [
  {
   shapeType=Capsule
   length=10.1218
   radius=7.6557
   shapeName=spine_04_capsule
   center=<4.3521 2.3233 -3.5503>
   rotation=<rot 87.3433 -66.1163 -28.0042>
  }
  {
   shapeType=Capsule
   length=10.1218
   radius=7.6557
   shapeName=spine_04_capsule
   center=<4.3643 2.4548 3.1829>
   rotation=<rot -87.6123 -86.4798 49.2439>
  }
  {shapeType=Capsule, length=9.5653, radius=7.8221, shapeName=spine_04_capsule, center=<-0.7601 5.892 0>}
 ]
}
constraints [
 {bone1Name=fluff_l, bone2Name=base_fluff_l, swing1LimitDegrees=12, swing2LimitDegrees=12, twistLimitDegrees=3}
 {bone1Name=fluff_r, bone2Name=base_fluff_r, swing1LimitDegrees=12, swing2LimitDegrees=12, twistLimitDegrees=3}
 {
  bone1Name=base_fluff_l
  bone2Name=spine_04
  swing1LimitDegrees=45
  swing1Motion=Locked
  swing2LimitDegrees=45
  swing2Motion=Locked
  twistLimitDegrees=45
  twistMotion=Locked
 }
 {
  bone1Name=base_fluff_r
  bone2Name=spine_04
  swing1LimitDegrees=45
  swing1Motion=Locked
  swing2LimitDegrees=45
  swing2Motion=Locked
  twistLimitDegrees=45
  twistMotion=Locked
 }
 {
  bone1Name=base_ribbon_a_l
  bone2Name=spine_04
  swing1LimitDegrees=45
  swing1Motion=Locked
  swing2LimitDegrees=45
  swing2Motion=Locked
  twistLimitDegrees=45
  twistMotion=Locked
 }
 {bone1Name=dyn_ribbon_a_l_01, bone2Name=base_ribbon_a_l, swing1LimitDegrees=5, swing2LimitDegrees=5, twistLimitDegrees=0}
 {bone1Name=dyn_ribbon_a_l_02, bone2Name=dyn_ribbon_a_l_01, swing1LimitDegrees=5, swing2LimitDegrees=5, twistLimitDegrees=0}
 {bone1Name=dyn_ribbon_a_l_03, bone2Name=dyn_ribbon_a_l_02, swing1LimitDegrees=5, swing2LimitDegrees=5, twistLimitDegrees=0}
 {
  bone1Name=base_ribbon_a_r
  bone2Name=spine_04
  swing1LimitDegrees=45
  swing1Motion=Locked
  swing2LimitDegrees=45
  swing2Motion=Locked
  twistLimitDegrees=45
  twistMotion=Locked
 }
 {bone1Name=dyn_ribbon_a_r_01, bone2Name=base_ribbon_a_r, swing1LimitDegrees=5, swing2LimitDegrees=5, twistLimitDegrees=0}
 {bone1Name=dyn_ribbon_a_r_02, bone2Name=dyn_ribbon_a_r_01, swing1LimitDegrees=5, swing2LimitDegrees=5, twistLimitDegrees=0}
 {bone1Name=dyn_ribbon_a_r_03, bone2Name=dyn_ribbon_a_r_02, swing1LimitDegrees=5, swing2LimitDegrees=5, twistLimitDegrees=0}
 {
  bone1Name=dyn_upperTorso_l
  bone2Name=spine_04
  swing1LimitDegrees=45
  swing1Motion=Locked
  swing2LimitDegrees=45
  swing2Motion=Locked
  twistLimitDegrees=45
  twistMotion=Locked
 }
 {
  bone1Name=dyn_upperTorso_r
  bone2Name=spine_04
  swing1LimitDegrees=45
  swing1Motion=Locked
  swing2LimitDegrees=45
  swing2Motion=Locked
  twistLimitDegrees=45
  twistMotion=Locked
 }
]
