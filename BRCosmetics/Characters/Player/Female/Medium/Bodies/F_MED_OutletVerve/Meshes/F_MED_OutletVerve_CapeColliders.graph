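# Physics asset that supplies collision shapes for the cape on the F_MED_OutletVerve medium female player body.
# Capsules cover head, clavicles, spine, pelvis and feet, with paired capsules thickening spine_02 to spine_04, while arm, leg and neck bodies carry no shapes.
# Five dyn_Cloth_volume bones (back, left, right) add capsules, each constrained to the pelvis with free swing and twist.
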PhysicsAsset /BRCosmetics/Characters/Player/Female/Medium/Bodies/F_MED_OutletVerve/Meshes/F_MED_OutletVerve_CapeColliders
bodies {
 calf_l []
 calf_r []
 clavicle_l [{shapeType=Capsule, length=6.5623, radius=7.5194, center=<3.9566 -1.0862 -2.1127>, rotation=<rot -77.5566 177.3915 7.7314>}]
 clavicle_r [{shapeType=Capsule, length=6.5623, radius=7.5194, center=<-3.9566 1.0862 2.1127>, rotation=<rot -77.5566 177.3915 -172.2686>}]
 dyn_Cloth_volume_bk02 [{
   shapeType=Capsule
   length=11.0266
   radius=9.8138
   shapeName=dyn_Cloth_volume_bk02_capsule
   center=<-4.3917 1.8674 0.0224>
   rotation=<rot -2.6542 -90 -0.0173>
  }]
 dyn_Cloth_volume_l_B02 [{
   shapeType=Capsule
   length=12.4499
   radius=10.1643
   shapeName=dyn_Cloth_volume_l_B02_capsule
   center=<-2.3187 4.3862 3.2675>
   rotation=<rot 1.9386 -90 -2.4062>
  }]
 dyn_Cloth_volume_l_C02 [{
   shapeType=Capsule
   length=11.6141
   radius=9.6261
   shapeName=dyn_Cloth_volume_l_C02_capsule
   center=<-5.7127 4.0697 0.4716>
   rotation=<rot -15.1438 -90 19.0456>
  }]
 dyn_Cloth_volume_r_B02 [{
   shapeType=Capsule
   length=12.3464
   radius=10.0606
   shapeName=dyn_Cloth_volume_r_B02_capsule
   center=<-2.2156 4.3397 -3.326>
   rotation=<rot 0.3395 -90 1.3274>
  }]
 dyn_Cloth_volume_r_C02 [{
   shapeType=Capsule
   length=11.6958
   radius=9.6194
   shapeName=dyn_Cloth_volume_r_C02_capsule
   center=<-5.8574 3.5836 -0.4>
   rotation=<rot 14.3011 -90 -19.5949>
  }]
 foot_l [{shapeType=Capsule, length=11.0765, radius=8.0344, center=<-4.1811 6.0027 -0.338>, rotation=<rot 0.0388 1.4701 -89.3982>}]
 foot_r [{shapeType=Capsule, length=11.0765, radius=8.0344, center=<4.1811 -6.0027 0.338>, rotation=<rot 0.0388 1.4701 90.6018>}]
 head [{shapeType=Capsule, length=0, radius=18, center=<9.0697 0.0746 0>, rotation=<rot 0 -90 0>}]
 lowerarm_l []
 lowerarm_r []
 neck_02 []
 pelvis [{shapeType=Capsule, length=9.6042, radius=11.1441, center=<-2.4471 -2.5289 -0.1981>}]
 spine_01 [{shapeType=Capsule, length=1.2276, radius=17.5249, center=<-8.9021 -4.3984 0>}]
 spine_02 [
  {shapeType=Capsule, length=0, radius=14.494, center=<-2.6885 -6.326 0>}
  {shapeType=Capsule, length=20.6875, radius=9.3125, center=<-1.8317 1.9362 0>, rotation=<rot 90 0 0>}
 ]
 spine_03 [
  {shapeType=Capsule, length=0, radius=15.7005, center=<0.0574 -3.2995 0>}
  {shapeType=Capsule, length=20.7188, radius=9.2812, center=<-1.2163 0.549 0>, rotation=<rot 90 180 -175.0732>}
 ]
 spine_04 [
  {shapeType=Capsule, length=6.452, radius=12.9473, center=<5.6234 -5.0816 0>}
  {shapeType=Capsule, length=10.0312, radius=9.2812, center=<-1.2326 1.308 0>, rotation=<rot 90 0 0>}
 ]
 spine_05 [{shapeType=Capsule, length=7.6201, radius=14.4149, center=<0.6969 -3.2544 0>}]
 thigh_l []
 thigh_r []
 upperarm_l []
 upperarm_r []
}
constraints [
 {
  bone1Name=dyn_Cloth_volume_r_C02
  bone2Name=pelvis
  swing1LimitDegrees=45
  swing1Motion=Free
  swing2LimitDegrees=45
  swing2Motion=Free
  twistLimitDegrees=45
  twistMotion=Free
 }
 {
  bone1Name=dyn_Cloth_volume_l_C02
  bone2Name=pelvis
  swing1LimitDegrees=45
  swing1Motion=Free
  swing2LimitDegrees=45
  swing2Motion=Free
  twistLimitDegrees=45
  twistMotion=Free
 }
 {
  bone1Name=dyn_Cloth_volume_bk02
  bone2Name=pelvis
  swing1LimitDegrees=45
  swing1Motion=Free
  swing2LimitDegrees=45
  swing2Motion=Free
  twistLimitDegrees=45
  twistMotion=Free
 }
 {
  bone1Name=dyn_Cloth_volume_r_B02
  bone2Name=pelvis
  swing1LimitDegrees=45
  swing1Motion=Free
  swing2LimitDegrees=45
  swing2Motion=Free
  twistLimitDegrees=45
  twistMotion=Free
 }
 {
  bone1Name=dyn_Cloth_volume_l_B02
  bone2Name=pelvis
  swing1LimitDegrees=45
  swing1Motion=Free
  swing2LimitDegrees=45
  swing2Motion=Free
  twistLimitDegrees=45
  twistMotion=Free
 }
]
