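# Control Rig for the female medium OutletVerve body part, with functions for spline-driven bone chains and upper-arm pose reading.
# Spline helpers create nulls, blend transforms by weighted parent constraints, and record bones' initial percents and offsets, later reapplied along the live spline with length compensation.
# UpperArm Readers outputs directional weights from spherical pose readers with optional debugging.
# Its Aim Dn In reader is duplicated, leaving Aim Dn Ot unset.
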ControlRigBlueprint /BRCosmetics/Characters/Player/Female/Medium/Bodies/F_MED_OutletVerve/Meshes/Parts/F_MED_OutletVerve_Spl_CtrlRig
graphs [
 New Function_ContainedGraph {nodes [
   n1 = RigVMFunctionEntryNode()
   > RigVMDispatchNode(Array=RigVMVariableNode() {variable=Result Percents}.Value) {template="DISPATCH_RigVMDispatch_ArrayReset(io Array)"}
   > RigVMDispatchNode(Array=RigVMVariableNode() {variable=Return_Out Local Offsets}.Value) {template="DISPATCH_RigVMDispatch_ArrayReset(io Array)"}
   > For_Each = RigVMDispatchNode(Array=@n1.Driven Bones) {template="DISPATCH_RigVMDispatch_ArrayIterator(in Array,out Element,out Index,out Count,out Ratio)"}
   n2 = &RigUnit_GetTransform(bInitial=true, Item=@For_Each.Element)
   n3 = /Script/ControlRigSpline.RigUnit_ClosestParameterFromControlRigSpline(Position=@n2.Transform.Translation, Spline=@n1.In Spline) {is=struct RigVMUnitNode}
   > RigVMDispatchNode(
    Array=RigVMVariableNode() {variable=Return_Out Local Offsets}.Value
    Element=&RigVMFunction_MathTransformMakeRelative(
     Parent=/Script/ControlRigSpline.RigUnit_TransformFromControlRigSpline2(PrimaryAxis=(X=1,Y=0,Z=0), SecondaryAxis=(X=0,Y=1,Z=0), U=@n3.U, Spline=@n1.In Spline) {is=struct RigVMUnitNode}.Transform
     Global=@n2.Transform
    ).Local
   ) {template="DISPATCH_RigVMDispatch_ArrayAdd(io Array,in Element,out Index)"}
   Divide = &RigVMFunction_MathDoubleDiv(
    A=/Script/ControlRigSpline.RigUnit_GetLengthAtParamControlRigSpline(U=@n3.U, Spline=@n1.In Spline) {is=struct RigVMUnitNode}.Length
    B=/Script/ControlRigSpline.RigUnit_GetLengthControlRigSpline(Spline=@n1.In Spline) {is=struct RigVMUnitNode}.Length
   )
   > RigVMDispatchNode(Element=@Divide.Result, Array=RigVMVariableNode() {variable=Result Percents}.Value) {template="DISPATCH_RigVMDispatch_ArrayAdd(io Array,in Element,out Index)"}
   RigVMFunctionReturnNode(
    Out Percents=RigVMVariableNode() {variable=Result Percents}.Value
    Out Local Offsets=RigVMVariableNode() {variable=Return_Out Local Offsets}.Value
    ExecuteContext=@For_Each.Completed
   )
  ]}
 New Function_ContainedGraph {nodes [
   n1 = RigVMFunctionEntryNode()
   > RigVMDispatchNode(Array=RigVMVariableNode() {variable=Return_Out Transforms}.Value) {template="DISPATCH_RigVMDispatch_ArrayReset(io Array)"}
   > For_Each = RigVMDispatchNode(Array=@n1.In Transforms) {template="DISPATCH_RigVMDispatch_ArrayIterator(in Array,out Element,out Index,out Count,out Ratio)"}
   > RigVMDispatchNode(
    Element=/Script/ControlRig.RigUnit_ParentConstraintMath(
     AdvancedSettings=(InterpolationType=Average)
     Parents=RigVMDispatchNode(Index=@For_Each.Index, Array=@n1.In Weights) {template="DISPATCH_RigVMDispatch_ArrayGetAtIndex(in Array,in Index,out Element)"}.Element.ConstraintParents_3_#1
     Input=@For_Each.Element
    ) {is=struct RigVMUnitNode}.Output
    Array=RigVMVariableNode() {variable=Return_Out Transforms}.Value
   ) {template="DISPATCH_RigVMDispatch_ArrayAdd(io Array,in Element,out Index)"}
   RigVMFunction_ControlFlowBranch = &RigVMFunction_ControlFlowBranch(Condition=RigVMVariableNode() {variable=Display Transforms}.Value, ExecuteContext=@For_Each.Completed)
   RigVMFunctionReturnNode(ExecuteContext=@RigVMFunction_ControlFlowBranch.Completed, Out Transforms=RigVMVariableNode() {variable=Return_Out Transforms}.Value)
   RigVMFunction_DebugTransformArrayMutableNoSpace = /Script/RigVM.RigVMFunction_DebugTransformArrayMutableNoSpace(
    Color=(R=1,G=1,B=1,A=1)
    DebugDrawSettings=(DepthPriority=SDPG_Foreground,Lifetime=-1)
    Mode=Axes
    ParentIndices=()
    Scale=10
    Thickness=0
    WorldOffset=(Rotation=(X=0,Y=0,Z=0,W=1),Translation=(X=0,Y=0,Z=0),Scale3D=(X=1,Y=1,Z=1))
    bEnabled=True
    ExecutePin=@RigVMFunction_ControlFlowBranch.True
    Transforms=RigVMVariableNode() {variable=Return_Out Transforms}.Value
   ) {is=struct RigVMUnitNode}
  ]}
 New Function_ContainedGraph {nodes [
   RigVMFunctionEntryNode()
   > RigVMDispatchNode(Array=RigVMVariableNode() {variable=Return_Out Transforms}.Value) {template="DISPATCH_RigVMDispatch_ArrayReset(io Array)"}
   n1 = &RigUnit_GetLengthControlRigSpline(Spline=RigVMVariableNode() {variable=Tick Spline}.Value)
   > RigVMVariableNode(Value=@n1.Length) {variable=Tick Spline Length}
   Divide = &RigVMFunction_MathDoubleDiv(A=RigVMVariableNode() {variable=Original Length}.Value, B=@n1.Length)
   > RigVMVariableNode(Value=@Divide.Result) {variable=Scale Factor}
   > For_Each = RigVMDispatchNode(Array=RigVMVariableNode() {variable=Chain Percents}.Value) {template="DISPATCH_RigVMDispatch_ArrayIterator(in Array,out Element,out Index,out Count,out Ratio)"}
   n2 = RigVMVariableNode() {variable=Local Transforms}
   n3 = RigVMVariableNode() {variable=Tick Spline}
   Multiply = &RigVMFunction_MathDoubleMul(B=RigVMVariableNode() {variable=Scale Factor}.Value, A=@For_Each.Element)
   n4 = &RigUnit_TransformFromControlRigSpline2(
    PrimaryAxis=(X=1,Y=0,Z=0)
    SecondaryAxis=(X=0,Y=1,Z=0)
    U=/Script/ControlRigSpline.RigUnit_ParameterAtPercentage(Spline=@n3.Value, Percentage=@Multiply.Result) {is=struct RigVMUnitNode}.U
    Spline=@n3.Value
   )
   > &RigUnit_SetTransform(
    Weight=1
    bInitial=False
    bPropagateToChildren=true
    Value=/Script/RigVM.RigVMFunction_MathTransformMakeAbsolute(
     Local=RigVMDispatchNode(Array=@n2.Value, Index=@For_Each.Index) {template="DISPATCH_RigVMDispatch_ArrayGetAtIndex(in Array,in Index,out Element)"}.Element
     Parent=@n4.Transform
    ) {is=struct RigVMUnitNode, template="Make Absolute::Execute(in Local,in Parent,out Global)"}.Global
    Item=RigVMDispatchNode(Array=RigVMVariableNode() {variable=Driven Chain}.Value, Index=@For_Each.Index) {template="DISPATCH_RigVMDispatch_ArrayGetAtIndex(in Array,in Index,out Element)"}.Element
   )
   > /Script/RigVM.RigVMFunction_DebugTransformMutableNoSpace(
    Color=(R=1,G=0,B=0,A=1)
    DebugDrawSettings=(DepthPriority=SDPG_Foreground,Lifetime=-1)
    Mode=Axes
    Scale=10
    Thickness=0
    WorldOffset=(Rotation=(X=0,Y=0,Z=0,W=1),Translation=(X=0,Y=0,Z=0),Scale3D=(X=1,Y=1,Z=1))
    Transform=@n4.Transform
    bEnabled=RigVMVariableNode() {variable=Display Transforms}.Value
   ) {is=struct RigVMUnitNode}
   RigVMFunctionReturnNode(Out Transforms=(), ExecuteContext=@For_Each.Completed)
  ]}
 New Function_ContainedGraph {nodes [
   RigVMFunctionEntryNode()
   > RigVMDispatchNode(Array=RigVMVariableNode() {variable=Return_Out Nulls}.Value) {template="DISPATCH_RigVMDispatch_ArrayReset(io Array)"}
   > For_Each = RigVMDispatchNode(Array=RigVMVariableNode() {variable=In Transforms}.Value) {template="DISPATCH_RigVMDispatch_ArrayIterator(in Array,out Element,out Index,out Count,out Ratio)"}
   Int to Name = /Script/RigVM.RigVMFunction_MathIntToName(PaddedSize=2, Number=@For_Each.Index) {is=struct RigVMUnitNode, template="Int to Name::Execute(in Number,in PaddedSize,out Result)"}
   Concat = /Script/RigVM.RigVMFunction_NameConcat(B=@"Int to Name".Result, A=RigVMVariableNode() {variable=In Name}.Value) {is=struct RigVMUnitNode, template="Concat::Execute(in A,in B,out Result)"}
   > n1 = &RigUnit_HierarchyAddNull(Parent=(Type=None,Name="None"), Space=LocalSpace, Name=@Concat.Result, Transform=@For_Each.Element)
   > RigVMDispatchNode(Array=RigVMVariableNode() {variable=Return_Out Nulls}.Value, Element=@n1.Item) {template="DISPATCH_RigVMDispatch_ArrayAdd(io Array,in Element,out Index)"}
   RigVMFunctionReturnNode(Out Nulls=RigVMVariableNode() {variable=Return_Out Nulls}.Value, ExecuteContext=@For_Each.Completed)
  ]}
 New Function_ContainedGraph {nodes [
   n1 = RigVMFunctionEntryNode()
   > RigVMVariableNode(Value=/Script/ControlRig.RigUnit_HierarchyGetParent(bDefaultParent=True, Child=@n1.Driver Item) {is=struct RigVMUnitNode}.Parent) {variable=Driver Item Parent}
   To_String = RigVMDispatchNode(Value=@n1.Driver Item.Name) {template="DISPATCH_RigDispatch_ToString(in Value,out Result)"}
   > RigVMVariableNode(Value=@To_String.Result) {variable=Driver Item Name}
   > RigVMFunction_Sequence = RigVMAggregateNode()
   RigVMFunctionReturnNode(
    ExecuteContext=@RigVMFunction_Sequence.F
    Up Post=RigVMVariableNode() {variable=Result Up Post}.Value
    Aim Up In=RigVMVariableNode() {variable=Result Aim Up In}.Value
    Aim Up Ot=RigVMVariableNode() {variable=Result Aim Up Ot}.Value
    Aim Dn In=RigVMVariableNode() {variable=Result Aim Dn In}.Value
    Aim Dn Ot=RigVMVariableNode() {variable=Result Aim Dn Ot}.Value
    Dn=RigVMVariableNode() {variable=Result Dn}.Value
    Dn Pre=RigVMVariableNode() {variable=Result Dn Pre}.Value
    Dn Post=RigVMVariableNode() {variable=Result Dn Post}.Value
    Fr=RigVMVariableNode() {variable=Result Fr}.Value
    Fr Pre=RigVMVariableNode() {variable=Result Fr Pre}.Value
    Fr Post=RigVMVariableNode() {variable=Result Fr Post}.Value
    Fr Up=RigVMVariableNode() {variable=Result Fr Up}.Value
    Fr Dn=RigVMVariableNode() {variable=Result Fr Dn}.Value
    Ba=RigVMVariableNode() {variable=Result Ba}.Value
    Ba Pre=RigVMVariableNode() {variable=Result Ba Pre}.Value
    Ba Post=RigVMVariableNode() {variable=Result Ba Post}.Value
    Up=RigVMVariableNode() {variable=Result Up}.Value
    Ba Up=RigVMVariableNode() {variable=Result Ba Up}.Value
    Ba Dn=RigVMVariableNode() {variable=Result Ba Dn}.Value
    Up Pre=RigVMVariableNode() {variable=Result Up Pre}.Value
   )
   n2 = &RigUnit_SphericalPoseReader(
    ActiveRegionScaleFactors=(PositiveWidth=1,NegativeWidth=1,PositiveHeight=1,NegativeHeight=1)
    ActiveRegionSize=0.001
    Debug=(bDrawDebug=false,bDraw2D=False,bDrawLocalAxes=False,DebugScale=25,DebugSegments=20,DebugThickness=0.05)
    DriverAxis=(X=1,Y=0,Z=0)
    FalloffRegionScaleFactors=(PositiveWidth=1.3471,NegativeWidth=1,PositiveHeight=1,NegativeHeight=1)
    FalloffSize=0.5
    FlipHeightScaling=False
    FlipWidthScaling=False
    RotationOffset=(X=0,Y=31.239,Z=0)
    Debug.bDrawDebug=RigVMVariableNode() {variable=Up Viz}.Value
    DriverItem=RigVMVariableNode() {variable=Driver Item}.Value
    OptionalParentItem=RigVMVariableNode() {variable=Driver Item Parent}.Value
    ExecutePin=@RigVMFunction_Sequence.A
   )
   > n3 = RigVMVariableNode(Value=@n2.OutputParam) {variable=Result Up}
   RigVMFunction_ControlFlowBranch = &RigVMFunction_ControlFlowBranch(ExecuteContext=@n3.ExecuteContext, Condition=RigVMVariableNode() {variable=Up Viz}.Value)
   n4 = &RigUnit_SphericalPoseReader(
    ActiveRegionScaleFactors=(PositiveWidth=1,NegativeWidth=1,PositiveHeight=1,NegativeHeight=1)
    ActiveRegionSize=0.001
    Debug=(bDrawDebug=false,bDraw2D=False,bDrawLocalAxes=False,DebugScale=25,DebugSegments=20,DebugThickness=0.05)
    DriverAxis=(X=1,Y=0,Z=0)
    FalloffRegionScaleFactors=(PositiveWidth=1.6942,NegativeWidth=1,PositiveHeight=2,NegativeHeight=2)
    FalloffSize=0.25
    FlipHeightScaling=False
    FlipWidthScaling=False
    RotationOffset=(X=0,Y=-13.761,Z=0)
    ExecutePin=@RigVMFunction_ControlFlowBranch.Completed
    DriverItem=RigVMVariableNode() {variable=Driver Item}.Value
    OptionalParentItem=RigVMVariableNode() {variable=Driver Item Parent}.Value
    Debug.bDrawDebug=RigVMVariableNode() {variable=Up Pre Viz}.Value
   )
   > n5 = RigVMVariableNode(Value=@n4.OutputParam) {variable=Result Up Pre}
   n6 = &RigVMFunction_ControlFlowBranch(Condition=RigVMVariableNode() {variable=Up Pre Viz}.Value, ExecuteContext=@n5.ExecuteContext)
   n7 = &RigUnit_SphericalPoseReader(
    ActiveRegionScaleFactors=(PositiveWidth=1,NegativeWidth=1,PositiveHeight=1,NegativeHeight=1)
    ActiveRegionSize=0.001
    Debug=(bDrawDebug=false,bDraw2D=False,bDrawLocalAxes=False,DebugScale=25,DebugSegments=20,DebugThickness=0.05)
    DriverAxis=(X=1,Y=0,Z=0)
    FalloffRegionScaleFactors=(PositiveWidth=0.5,NegativeWidth=0.5,PositiveHeight=1,NegativeHeight=1)
    FalloffSize=0.5
    FlipHeightScaling=False
    FlipWidthScaling=False
    RotationOffset=(X=0,Y=76.239,Z=0)
    ExecutePin=@n6.Completed
    Debug.bDrawDebug=RigVMVariableNode() {variable=Up Post Viz}.Value
    DriverItem=RigVMVariableNode() {variable=Driver Item}.Value
    OptionalParentItem=RigVMVariableNode() {variable=Driver Item Parent}.Value
   )
   > n8 = RigVMVariableNode(Value=@n7.OutputParam) {variable=Result Up Post}
   n9 = &RigVMFunction_ControlFlowBranch(Condition=RigVMVariableNode() {variable=Up Post Viz}.Value, ExecuteContext=@n8.ExecuteContext)
   RigVMDispatchNode(
    Enabled=True
    ScreenColor=(R=1,G=1,B=1,A=1)
    ScreenDuration=0.05
    Prefix=/Script/RigVM.RigVMFunction_StringConcat(B=" - UP POST:   ", A=RigVMVariableNode() {variable=Driver Item Name}.Value) {is=struct RigVMUnitNode, template="Concat::Execute(in A,in B,out Result)"}.Result
    Value=RigVMVariableNode() {variable=Result Up Post}.Value
    ExecuteContext=@n9.True
   ) {template="DISPATCH_RigVMDispatch_Print(in Prefix,in Value,in Enabled,in ScreenDuration,in ScreenColor)"}
   RigVMDispatchNode(
    Enabled=True
    ScreenColor=(R=1,G=1,B=1,A=1)
    ScreenDuration=0.05
    Prefix=&RigVMFunction_StringConcat(B=" - UP PRE:   ", A=RigVMVariableNode() {variable=Driver Item Name}.Value) {template="Concat::Execute(in A,in B,out Result)"}.Result
    Value=RigVMVariableNode() {variable=Result Up Pre}.Value
    ExecuteContext=@n6.True
   ) {template="DISPATCH_RigVMDispatch_Print(in Prefix,in Value,in Enabled,in ScreenDuration,in ScreenColor)"}
   Concat = &RigVMFunction_StringConcat(B=" - UP:   ", A=RigVMVariableNode() {variable=Driver Item Name}.Value) {template="Concat::Execute(in A,in B,out Result)"}
   RigVMDispatchNode(
    Enabled=True
    ScreenColor=(R=1,G=1,B=1,A=1)
    ScreenDuration=0.05
    Prefix=@Concat.Result
    Value=RigVMVariableNode() {variable=Result Up}.Value
    ExecuteContext=@RigVMFunction_ControlFlowBranch.True
   ) {template="DISPATCH_RigVMDispatch_Print(in Prefix,in Value,in Enabled,in ScreenDuration,in ScreenColor)"}
   n10 = &RigUnit_SphericalPoseReader(
    ActiveRegionScaleFactors=(PositiveWidth=1,NegativeWidth=1,PositiveHeight=1,NegativeHeight=1)
    ActiveRegionSize=0.001
    Debug=(bDrawDebug=false,bDraw2D=False,bDrawLocalAxes=False,DebugScale=25,DebugSegments=20,DebugThickness=0.05)
    DriverAxis=(X=1,Y=0,Z=0)
    FalloffRegionScaleFactors=(PositiveWidth=1,NegativeWidth=1,PositiveHeight=1,NegativeHeight=1)
    FalloffSize=0.5
    FlipHeightScaling=False
    FlipWidthScaling=False
    RotationOffset=(X=90,Y=31.239,Z=0)
    DriverItem=RigVMVariableNode() {variable=Driver Item}.Value
    OptionalParentItem=RigVMVariableNode() {variable=Driver Item Parent}.Value
    Debug.bDrawDebug=RigVMVariableNode() {variable=Fr Viz}.Value
    ExecutePin=@RigVMFunction_Sequence.C
   )
   > n11 = RigVMVariableNode(Value=@n10.OutputParam) {variable=Result Fr}
   n12 = &RigVMFunction_ControlFlowBranch(ExecuteContext=@n11.ExecuteContext, Condition=RigVMVariableNode() {variable=Fr Viz}.Value)
   n13 = &RigUnit_SphericalPoseReader(
    ActiveRegionScaleFactors=(PositiveWidth=1,NegativeWidth=1,PositiveHeight=1,NegativeHeight=1)
    ActiveRegionSize=0.001
    Debug=(bDrawDebug=false,bDraw2D=False,bDrawLocalAxes=False,DebugScale=25,DebugSegments=20,DebugThickness=0.05)
    DriverAxis=(X=1,Y=0,Z=0)
    FalloffRegionScaleFactors=(PositiveWidth=2,NegativeWidth=2,PositiveHeight=1,NegativeHeight=1)
    FalloffSize=0.25
    FlipHeightScaling=False
    FlipWidthScaling=False
    RotationOffset=(X=45,Y=-58.761,Z=0)
    ExecutePin=@n12.Completed
    DriverItem=RigVMVariableNode() {variable=Driver Item}.Value
    OptionalParentItem=RigVMVariableNode() {variable=Driver Item Parent}.Value
    Debug.bDrawDebug=RigVMVariableNode() {variable=Fr Pre Viz}.Value
   )
   > n14 = RigVMVariableNode(Value=@n13.OutputParam) {variable=Result Fr Pre}
   n15 = &RigVMFunction_ControlFlowBranch(ExecuteContext=@n14.ExecuteContext, Condition=RigVMVariableNode() {variable=Fr Pre Viz}.Value)
   n16 = &RigUnit_SphericalPoseReader(
    ActiveRegionScaleFactors=(PositiveWidth=1,NegativeWidth=1,PositiveHeight=1,NegativeHeight=1)
    ActiveRegionSize=0.001
    Debug=(bDrawDebug=false,bDraw2D=False,bDrawLocalAxes=False,DebugScale=25,DebugSegments=20,DebugThickness=0.05)
    DriverAxis=(X=1,Y=0,Z=0)
    FalloffRegionScaleFactors=(PositiveWidth=2,NegativeWidth=2,PositiveHeight=1,NegativeHeight=1)
    FalloffSize=0.25
    FlipHeightScaling=False
    FlipWidthScaling=False
    RotationOffset=(X=135,Y=-58.761,Z=0)
    ExecutePin=@n15.Completed
    DriverItem=RigVMVariableNode() {variable=Driver Item}.Value
    OptionalParentItem=RigVMVariableNode() {variable=Driver Item Parent}.Value
    Debug.bDrawDebug=RigVMVariableNode() {variable=Fr Post Viz}.Value
   )
   > n17 = RigVMVariableNode(Value=@n16.OutputParam) {variable=Result Fr Post}
   n18 = &RigVMFunction_ControlFlowBranch(ExecuteContext=@n17.ExecuteContext, Condition=RigVMVariableNode() {variable=Fr Post Viz}.Value)
   n19 = &RigUnit_SphericalPoseReader(
    ActiveRegionScaleFactors=(PositiveWidth=1,NegativeWidth=1,PositiveHeight=1,NegativeHeight=1)
    ActiveRegionSize=0.001
    Debug=(bDrawDebug=false,bDraw2D=False,bDrawLocalAxes=False,DebugScale=25,DebugSegments=20,DebugThickness=0.05)
    DriverAxis=(X=1,Y=0,Z=0)
    FalloffRegionScaleFactors=(PositiveWidth=1,NegativeWidth=1,PositiveHeight=0.5,NegativeHeight=0.5)
    FalloffSize=0.5
    FlipHeightScaling=False
    FlipWidthScaling=False
    RotationOffset=(X=45,Y=31.239,Z=0)
    ExecutePin=@n18.Completed
    DriverItem=RigVMVariableNode() {variable=Driver Item}.Value
    OptionalParentItem=RigVMVariableNode() {variable=Driver Item Parent}.Value
    Debug.bDrawDebug=RigVMVariableNode() {variable=Fr Up Viz}.Value
   )
   > n20 = RigVMVariableNode(Value=@n19.OutputParam) {variable=Result Fr Up}
   n21 = &RigVMFunction_ControlFlowBranch(ExecuteContext=@n20.ExecuteContext, Condition=RigVMVariableNode() {variable=Fr Up Viz}.Value)
   n22 = &RigUnit_SphericalPoseReader(
    ActiveRegionScaleFactors=(PositiveWidth=1,NegativeWidth=1,PositiveHeight=1,NegativeHeight=1)
    ActiveRegionSize=0.001
    Debug=(bDrawDebug=false,bDraw2D=False,bDrawLocalAxes=False,DebugScale=25,DebugSegments=20,DebugThickness=0.05)
    DriverAxis=(X=1,Y=0,Z=0)
    FalloffRegionScaleFactors=(PositiveWidth=1,NegativeWidth=1,PositiveHeight=0.5,NegativeHeight=0.5)
    FalloffSize=0.5
    FlipHeightScaling=False
    FlipWidthScaling=False
    RotationOffset=(X=135,Y=31.239,Z=0)
    ExecutePin=@n21.Completed
    DriverItem=RigVMVariableNode() {variable=Driver Item}.Value
    OptionalParentItem=RigVMVariableNode() {variable=Driver Item Parent}.Value
    Debug.bDrawDebug=RigVMVariableNode() {variable=Fr Dn Viz}.Value
   )
   > n23 = RigVMVariableNode(Value=@n22.OutputParam) {variable=Result Fr Dn}
   n24 = &RigVMFunction_ControlFlowBranch(ExecuteContext=@n23.ExecuteContext, Condition=RigVMVariableNode() {variable=Fr Dn Viz}.Value)
   RigVMDispatchNode(
    Enabled=True
    ScreenColor=(R=1,G=1,B=1,A=1)
    ScreenDuration=0.05
    Prefix=&RigVMFunction_StringConcat(B=" - FR DN:   ", A=RigVMVariableNode() {variable=Driver Item Name}.Value) {template="Concat::Execute(in A,in B,out Result)"}.Result
    ExecuteContext=@n24.True
    Value=RigVMVariableNode() {variable=Result Fr Dn}.Value
   ) {template="DISPATCH_RigVMDispatch_Print(in Prefix,in Value,in Enabled,in ScreenDuration,in ScreenColor)"}
   RigVMDispatchNode(
    Enabled=True
    ScreenColor=(R=1,G=1,B=1,A=1)
    ScreenDuration=0.05
    Prefix=&RigVMFunction_StringConcat(B=" - FR UP:   ", A=RigVMVariableNode() {variable=Driver Item Name}.Value) {template="Concat::Execute(in A,in B,out Result)"}.Result
    ExecuteContext=@n21.True
    Value=RigVMVariableNode() {variable=Result Fr Up}.Value
   ) {template="DISPATCH_RigVMDispatch_Print(in Prefix,in Value,in Enabled,in ScreenDuration,in ScreenColor)"}
   RigVMDispatchNode(
    Enabled=True
    ScreenColor=(R=1,G=1,B=1,A=1)
    ScreenDuration=0.05
    Prefix=&RigVMFunction_StringConcat(B=" - FR POST:   ", A=RigVMVariableNode() {variable=Driver Item Name}.Value) {template="Concat::Execute(in A,in B,out Result)"}.Result
    ExecuteContext=@n18.True
    Value=RigVMVariableNode() {variable=Result Fr Post}.Value
   ) {template="DISPATCH_RigVMDispatch_Print(in Prefix,in Value,in Enabled,in ScreenDuration,in ScreenColor)"}
   RigVMDispatchNode(
    Enabled=True
    ScreenColor=(R=1,G=1,B=1,A=1)
    ScreenDuration=0.05
    Prefix=&RigVMFunction_StringConcat(B=" - FR PRE:   ", A=RigVMVariableNode() {variable=Driver Item Name}.Value) {template="Concat::Execute(in A,in B,out Result)"}.Result
    ExecuteContext=@n15.True
    Value=RigVMVariableNode() {variable=Result Fr Pre}.Value
   ) {template="DISPATCH_RigVMDispatch_Print(in Prefix,in Value,in Enabled,in ScreenDuration,in ScreenColor)"}
   RigVMDispatchNode(
    Enabled=True
    ScreenColor=(R=1,G=1,B=1,A=1)
    ScreenDuration=0.05
    Prefix=&RigVMFunction_StringConcat(B=" - FR:   ", A=RigVMVariableNode() {variable=Driver Item Name}.Value) {template="Concat::Execute(in A,in B,out Result)"}.Result
    ExecuteContext=@n12.True
    Value=RigVMVariableNode() {variable=Result Fr}.Value
   ) {template="DISPATCH_RigVMDispatch_Print(in Prefix,in Value,in Enabled,in ScreenDuration,in ScreenColor)"}
   n25 = &RigUnit_SphericalPoseReader(
    ActiveRegionScaleFactors=(PositiveWidth=1,NegativeWidth=1,PositiveHeight=1,NegativeHeight=1)
    ActiveRegionSize=0.001
    Debug=(bDrawDebug=true,bDraw2D=False,bDrawLocalAxes=False,DebugScale=25,DebugSegments=20,DebugThickness=0.05)
    DriverAxis=(X=1,Y=0,Z=0)
    FalloffRegionScaleFactors=(PositiveWidth=1,NegativeWidth=1,PositiveHeight=1,NegativeHeight=1)
    FalloffSize=0.18
    FlipHeightScaling=False
    FlipWidthScaling=False
    RotationOffset=(X=28.4478,Y=74.4998,Z=0)
    DriverItem=RigVMVariableNode() {variable=Driver Item}.Value
    OptionalParentItem=RigVMVariableNode() {variable=Driver Item Parent}.Value
    Debug.bDrawDebug=RigVMVariableNode() {variable=Aim Up In Viz}.Value
    ExecutePin=@RigVMFunction_Sequence.E
   )
   > n26 = RigVMVariableNode(Value=@n25.OutputParam) {variable=Result Aim Up In}
   n27 = &RigVMFunction_ControlFlowBranch(Condition=RigVMVariableNode() {variable=Aim Up In Viz}.Value, ExecuteContext=@n26.ExecuteContext)
   n28 = &RigUnit_SphericalPoseReader(
    ActiveRegionScaleFactors=(PositiveWidth=1,NegativeWidth=1,PositiveHeight=1,NegativeHeight=1)
    ActiveRegionSize=0.001
    Debug=(bDrawDebug=true,bDraw2D=False,bDrawLocalAxes=False,DebugScale=25,DebugSegments=20,DebugThickness=0.05)
    DriverAxis=(X=1,Y=0,Z=0)
    FalloffRegionScaleFactors=(PositiveWidth=1,NegativeWidth=1,PositiveHeight=1,NegativeHeight=1)
    FalloffSize=0.18
    FlipHeightScaling=False
    FlipWidthScaling=False
    RotationOffset=(X=31.239,Y=-1.4062,Z=0)
    ExecutePin=@n27.Completed
    DriverItem=RigVMVariableNode() {variable=Driver Item}.Value
    OptionalParentItem=RigVMVariableNode() {variable=Driver Item Parent}.Value
    Debug.bDrawDebug=RigVMVariableNode() {variable=Aim Up Ot Viz}.Value
   )
   > n29 = RigVMVariableNode(Value=@n28.OutputParam) {variable=Result Aim Up Ot}
   n30 = &RigVMFunction_ControlFlowBranch(Condition=RigVMVariableNode() {variable=Aim Up Ot Viz}.Value, ExecuteContext=@n29.ExecuteContext)
   n31 = &RigUnit_SphericalPoseReader(
    ActiveRegionScaleFactors=(PositiveWidth=1,NegativeWidth=1,PositiveHeight=1,NegativeHeight=1)
    ActiveRegionSize=0.001
    Debug=(bDrawDebug=true,bDraw2D=False,bDrawLocalAxes=False,DebugScale=25,DebugSegments=20,DebugThickness=0.05)
    DriverAxis=(X=1,Y=0,Z=0)
    FalloffRegionScaleFactors=(PositiveWidth=1,NegativeWidth=1,PositiveHeight=1,NegativeHeight=1)
    FalloffSize=0.18
    FlipHeightScaling=False
    FlipWidthScaling=False
    RotationOffset=(X=147.9195,Y=-7.4723,Z=1.8383)
    ExecutePin=@n30.Completed
    DriverItem=RigVMVariableNode() {variable=Driver Item}.Value
    OptionalParentItem=RigVMVariableNode() {variable=Driver Item Parent}.Value
    Debug.bDrawDebug=RigVMVariableNode() {variable=Aim Dn In Viz}.Value
   )
   > n32 = RigVMVariableNode(Value=@n31.OutputParam) {variable=Result Aim Dn In}
   n33 = &RigVMFunction_ControlFlowBranch(Condition=RigVMVariableNode() {variable=Aim Dn In Viz}.Value, ExecuteContext=@n32.ExecuteContext)
   n34 = &RigUnit_SphericalPoseReader(
    ActiveRegionScaleFactors=(PositiveWidth=1,NegativeWidth=1,PositiveHeight=1,NegativeHeight=1)
    ActiveRegionSize=0.001
    Debug=(bDrawDebug=true,bDraw2D=False,bDrawLocalAxes=False,DebugScale=25,DebugSegments=20,DebugThickness=0.05)
    DriverAxis=(X=1,Y=0,Z=0)
    FalloffRegionScaleFactors=(PositiveWidth=1,NegativeWidth=1,PositiveHeight=1,NegativeHeight=1)
    FalloffSize=0.18
    FlipHeightScaling=False
    FlipWidthScaling=False
    RotationOffset=(X=147.9195,Y=-7.4723,Z=1.8383)
    ExecutePin=@n33.Completed
    DriverItem=RigVMVariableNode() {variable=Driver Item}.Value
    OptionalParentItem=RigVMVariableNode() {variable=Driver Item Parent}.Value
    Debug.bDrawDebug=RigVMVariableNode() {variable=Aim Dn In Viz}.Value
   )
   > n35 = RigVMVariableNode(Value=@n34.OutputParam) {variable=Result Aim Dn In}
   n36 = &RigVMFunction_ControlFlowBranch(Condition=RigVMVariableNode() {variable=Aim Dn In Viz}.Value, ExecuteContext=@n35.ExecuteContext)
   RigVMDispatchNode(
    Enabled=True
    ScreenColor=(R=1,G=1,B=1,A=1)
    ScreenDuration=0.05
    Prefix=&RigVMFunction_StringConcat(B=" - AIM DN IN:   ", A=RigVMVariableNode() {variable=Driver Item Name}.Value) {template="Concat::Execute(in A,in B,out Result)"}.Result
    ExecuteContext=@n36.True
    Value=RigVMVariableNode() {variable=Result Aim Dn In}.Value
   ) {template="DISPATCH_RigVMDispatch_Print(in Prefix,in Value,in Enabled,in ScreenDuration,in ScreenColor)"}
   RigVMDispatchNode(
    Enabled=True
    ScreenColor=(R=1,G=1,B=1,A=1)
    ScreenDuration=0.05
    Prefix=&RigVMFunction_StringConcat(B=" - AIM DN IN:   ", A=RigVMVariableNode() {variable=Driver Item Name}.Value) {template="Concat::Execute(in A,in B,out Result)"}.Result
    ExecuteContext=@n33.True
    Value=RigVMVariableNode() {variable=Result Aim Dn In}.Value
   ) {template="DISPATCH_RigVMDispatch_Print(in Prefix,in Value,in Enabled,in ScreenDuration,in ScreenColor)"}
   RigVMDispatchNode(
    Enabled=True
    ScreenColor=(R=1,G=1,B=1,A=1)
    ScreenDuration=0.05
    Prefix=&RigVMFunction_StringConcat(B=" - AIM UP OT:   ", A=RigVMVariableNode() {variable=Driver Item Name}.Value) {template="Concat::Execute(in A,in B,out Result)"}.Result
    ExecuteContext=@n30.True
    Value=RigVMVariableNode() {variable=Result Aim Up Ot}.Value
   ) {template="DISPATCH_RigVMDispatch_Print(in Prefix,in Value,in Enabled,in ScreenDuration,in ScreenColor)"}
   RigVMDispatchNode(
    Enabled=True
    ScreenColor=(R=1,G=1,B=1,A=1)
    ScreenDuration=0.05
    Prefix=&RigVMFunction_StringConcat(B=" - AIM UP IN:   ", A=RigVMVariableNode() {variable=Driver Item Name}.Value) {template="Concat::Execute(in A,in B,out Result)"}.Result
    ExecuteContext=@n27.True
    Value=RigVMVariableNode() {variable=Result Aim Up In}.Value
   ) {template="DISPATCH_RigVMDispatch_Print(in Prefix,in Value,in Enabled,in ScreenDuration,in ScreenColor)"}
   n37 = &RigUnit_SphericalPoseReader(
    ActiveRegionScaleFactors=(PositiveWidth=1,NegativeWidth=1,PositiveHeight=1,NegativeHeight=1)
    ActiveRegionSize=0.001
    Debug=(bDrawDebug=false,bDraw2D=False,bDrawLocalAxes=False,DebugScale=25,DebugSegments=20,DebugThickness=0.05)
    DriverAxis=(X=1,Y=0,Z=0)
    ExecutePin=()
    FalloffRegionScaleFactors=(PositiveWidth=1,NegativeWidth=1,PositiveHeight=1,NegativeHeight=1)
    FalloffSize=0.5
    FlipHeightScaling=False
    FlipWidthScaling=False
    RotationOffset=(X=-90,Y=31.239,Z=0)
    DriverItem=RigVMVariableNode() {variable=Driver Item}.Value
    OptionalParentItem=RigVMVariableNode() {variable=Driver Item Parent}.Value
    Debug.bDrawDebug=RigVMVariableNode() {variable=Ba Viz}.Value
   )
   > n38 = RigVMVariableNode(Value=@n37.OutputParam) {variable=Result Ba}
   n39 = &RigVMFunction_ControlFlowBranch(ExecuteContext=@n38.ExecuteContext, Condition=RigVMVariableNode() {variable=Ba Viz}.Value)
   n40 = &RigUnit_SphericalPoseReader(
    ActiveRegionScaleFactors=(PositiveWidth=1,NegativeWidth=1,PositiveHeight=1,NegativeHeight=1)
    ActiveRegionSize=0.001
    Debug=(bDrawDebug=false,bDraw2D=False,bDrawLocalAxes=False,DebugScale=25,DebugSegments=20,DebugThickness=0.05)
    DriverAxis=(X=1,Y=0,Z=0)
    FalloffRegionScaleFactors=(PositiveWidth=2,NegativeWidth=2,PositiveHeight=1,NegativeHeight=1)
    FalloffSize=0.25
    FlipHeightScaling=False
    FlipWidthScaling=False
    RotationOffset=(X=-45,Y=-58.761,Z=0)
    ExecutePin=@n39.Completed
    Debug.bDrawDebug=RigVMVariableNode() {variable=Ba Pre Viz}.Value
    DriverItem=RigVMVariableNode() {variable=Driver Item}.Value
    OptionalParentItem=RigVMVariableNode() {variable=Driver Item Parent}.Value
   )
   > n41 = RigVMVariableNode(Value=@n40.OutputParam) {variable=Result Ba Pre}
   n42 = &RigVMFunction_ControlFlowBranch(Condition=RigVMVariableNode() {variable=Ba Pre Viz}.Value, ExecuteContext=@n41.ExecuteContext)
   n43 = &RigUnit_SphericalPoseReader(
    ActiveRegionScaleFactors=(PositiveWidth=1,NegativeWidth=1,PositiveHeight=1,NegativeHeight=1)
    ActiveRegionSize=0.001
    Debug=(bDrawDebug=false,bDraw2D=False,bDrawLocalAxes=False,DebugScale=25,DebugSegments=20,DebugThickness=0.05)
    DriverAxis=(X=1,Y=0,Z=0)
    FalloffRegionScaleFactors=(PositiveWidth=2,NegativeWidth=2,PositiveHeight=1,NegativeHeight=1)
    FalloffSize=0.25
    FlipHeightScaling=False
    FlipWidthScaling=False
    RotationOffset=(X=-135,Y=-58.761,Z=0)
    ExecutePin=@n42.Completed
    Debug.bDrawDebug=RigVMVariableNode() {variable=Ba Post Viz}.Value
    DriverItem=RigVMVariableNode() {variable=Driver Item}.Value
    OptionalParentItem=RigVMVariableNode() {variable=Driver Item Parent}.Value
   )
   > n44 = RigVMVariableNode(Value=@n43.OutputParam) {variable=Result Ba Post}
   n45 = &RigVMFunction_ControlFlowBranch(Condition=RigVMVariableNode() {variable=Ba Post Viz}.Value, ExecuteContext=@n44.ExecuteContext)
   n46 = &RigUnit_SphericalPoseReader(
    ActiveRegionScaleFactors=(PositiveWidth=1,NegativeWidth=1,PositiveHeight=1,NegativeHeight=1)
    ActiveRegionSize=0.001
    Debug=(bDrawDebug=false,bDraw2D=False,bDrawLocalAxes=False,DebugScale=25,DebugSegments=20,DebugThickness=0.05)
    DriverAxis=(X=1,Y=0,Z=0)
    FalloffRegionScaleFactors=(PositiveWidth=1,NegativeWidth=1,PositiveHeight=0.5,NegativeHeight=0.5)
    FalloffSize=0.5
    FlipHeightScaling=False
    FlipWidthScaling=False
    RotationOffset=(X=-45,Y=31.239,Z=0)
    ExecutePin=@n45.Completed
    Debug.bDrawDebug=RigVMVariableNode() {variable=Ba Up Viz}.Value
    DriverItem=RigVMVariableNode() {variable=Driver Item}.Value
    OptionalParentItem=RigVMVariableNode() {variable=Driver Item Parent}.Value
   )
   > n47 = RigVMVariableNode(Value=@n46.OutputParam) {variable=Result Ba Up}
   n48 = &RigVMFunction_ControlFlowBranch(Condition=RigVMVariableNode() {variable=Ba Up Viz}.Value, ExecuteContext=@n47.ExecuteContext)
   n49 = &RigUnit_SphericalPoseReader(
    ActiveRegionScaleFactors=(PositiveWidth=1,NegativeWidth=1,PositiveHeight=1,NegativeHeight=1)
    ActiveRegionSize=0.001
    Debug=(bDrawDebug=false,bDraw2D=False,bDrawLocalAxes=False,DebugScale=25,DebugSegments=20,DebugThickness=0.05)
    DriverAxis=(X=1,Y=0,Z=0)
    FalloffRegionScaleFactors=(PositiveWidth=1,NegativeWidth=1,PositiveHeight=0.5,NegativeHeight=0.5)
    FalloffSize=0.5
    FlipHeightScaling=False
    FlipWidthScaling=False
    RotationOffset=(X=-135,Y=31.239,Z=0)
    ExecutePin=@n48.Completed
    Debug.bDrawDebug=RigVMVariableNode() {variable=Ba Dn Viz}.Value
    DriverItem=RigVMVariableNode() {variable=Driver Item}.Value
    OptionalParentItem=RigVMVariableNode() {variable=Driver Item Parent}.Value
   )
   > n50 = RigVMVariableNode(Value=@n49.OutputParam) {variable=Result Ba Dn}
   n51 = &RigVMFunction_ControlFlowBranch(Condition=RigVMVariableNode() {variable=Ba Dn Viz}.Value, ExecuteContext=@n50.ExecuteContext)
   RigVMDispatchNode(
    Enabled=True
    ScreenColor=(R=1,G=1,B=1,A=1)
    ScreenDuration=0.05
    Prefix=&RigVMFunction_StringConcat(B=" - BA DN:   ", A=RigVMVariableNode() {variable=Driver Item Name}.Value) {template="Concat::Execute(in A,in B,out Result)"}.Result
    ExecuteContext=@n51.True
    Value=RigVMVariableNode() {variable=Result Ba Dn}.Value
   ) {template="DISPATCH_RigVMDispatch_Print(in Prefix,in Value,in Enabled,in ScreenDuration,in ScreenColor)"}
   RigVMDispatchNode(
    Enabled=True
    ScreenColor=(R=1,G=1,B=1,A=1)
    ScreenDuration=0.05
    Prefix=&RigVMFunction_StringConcat(B=" - BA UP:   ", A=RigVMVariableNode() {variable=Driver Item Name}.Value) {template="Concat::Execute(in A,in B,out Result)"}.Result
    ExecuteContext=@n48.True
    Value=RigVMVariableNode() {variable=Result Ba Up}.Value
   ) {template="DISPATCH_RigVMDispatch_Print(in Prefix,in Value,in Enabled,in ScreenDuration,in ScreenColor)"}
   RigVMDispatchNode(
    Enabled=True
    ScreenColor=(R=1,G=1,B=1,A=1)
    ScreenDuration=0.05
    Prefix=&RigVMFunction_StringConcat(B=" - BA POST:   ", A=RigVMVariableNode() {variable=Driver Item Name}.Value) {template="Concat::Execute(in A,in B,out Result)"}.Result
    ExecuteContext=@n45.True
    Value=RigVMVariableNode() {variable=Result Ba Post}.Value
   ) {template="DISPATCH_RigVMDispatch_Print(in Prefix,in Value,in Enabled,in ScreenDuration,in ScreenColor)"}
   RigVMDispatchNode(
    Enabled=True
    ScreenColor=(R=1,G=1,B=1,A=1)
    ScreenDuration=0.05
    Prefix=&RigVMFunction_StringConcat(B=" - BA PRE:   ", A=RigVMVariableNode() {variable=Driver Item Name}.Value) {template="Concat::Execute(in A,in B,out Result)"}.Result
    ExecuteContext=@n42.True
    Value=RigVMVariableNode() {variable=Result Ba Pre}.Value
   ) {template="DISPATCH_RigVMDispatch_Print(in Prefix,in Value,in Enabled,in ScreenDuration,in ScreenColor)"}
   RigVMDispatchNode(
    Enabled=True
    ScreenColor=(R=1,G=1,B=1,A=1)
    ScreenDuration=0.05
    Prefix=&RigVMFunction_StringConcat(B=" - BA:   ", A=RigVMVariableNode() {variable=Driver Item Name}.Value) {template="Concat::Execute(in A,in B,out Result)"}.Result
    ExecuteContext=@n39.True
    Value=RigVMVariableNode() {variable=Result Ba}.Value
   ) {template="DISPATCH_RigVMDispatch_Print(in Prefix,in Value,in Enabled,in ScreenDuration,in ScreenColor)"}
   n52 = &RigUnit_SphericalPoseReader(
    ActiveRegionScaleFactors=(PositiveWidth=1,NegativeWidth=1,PositiveHeight=1,NegativeHeight=1)
    ActiveRegionSize=0.001
    Debug=(bDrawDebug=false,bDraw2D=False,bDrawLocalAxes=False,DebugScale=25,DebugSegments=20,DebugThickness=0.05)
    DriverAxis=(X=1,Y=0,Z=0)
    ExecutePin=()
    FalloffRegionScaleFactors=(PositiveWidth=1,NegativeWidth=0.65,PositiveHeight=1,NegativeHeight=1)
    FalloffSize=0.5
    FlipHeightScaling=False
    FlipWidthScaling=False
    RotationOffset=(X=0,Y=-148.761,Z=0)
    DriverItem=RigVMVariableNode() {variable=Driver Item}.Value
    OptionalParentItem=RigVMVariableNode() {variable=Driver Item Parent}.Value
    Debug.bDrawDebug=RigVMVariableNode() {variable=Dn Viz}.Value
   )
   > n53 = RigVMVariableNode(Value=@n52.OutputParam) {variable=Result Dn}
   n54 = &RigVMFunction_ControlFlowBranch(ExecuteContext=@n53.ExecuteContext, Condition=RigVMVariableNode() {variable=Dn Viz}.Value)
   n55 = &RigUnit_SphericalPoseReader(
    ActiveRegionScaleFactors=(PositiveWidth=1,NegativeWidth=1,PositiveHeight=1,NegativeHeight=1)
    ActiveRegionSize=0.001
    Debug=(bDrawDebug=false,bDraw2D=False,bDrawLocalAxes=False,DebugScale=25,DebugSegments=20,DebugThickness=0.05)
    DriverAxis=(X=1,Y=0,Z=0)
    FalloffRegionScaleFactors=(PositiveWidth=1,NegativeWidth=0.298,PositiveHeight=2,NegativeHeight=2)
    FalloffSize=0.25
    FlipHeightScaling=False
    FlipWidthScaling=False
    RotationOffset=(X=0,Y=-103.761,Z=0)
    ExecutePin=@n54.Completed
    DriverItem=RigVMVariableNode() {variable=Driver Item}.Value
    OptionalParentItem=RigVMVariableNode() {variable=Driver Item Parent}.Value
    Debug.bDrawDebug=RigVMVariableNode() {variable=Dn Pre Viz}.Value
   )
   > n56 = RigVMVariableNode(Value=@n55.OutputParam) {variable=Result Dn Pre}
   n57 = &RigVMFunction_ControlFlowBranch(ExecuteContext=@n56.ExecuteContext, Condition=RigVMVariableNode() {variable=Dn Pre Viz}.Value)
   n58 = &RigUnit_SphericalPoseReader(
    ActiveRegionScaleFactors=(PositiveWidth=1,NegativeWidth=1,PositiveHeight=1,NegativeHeight=1)
    ActiveRegionSize=0.001
    Debug=(bDrawDebug=false,bDraw2D=False,bDrawLocalAxes=False,DebugScale=25,DebugSegments=20,DebugThickness=0.05)
    DriverAxis=(X=1,Y=0,Z=0)
    FalloffRegionScaleFactors=(PositiveWidth=1,NegativeWidth=1,PositiveHeight=2,NegativeHeight=2)
    FalloffSize=0.25
    FlipHeightScaling=False
    FlipWidthScaling=False
    RotationOffset=(X=0,Y=-193.761,Z=0)
    ExecutePin=@n57.Completed
    DriverItem=RigVMVariableNode() {variable=Driver Item}.Value
    OptionalParentItem=RigVMVariableNode() {variable=Driver Item Parent}.Value
    Debug.bDrawDebug=RigVMVariableNode() {variable=Dn Post Viz}.Value
   )
   > n59 = RigVMVariableNode(Value=@n58.OutputParam) {variable=Result Dn Post}
   n60 = &RigVMFunction_ControlFlowBranch(ExecuteContext=@n59.ExecuteContext, Condition=RigVMVariableNode() {variable=Dn Post Viz}.Value)
   RigVMDispatchNode(
    Enabled=True
    ScreenColor=(R=1,G=1,B=1,A=1)
    ScreenDuration=0.05
    Prefix=&RigVMFunction_StringConcat(B=" - DN POST:   ", A=RigVMVariableNode() {variable=Driver Item Name}.Value) {template="Concat::Execute(in A,in B,out Result)"}.Result
    ExecuteContext=@n60.True
    Value=RigVMVariableNode() {variable=Result Dn Post}.Value
   ) {template="DISPATCH_RigVMDispatch_Print(in Prefix,in Value,in Enabled,in ScreenDuration,in ScreenColor)"}
   RigVMDispatchNode(
    Enabled=True
    ScreenColor=(R=1,G=1,B=1,A=1)
    ScreenDuration=0.05
    Prefix=&RigVMFunction_StringConcat(B=" - DN PRE:   ", A=RigVMVariableNode() {variable=Driver Item Name}.Value) {template="Concat::Execute(in A,in B,out Result)"}.Result
    ExecuteContext=@n57.True
    Value=RigVMVariableNode() {variable=Result Dn Pre}.Value
   ) {template="DISPATCH_RigVMDispatch_Print(in Prefix,in Value,in Enabled,in ScreenDuration,in ScreenColor)"}
   RigVMDispatchNode(
    Enabled=True
    ScreenColor=(R=1,G=1,B=1,A=1)
    ScreenDuration=0.05
    Prefix=&RigVMFunction_StringConcat(B=" - DN:   ", A=RigVMVariableNode() {variable=Driver Item Name}.Value) {template="Concat::Execute(in A,in B,out Result)"}.Result
    ExecuteContext=@n54.True
    Value=RigVMVariableNode() {variable=Result Dn}.Value
   ) {template="DISPATCH_RigVMDispatch_Print(in Prefix,in Value,in Enabled,in ScreenDuration,in ScreenColor)"}
   RigVMRerouteNode(Value=@RigVMFunction_Sequence.D)
   RigVMRerouteNode(Value=@RigVMFunction_Sequence.B)
  ]}
 RigVMFunctionLibrary {nodes [
   Create Spline Nulls = RigVMCollapseNode(ExecuteContext=(), In Transforms=())
   Drive Chain By Spline = RigVMCollapseNode(
    Chain Percents=()
    Driven Chain=()
    End Twist=(X=0,Y=0,Z=0)
    ExecuteContext=()
    Local Transforms=()
    Start Twist=(X=0,Y=0,Z=0)
    Tick Spline=()
   )
   Parent Constraint Splines = RigVMCollapseNode(ExecuteContext=(), In Transforms=(), In Weights=())
   Spline Initial Percents And Offsets = RigVMCollapseNode(Driven Bones=(), ExecuteContext=(), In Spline=())
   UpperArm Readers = RigVMCollapseNode(Driver Item=(Type=None,Name="None"), ExecuteContext=())
  ]}
 RigVMFunction_Sequence_1_ContainedGraph {nodes [
   n1 = RigVMFunctionEntryNode()
   RigVMFunction_Sequence = &RigVMFunction_Sequence(ExecuteContext=@n1.ExecuteContext)
   n2 = &RigVMFunction_Sequence(ExecuteContext=@RigVMFunction_Sequence.B)
   n3 = &RigVMFunction_Sequence(ExecuteContext=@n2.B)
   n4 = &RigVMFunction_Sequence(ExecuteContext=@n3.B)
   n5 = &RigVMFunction_Sequence(ExecuteContext=@n4.B)
   n6 = &RigVMFunction_Sequence(ExecuteContext=@n5.B)
   n7 = &RigVMFunction_Sequence(ExecuteContext=@n6.B)
   RigVMFunctionReturnNode(A=@RigVMFunction_Sequence.A, B=@n2.A, C=@n3.A, D=@n4.A, E=@n5.A, F=@n6.A, H=@n7.A, I=@n7.B)
  ]}
 RigVMFunction_Sequence_1_ContainedGraph {nodes [
   n1 = RigVMFunctionEntryNode()
   RigVMFunction_Sequence = &RigVMFunction_Sequence(ExecuteContext=@n1.ExecuteContext)
   n2 = &RigVMFunction_Sequence(ExecuteContext=@RigVMFunction_Sequence.B)
   n3 = &RigVMFunction_Sequence(ExecuteContext=@n2.B)
   n4 = &RigVMFunction_Sequence(ExecuteContext=@n3.B)
   n5 = &RigVMFunction_Sequence(ExecuteContext=@n4.B)
   RigVMFunctionReturnNode(A=@RigVMFunction_Sequence.A, B=@n2.A, C=@n3.A, D=@n4.A, E=@n5.A, F=@n5.B)
  ]}
 RigVMFunction_Sequence_1_ContainedGraph {nodes [
   n1 = RigVMFunctionEntryNode()
   RigVMFunction_Sequence = &RigVMFunction_Sequence(ExecuteContext=@n1.ExecuteContext)
   n2 = &RigVMFunction_Sequence(ExecuteContext=@RigVMFunction_Sequence.B)
   n3 = &RigVMFunction_Sequence(ExecuteContext=@n2.B)
   n4 = &RigVMFunction_Sequence(ExecuteContext=@n3.B)
   n5 = &RigVMFunction_Sequence(ExecuteContext=@n4.B)
   RigVMFunctionReturnNode(A=@RigVMFunction_Sequence.A, B=@n2.A, C=@n3.A, D=@n4.A, E=@n5.A, F=@n5.B)
  ]}
 RigVMModel {nodes [
   ItemArray = RigVMDispatchNode(Value=((Type=Bone,Name="dyn_hair_ba_r_02"),(Type=Bone,Name="dyn_hair_ba_l_02"))) {template="DISPATCH_RigVMDispatch_Constant(io Value)"}
   For_Each_1 = RigVMDispatchNode(ExecuteContext=(), Array=@ItemArray.Value) {template="DISPATCH_RigVMDispatch_ArrayIterator(in Array,out Element,out Index,out Count,out Ratio)"}
   ItemArray_1 = RigVMDispatchNode(Value=((Type=Bone,Name="aim_hair_ba_r_01"),(Type=Bone,Name="aim_hair_ba_l_01"))) {template="DISPATCH_RigVMDispatch_Constant(io Value)"}
   n1 = RigVMDispatchNode(Index=@For_Each_1.Index, Array=@ItemArray_1.Value) {template="DISPATCH_RigVMDispatch_ArrayGetAtIndex(in Array,in Index,out Element)"}
   RigUnit_GetTransform = &RigUnit_GetTransform(Item=(Type=Bone,Name="head"), bInitial=False)
   n2 = /Script/ControlRig.RigUnit_AimBoneMath(
    DebugSettings=(bEnabled=False,Scale=10,WorldOffset=(Rotation=(X=0,Y=0,Z=0,W=1),Translation=(X=0,Y=0,Z=0),Scale3D=(X=1,Y=1,Z=1)))
    Primary=(Weight=1,Axis=(X=1,Y=0,Z=0),Target=(X=1,Y=0,Z=0),Kind=Location,Space=(Type=Bone,Name="None"))
    Secondary=(Weight=1,Axis=(X=0,Y=0,Z=1),Target=(X=1,Y=0,Z=0),Kind=Direction,Space=(Type=Bone,Name="None"))
    Weight=1
    InputTransform=&RigUnit_GetTransform(bInitial=False, Item=@n1.Element).Transform
    Primary.Target=&RigUnit_GetTransform(bInitial=False, Item=@For_Each_1.Element).Transform.Translation
    Secondary.Target=/Script/RigVM.RigVMFunction_MathTransformRotateVector(Vector=(X=0,Y=-1,Z=0), Transform=@RigUnit_GetTransform.Transform) {is=struct RigVMUnitNode, template="Rotate Vector::Execute(in Transform,in Vector,out Result)"}.Result
   ) {is=struct RigVMUnitNode}
   > &RigUnit_SetTransform(Weight=1, bInitial=False, Value=@n2.Result, Item=@n1.Element)
   > RigVMFunction_DebugTransformMutableNoSpace = &RigVMFunction_DebugTransformMutableNoSpace(
    Color=(R=1,G=0,B=0,A=1)
    DebugDrawSettings=(DepthPriority=SDPG_Foreground,Lifetime=-1)
    Mode=Axes
    Scale=10
    Thickness=0
    WorldOffset=(Rotation=(X=0,Y=0,Z=0,W=1),Translation=(X=0,Y=0,Z=0),Scale3D=(X=1,Y=1,Z=1))
    bEnabled=True
    Transform=@n2.Result
   )
   &RigUnit_ModifyTransforms(
    ExecutePin=()
    ItemToModify="((Item=(Type=Null,Name=\"Hair Ba L 03\"),Transform=(Rotation=(X=0,Y=0,Z=0,W=-1),Translation=(X=6.0003,Y=0,Z=-2.4167),Scale3D=(X=1,Y=1,Z=1))),(Item=(Type=Null,Name=\"Hair Ba L 04\"),Transform=(Rotation=(X=0,Y=0,Z=0,W=1),Translation=(X=0,Y=0,Z=-6.3776),Scale3D=( ...(truncated)"
    Weight=1
    WeightMinimum=0
   )
   > &RigUnit_ModifyTransforms(
    ItemToModify="((Item=(Type=Null,Name=\"Hair Ba R 03\"),Transform=(Rotation=(X=0,Y=0,Z=0,W=-1),Translation=(X=6.0913,Y=0,Z=-2.9809),Scale3D=(X=1,Y=1,Z=1))),(Item=(Type=Null,Name=\"Hair Ba R 04\"),Transform=(Rotation=(X=0,Y=0,Z=0,W=1),Translation=(X=0,Y=0,Z=-7.5336),Scale3D=( ...(truncated)"
    Weight=1
    WeightMinimum=0
   )
   > &RigUnit_ModifyTransforms(
    ItemToModify="((Item=(Type=Null,Name=\"Hair Ba C 03\"),Transform=(Rotation=(X=0,Y=0,Z=0,W=1),Translation=(X=5.2588,Y=0,Z=-4.5743),Scale3D=(X=1,Y=1,Z=1))),(Item=(Type=Null,Name=\"Hair Ba C 04\"),Transform=(Rotation=(X=0,Y=0,Z=0,W=1),Translation=(X=0,Y=0,Z=-7.1801),Scale3D=(X ...(truncated)"
    Weight=1
    WeightMinimum=0
   )
   RigUnit_BeginExecution = &RigUnit_BeginExecution()
   RigVMFunction_Sequence_1 = RigVMAggregateNode(ExecuteContext=@RigUnit_BeginExecution.ExecutePin)
   n3 = &RigUnit_SphericalPoseReader(
    ActiveRegionScaleFactors=(PositiveWidth=1,NegativeWidth=1,PositiveHeight=1,NegativeHeight=1)
    ActiveRegionSize=0.001
    Debug=(bDrawDebug=false,bDraw2D=False,bDrawLocalAxes=False,DebugScale=25,DebugSegments=20,DebugThickness=0.25)
    DriverAxis=(X=1,Y=0,Z=0)
    DriverItem=(Type=Bone,Name="head")
    FalloffRegionScaleFactors=(PositiveWidth=1,NegativeWidth=1,PositiveHeight=1,NegativeHeight=1)
    FalloffSize=0.5
    FlipHeightScaling=False
    FlipWidthScaling=False
    OptionalParentItem=(Type=Bone,Name="None")
    RotationOffset=(X=-90,Y=0,Z=0)
    ExecutePin=@RigVMFunction_Sequence_1.E
   )
   > &RigUnit_ModifyTransforms(
    ItemToModify="((Item=(Type=Null,Name=\"Hair Ba L 01\"),Transform=(Rotation=(X=0,Y=0,Z=0,W=-1),Translation=(X=-2.8229,Y=0,Z=13.529),Scale3D=(X=1,Y=1,Z=1))),(Item=(Type=Null,Name=\"Hair Ba L 00\"),Transform=(Rotation=(X=0,Y=0,Z=0,W=-1),Translation=(X=-6.7895,Y=0,Z=6.576),Scal ...(truncated)"
    WeightMinimum=0
    Weight=@n3.OutputParam
   )
   > &RigUnit_ModifyTransforms(
    ItemToModify="((Item=(Type=Null,Name=\"Hair Ba R 01\"),Transform=(Rotation=(X=0,Y=0,Z=0,W=-1),Translation=(X=-5,Y=0,Z=15),Scale3D=(X=1,Y=1,Z=1))),(Item=(Type=Null,Name=\"Hair Ba R 00\"),Transform=(Rotation=(X=0,Y=0,Z=0,W=1),Translation=(X=-6.7646,Y=0,Z=-2),Scale3D=(X=1,Y=1, ...(truncated)"
    WeightMinimum=0
    Weight=@n3.OutputParam
   )
   > n4 = &RigUnit_SphericalPoseReader(
    ActiveRegionScaleFactors=(PositiveWidth=1,NegativeWidth=1,PositiveHeight=1,NegativeHeight=1)
    ActiveRegionSize=0.001
    Debug=(bDrawDebug=false,bDraw2D=False,bDrawLocalAxes=False,DebugScale=25,DebugSegments=20,DebugThickness=0.05)
    DriverAxis=(X=1,Y=0,Z=0)
    DriverItem=(Type=Bone,Name="neck_02")
    FalloffRegionScaleFactors=(PositiveWidth=1,NegativeWidth=1,PositiveHeight=1,NegativeHeight=1)
    FalloffSize=0.5
    FlipHeightScaling=False
    FlipWidthScaling=False
    OptionalParentItem=(Type=Bone,Name="None")
    RotationOffset=(X=-90,Y=0,Z=0)
   )
   > &RigUnit_ModifyTransforms(
    ItemToModify="((Item=(Type=Null,Name=\"Hair Ba L 01\"),Transform=(Rotation=(X=0,Y=0,Z=0,W=1),Translation=(X=0.1451,Y=0,Z=13.9155),Scale3D=(X=1,Y=1,Z=1))),(Item=(Type=Null,Name=\"Hair Ba L 00\"),Transform=(Rotation=(X=0,Y=0,Z=0,W=1),Translation=(X=-6.741,Y=0,Z=6.3228),Scale3 ...(truncated)"
    WeightMinimum=0
    Weight=@n4.OutputParam
   )
   > &RigUnit_ModifyTransforms(
    ItemToModify="((Item=(Type=Null,Name=\"Hair Ba R 01\"),Transform=(Rotation=(X=0,Y=0,Z=0,W=-1),Translation=(X=0.4167,Y=0,Z=13.6816),Scale3D=(X=1,Y=1,Z=1))),(Item=(Type=Null,Name=\"Hair Ba R 00\"),Transform=(Rotation=(X=0,Y=0,Z=0,W=1),Translation=(X=-6.741,Y=0,Z=6.3228),Scale ...(truncated)"
    WeightMinimum=0
    Weight=@n4.OutputParam
   )
   > n5 = &RigUnit_SphericalPoseReader(
    ActiveRegionScaleFactors=(PositiveWidth=1,NegativeWidth=1,PositiveHeight=1,NegativeHeight=1)
    ActiveRegionSize=0.001
    Debug=(bDrawDebug=false,bDraw2D=False,bDrawLocalAxes=False,DebugScale=25,DebugSegments=20,DebugThickness=0.05)
    DriverAxis=(X=1,Y=0,Z=0)
    DriverItem=(Type=Bone,Name="neck_01")
    FalloffRegionScaleFactors=(PositiveWidth=1,NegativeWidth=1,PositiveHeight=1,NegativeHeight=1)
    FalloffSize=0.5
    FlipHeightScaling=False
    FlipWidthScaling=False
    OptionalParentItem=(Type=Bone,Name="None")
    RotationOffset=(X=-90,Y=0,Z=0)
   )
   > &RigUnit_ModifyTransforms(
    ItemToModify="((Item=(Type=Null,Name=\"Hair Ba L 01\"),Transform=(Rotation=(X=0,Y=0,Z=0,W=1),Translation=(X=10,Y=0,Z=13),Scale3D=(X=1,Y=1,Z=1))),(Item=(Type=Null,Name=\"Hair Ba L 00\"),Transform=(Rotation=(X=0,Y=0,Z=0,W=1),Translation=(X=0,Y=0,Z=-6),Scale3D=(X=1,Y=1,Z=1))), ...(truncated)"
    WeightMinimum=0
    Weight=@n5.OutputParam
   )
   > &RigUnit_ModifyTransforms(
    ItemToModify="((Item=(Type=Null,Name=\"Hair Ba R 01\"),Transform=(Rotation=(X=0,Y=0,Z=0,W=-1),Translation=(X=3,Y=0,Z=5),Scale3D=(X=1,Y=1,Z=1))),(Item=(Type=Null,Name=\"Hair Ba R 00\"),Transform=(Rotation=(X=0,Y=0,Z=0,W=1),Translation=(X=-15,Y=0,Z=6),Scale3D=(X=1,Y=1,Z=1))), ...(truncated)"
    WeightMinimum=0
    Weight=@n5.OutputParam
   )
   > n6 = &RigUnit_SphericalPoseReader(
    ActiveRegionScaleFactors=(PositiveWidth=1,NegativeWidth=1,PositiveHeight=1,NegativeHeight=1)
    ActiveRegionSize=0.001
    Debug=(bDrawDebug=false,bDraw2D=False,bDrawLocalAxes=False,DebugScale=25,DebugSegments=20,DebugThickness=0.05)
    DriverAxis=(X=0,Y=1,Z=0)
    DriverItem=(Type=Bone,Name="head")
    FalloffRegionScaleFactors=(PositiveWidth=1,NegativeWidth=1,PositiveHeight=1,NegativeHeight=1)
    FalloffSize=0.5
    FlipHeightScaling=False
    FlipWidthScaling=False
    OptionalParentItem=(Type=Bone,Name="spine_05")
    RotationOffset=(X=0,Y=0,Z=0)
   )
   > &RigUnit_ModifyTransforms(
    ItemToModify="((Item=(Type=Null,Name=\"Hair Ba L 01\"),Transform=(Rotation=(X=0,Y=0,Z=0,W=1),Translation=(X=3.5349,Y=-3.371,Z=4.3979),Scale3D=(X=1,Y=1,Z=1))),(Item=(Type=Null,Name=\"Hair Ba L 02\"),Transform=(Rotation=(X=0,Y=0,Z=0,W=1),Translation=(X=0.4556,Y=0,Z=2.0089),Sc ...(truncated)"
    WeightMinimum=0
    Weight=@n6.OutputParam
   )
   > &RigUnit_ModifyTransforms(
    ItemToModify="((Item=(Type=Null,Name=\"Hair Fr R 01\"),Transform=(Rotation=(X=0,Y=0,Z=0,W=-1),Translation=(X=-3.5615,Y=3.154,Z=-1.7559),Scale3D=(X=1,Y=1,Z=1))),(Item=(Type=Null,Name=\"Hair Fr L 01\"),Transform=(Rotation=(X=0,Y=0,Z=0,W=-1),Translation=(X=0,Y=4.1884,Z=0),Scal ...(truncated)"
    WeightMinimum=0
    Weight=@n6.OutputParam
   )
   > n7 = &RigUnit_SphericalPoseReader(
    ActiveRegionScaleFactors=(PositiveWidth=1,NegativeWidth=1,PositiveHeight=1,NegativeHeight=1)
    ActiveRegionSize=0.001
    Debug=(bDrawDebug=false,bDraw2D=False,bDrawLocalAxes=False,DebugScale=25,DebugSegments=20,DebugThickness=0.05)
    DriverAxis=(X=0,Y=1,Z=0)
    DriverItem=(Type=Bone,Name="head")
    FalloffRegionScaleFactors=(PositiveWidth=1,NegativeWidth=1,PositiveHeight=1,NegativeHeight=1)
    FalloffSize=0.5
    FlipHeightScaling=False
    FlipWidthScaling=False
    OptionalParentItem=(Type=Bone,Name="spine_05")
    RotationOffset=(X=180,Y=0,Z=0)
   )
   > &RigUnit_ModifyTransforms(
    ItemToModify="((Item=(Type=Null,Name=\"Hair Ba R 01\"),Transform=(Rotation=(X=0,Y=0,Z=0,W=1),Translation=(X=7.1095,Y=4.0792,Z=0),Scale3D=(X=1,Y=1,Z=1))),(Item=(Type=Null,Name=\"Hair Ba R 02\"),Transform=(Rotation=(X=0,Y=0,Z=0,W=1),Translation=(X=-0.4093,Y=0,Z=1.7756),Scale3 ...(truncated)"
    WeightMinimum=0
    Weight=@n7.OutputParam
   )
   > &RigUnit_ModifyTransforms(
    ItemToModify="((Item=(Type=Null,Name=\"Hair Ba L 02\"),Transform=(Rotation=(X=0,Y=0,Z=0,W=1),Translation=(X=0.6915,Y=-1.9628,Z=0.282),Scale3D=(X=1,Y=1,Z=1))),(Item=(Type=Null,Name=\"Hair Fr L 01\"),Transform=(Rotation=(X=0,Y=0,Z=0,W=-1),Translation=(X=-2.3896,Y=-4.048,Z=0.8 ...(truncated)"
    WeightMinimum=0
    Weight=@n7.OutputParam
   )
   > n8 = &RigUnit_SphericalPoseReader(
    ActiveRegionScaleFactors=(PositiveWidth=1,NegativeWidth=1,PositiveHeight=1,NegativeHeight=1)
    ActiveRegionSize=0.001
    Debug=(bDrawDebug=false,bDraw2D=False,bDrawLocalAxes=False,DebugScale=25,DebugSegments=20,DebugThickness=0.05)
    DriverAxis=(X=1,Y=0,Z=0)
    DriverItem=(Type=Bone,Name="head")
    FalloffRegionScaleFactors=(PositiveWidth=1,NegativeWidth=1,PositiveHeight=1,NegativeHeight=1)
    FalloffSize=0.5
    FlipHeightScaling=False
    FlipWidthScaling=False
    OptionalParentItem=(Type=Bone,Name="spine_05")
    RotationOffset=(X=0,Y=0,Z=0)
   )
   > &RigUnit_ModifyTransforms(
    ItemToModify="((Item=(Type=Null,Name=\"Hair Ba R 01\"),Transform=(Rotation=(X=0,Y=0,Z=0,W=1),Translation=(X=2.7598,Y=-7.2863,Z=7.8115),Scale3D=(X=1,Y=1,Z=1))),(Item=(Type=Null,Name=\"Hair Fr R 00\"),Transform=(Rotation=(X=0,Y=0,Z=0,W=1),Translation=(X=-4.434,Y=0,Z=3.4087),S ...(truncated)"
    WeightMinimum=0
    Weight=@n8.OutputParam
   )
   Drive Chain By Spline = RigVMFunctionReferenceNode(
    Display Transforms=false
    End Twist=(X=0,Y=0,Z=0)
    Start Twist=(X=0,Y=0,Z=0)
    Tick Spline=RigVMVariableNode() {variable=Fr L Tick Spline}.Value
    Driven Chain=RigVMVariableNode() {variable=Fr L Bones}.Value
    Chain Percents=RigVMVariableNode() {variable=Fr L Orig Percents}.Value
    Original Length=RigVMVariableNode() {variable=Fr L Orig Length}.Value
    Local Transforms=RigVMVariableNode() {variable=Fr L Local Offsets}.Value
    ExecuteContext=@RigVMFunction_Sequence_1.I
   )
   > Drive Chain By Spline_1 = RigVMFunctionReferenceNode(
    Display Transforms=false
    End Twist=(X=0,Y=0,Z=0)
    Start Twist=(X=0,Y=0,Z=0)
    Chain Percents=RigVMVariableNode() {variable=Fr R Orig Percents}.Value
    Local Transforms=RigVMVariableNode() {variable=Fr R Local Offsets}.Value
    Original Length=RigVMVariableNode() {variable=Fr R Orig Length}.Value
    Tick Spline=RigVMVariableNode() {variable=Fr R Tick Spline}.Value
    Driven Chain=RigVMVariableNode() {variable=Fr R Bones}.Value
   )
   > Drive Chain By Spline_2 = RigVMFunctionReferenceNode(
    Display Transforms=false
    End Twist=(X=0,Y=0,Z=0)
    Start Twist=(X=0,Y=0,Z=0)
    Tick Spline=RigVMVariableNode() {variable="Ba L Tick Spline "}.Value
    Driven Chain=RigVMVariableNode() {variable=Ba L Bones}.Value
    Chain Percents=RigVMVariableNode() {variable=Ba L Orig Percents}.Value
    Local Transforms=RigVMVariableNode() {variable=Ba L Local Offsets}.Value
    Original Length=RigVMVariableNode() {variable=Ba L Orig Length}.Value
   )
   > Drive Chain By Spline_3 = RigVMFunctionReferenceNode(
    Display Transforms=false
    End Twist=(X=0,Y=0,Z=0)
    Start Twist=(X=0,Y=0,Z=0)
    Tick Spline=RigVMVariableNode() {variable=Ba C Tick Spline}.Value
    Driven Chain=RigVMVariableNode() {variable=Ba C Bones}.Value
    Chain Percents=RigVMVariableNode() {variable=Ba C Orig Percents}.Value
    Local Transforms=RigVMVariableNode() {variable=Ba C Local Offsets}.Value
    Original Length=RigVMVariableNode() {variable=Ba C Orig Length}.Value
   )
   > Drive Chain By Spline_4 = RigVMFunctionReferenceNode(
    Display Transforms=false
    End Twist=(X=0,Y=0,Z=0)
    Start Twist=(X=0,Y=0,Z=0)
    Tick Spline=RigVMVariableNode() {variable=Ba R Tick Spline}.Value
    Driven Chain=RigVMVariableNode() {variable=Ba R Bones}.Value
    Chain Percents=RigVMVariableNode() {variable=Ba R Orig Percents}.Value
    Local Transforms=RigVMVariableNode() {variable=Ba R Local Offsets}.Value
    Original Length=RigVMVariableNode() {variable=Ba R Orig Length}.Value
   )
   n9 = &RigUnit_SphericalPoseReader(
    ActiveRegionScaleFactors=(PositiveWidth=1,NegativeWidth=1,PositiveHeight=1,NegativeHeight=1)
    ActiveRegionSize=0.001
    Debug=(bDrawDebug=false,bDraw2D=False,bDrawLocalAxes=False,DebugScale=25,DebugSegments=20,DebugThickness=0.025)
    DriverAxis=(X=1,Y=0,Z=0)
    DriverItem=(Type=Bone,Name="clavicle_l")
    FalloffRegionScaleFactors=(PositiveWidth=1,NegativeWidth=1,PositiveHeight=1,NegativeHeight=1)
    FalloffSize=0.5
    FlipHeightScaling=False
    FlipWidthScaling=False
    OptionalParentItem=(Type=Bone,Name="None")
    RotationOffset=(X=-90,Y=0,Z=0)
    ExecutePin=@RigVMFunction_Sequence_1.F
   )
   > &RigUnit_ModifyTransforms(
    ItemToModify="((Item=(Type=Null,Name=\"Hair Ba L 01\"),Transform=(Rotation=(X=0,Y=0,Z=0,W=1),Translation=(X=1.7506,Y=0,Z=2.6929),Scale3D=(X=1,Y=1,Z=1))),(Item=(Type=Null,Name=\"Hair Ba L 02\"),Transform=(Rotation=(X=0,Y=0,Z=0,W=1),Translation=(X=0,Y=-5.8153,Z=11.0729),Scale ...(truncated)"
    WeightMinimum=0
    Weight=@n9.OutputParam
   )
   > n10 = &RigUnit_SphericalPoseReader(
    ActiveRegionScaleFactors=(PositiveWidth=1,NegativeWidth=1,PositiveHeight=1,NegativeHeight=1)
    ActiveRegionSize=0.001
    Debug=(bDrawDebug=false,bDraw2D=False,bDrawLocalAxes=False,DebugScale=25,DebugSegments=20,DebugThickness=0.025)
    DriverAxis=(X=1,Y=0,Z=0)
    DriverItem=(Type=Bone,Name="clavicle_r")
    FalloffRegionScaleFactors=(PositiveWidth=1,NegativeWidth=1,PositiveHeight=1,NegativeHeight=1)
    FalloffSize=0.5
    FlipHeightScaling=False
    FlipWidthScaling=False
    OptionalParentItem=(Type=Bone,Name="None")
    RotationOffset=(X=-90,Y=0,Z=0)
   )
   > &RigUnit_ModifyTransforms(
    ItemToModify="((Item=(Type=Null,Name=\"Hair Ba R 01\"),Transform=(Rotation=(X=0,Y=0,Z=0,W=1),Translation=(X=3.2904,Y=2.6751,Z=0.781),Scale3D=(X=1,Y=1,Z=1))),(Item=(Type=Null,Name=\"Hair Ba R 02\"),Transform=(Rotation=(X=0,Y=0,Z=0,W=-1),Translation=(X=0,Y=6.9041,Z=8.0661),Sc ...(truncated)"
    WeightMinimum=0
    Weight=@n10.OutputParam
   )
   n11 = &RigUnit_SphericalPoseReader(
    ActiveRegionScaleFactors=(PositiveWidth=1,NegativeWidth=1,PositiveHeight=1,NegativeHeight=1)
    ActiveRegionSize=0.001
    Debug=(bDrawDebug=false,bDraw2D=False,bDrawLocalAxes=False,DebugScale=25,DebugSegments=20,DebugThickness=0.05)
    DriverAxis=(X=1,Y=0,Z=0)
    DriverItem=(Type=Bone,Name="spine_01")
    FalloffRegionScaleFactors=(PositiveWidth=1,NegativeWidth=1,PositiveHeight=1,NegativeHeight=1)
    FalloffSize=0.5
    FlipHeightScaling=False
    FlipWidthScaling=False
    OptionalParentItem=(Type=Bone,Name="None")
    RotationOffset=(X=-90,Y=0,Z=0)
    ExecutePin=@RigVMFunction_Sequence_1.D
   )
   > &RigUnit_ModifyTransforms(
    ItemToModify=((Item=(Type=Null,Name="Hair Ba L 03"),Transform=(Rotation=(X=0,Y=0,Z=0,W=1),Translation=(X=9.5818,Y=0,Z=8.6379),Scale3D=(X=1,Y=1,Z=1))))
    WeightMinimum=0
    Weight=@n11.OutputParam
   )
   > &RigUnit_ModifyTransforms(
    ItemToModify=((Item=(Type=Null,Name="Hair Ba C 03"),Transform=(Rotation=(X=0,Y=0,Z=0,W=1),Translation=(X=8.7709,Y=0,Z=9.2304),Scale3D=(X=1,Y=1,Z=1))))
    WeightMinimum=0
    Weight=@n11.OutputParam
   )
   > &RigUnit_ModifyTransforms(
    ItemToModify=((Item=(Type=Null,Name="Hair Ba R 03"),Transform=(Rotation=(X=0,Y=0,Z=0,W=1),Translation=(X=8.2609,Y=0,Z=9.1295),Scale3D=(X=1,Y=1,Z=1))))
    WeightMinimum=0
    Weight=@n11.OutputParam
   )
   > n12 = &RigUnit_SphericalPoseReader(
    ActiveRegionScaleFactors=(PositiveWidth=1,NegativeWidth=1,PositiveHeight=1,NegativeHeight=1)
    ActiveRegionSize=0.001
    Debug=(bDrawDebug=false,bDraw2D=False,bDrawLocalAxes=False,DebugScale=25,DebugSegments=20,DebugThickness=0.05)
    DriverAxis=(X=1,Y=0,Z=0)
    DriverItem=(Type=Bone,Name="spine_02")
    FalloffRegionScaleFactors=(PositiveWidth=1,NegativeWidth=1,PositiveHeight=1,NegativeHeight=1)
    FalloffSize=0.5
    FlipHeightScaling=False
    FlipWidthScaling=False
    OptionalParentItem=(Type=Bone,Name="None")
    RotationOffset=(X=-90,Y=0,Z=0)
   )
   > &RigUnit_ModifyTransforms(
    ItemToModify="((Item=(Type=Null,Name=\"Hair Ba L 03\"),Transform=(Rotation=(X=0,Y=0,Z=0,W=-1),Translation=(X=6.5175,Y=0,Z=5.4976),Scale3D=(X=1,Y=1,Z=1))),(Item=(Type=Null,Name=\"Hair Ba L 02\"),Transform=(Rotation=(X=0,Y=0,Z=0,W=-1),Translation=(X=4.349,Y=0,Z=4.1807),Scale3 ...(truncated)"
    WeightMinimum=0
    Weight=@n12.OutputParam
   )
   > &RigUnit_ModifyTransforms(
    ItemToModify="((Item=(Type=Null,Name=\"Hair Ba C 03\"),Transform=(Rotation=(X=0,Y=0,Z=0,W=-1),Translation=(X=6.5175,Y=0,Z=5.4976),Scale3D=(X=1,Y=1,Z=1))),(Item=(Type=Null,Name=\"Hair Ba C 02\"),Transform=(Rotation=(X=0,Y=0,Z=0,W=-1),Translation=(X=2.0215,Y=0,Z=1.5888),Scale ...(truncated)"
    WeightMinimum=0
    Weight=@n12.OutputParam
   )
   > &RigUnit_ModifyTransforms(
    ItemToModify="((Item=(Type=Null,Name=\"Hair Ba R 03\"),Transform=(Rotation=(X=0,Y=0,Z=0,W=-1),Translation=(X=6.5175,Y=0,Z=5.4976),Scale3D=(X=1,Y=1,Z=1))),(Item=(Type=Null,Name=\"Hair Ba R 02\"),Transform=(Rotation=(X=0,Y=0,Z=0,W=1),Translation=(X=0.4776,Y=0,Z=2.3406),Scale3 ...(truncated)"
    WeightMinimum=0
    Weight=@n12.OutputParam
   )
   > n13 = &RigUnit_SphericalPoseReader(
    ActiveRegionScaleFactors=(PositiveWidth=1,NegativeWidth=1,PositiveHeight=1,NegativeHeight=1)
    ActiveRegionSize=0.001
    Debug=(bDrawDebug=false,bDraw2D=False,bDrawLocalAxes=False,DebugScale=25,DebugSegments=20,DebugThickness=0.05)
    DriverAxis=(X=1,Y=0,Z=0)
    DriverItem=(Type=Bone,Name="spine_03")
    FalloffRegionScaleFactors=(PositiveWidth=1,NegativeWidth=1,PositiveHeight=1,NegativeHeight=1)
    FalloffSize=0.5
    FlipHeightScaling=False
    FlipWidthScaling=False
    OptionalParentItem=(Type=Bone,Name="None")
    RotationOffset=(X=-90,Y=0,Z=0)
   )
   > &RigUnit_ModifyTransforms(
    ItemToModify="((Item=(Type=Null,Name=\"Hair Ba L 03\"),Transform=(Rotation=(X=0,Y=0,Z=0,W=-1),Translation=(X=6.5175,Y=0,Z=5.4976),Scale3D=(X=1,Y=1,Z=1))),(Item=(Type=Null,Name=\"Hair Ba L 02\"),Transform=(Rotation=(X=0,Y=0,Z=0,W=-1),Translation=(X=4.349,Y=0,Z=4.1807),Scale3 ...(truncated)"
    WeightMinimum=0
    Weight=@n13.OutputParam
   )
   > &RigUnit_ModifyTransforms(
    ItemToModify="((Item=(Type=Null,Name=\"Hair Ba C 03\"),Transform=(Rotation=(X=0,Y=0,Z=0,W=-1),Translation=(X=6.5175,Y=0,Z=5.4976),Scale3D=(X=1,Y=1,Z=1))),(Item=(Type=Null,Name=\"Hair Ba C 02\"),Transform=(Rotation=(X=0,Y=0,Z=0,W=-1),Translation=(X=2.0215,Y=0,Z=1.5888),Scale ...(truncated)"
    WeightMinimum=0
    Weight=@n13.OutputParam
   )
   > &RigUnit_ModifyTransforms(
    ItemToModify="((Item=(Type=Null,Name=\"Hair Ba R 03\"),Transform=(Rotation=(X=0,Y=0,Z=0,W=-1),Translation=(X=6.5175,Y=0,Z=5.4976),Scale3D=(X=1,Y=1,Z=1))),(Item=(Type=Null,Name=\"Hair Ba R 02\"),Transform=(Rotation=(X=0,Y=0,Z=0,W=1),Translation=(X=0.4776,Y=0,Z=2.3406),Scale3 ...(truncated)"
    WeightMinimum=0
    Weight=@n13.OutputParam
   )
   > n14 = &RigUnit_SphericalPoseReader(
    ActiveRegionScaleFactors=(PositiveWidth=1,NegativeWidth=1,PositiveHeight=1,NegativeHeight=1)
    ActiveRegionSize=0.001
    Debug=(bDrawDebug=false,bDraw2D=False,bDrawLocalAxes=False,DebugScale=25,DebugSegments=20,DebugThickness=0.05)
    DriverAxis=(X=1,Y=0,Z=0)
    DriverItem=(Type=Bone,Name="spine_04")
    FalloffRegionScaleFactors=(PositiveWidth=1,NegativeWidth=1,PositiveHeight=1,NegativeHeight=1)
    FalloffSize=0.5
    FlipHeightScaling=False
    FlipWidthScaling=False
    OptionalParentItem=(Type=Bone,Name="None")
    RotationOffset=(X=-90,Y=0,Z=0)
   )
   > &RigUnit_ModifyTransforms(
    ItemToModify="((Item=(Type=Null,Name=\"Hair Ba L 01\"),Transform=(Rotation=(X=0,Y=0,Z=0,W=-1),Translation=(X=-0.0934,Y=0,Z=4.0833),Scale3D=(X=1,Y=1,Z=1))),(Item=(Type=Null,Name=\"Hair Ba L 02\"),Transform=(Rotation=(X=0,Y=0,Z=0,W=1),Translation=(X=19.5322,Y=0,Z=20.3721),Sca ...(truncated)"
    WeightMinimum=0
    Weight=@n14.OutputParam
   )
   > &RigUnit_ModifyTransforms(
    ItemToModify="((Item=(Type=Null,Name=\"Hair Ba R 01\"),Transform=(Rotation=(X=0,Y=0,Z=0,W=-1),Translation=(X=-0.0934,Y=0,Z=4.0833),Scale3D=(X=1,Y=1,Z=1))),(Item=(Type=Null,Name=\"Hair Ba R 02\"),Transform=(Rotation=(X=0,Y=0,Z=0,W=1),Translation=(X=19.5322,Y=0,Z=20.3721),Sca ...(truncated)"
    WeightMinimum=0
    Weight=@n14.OutputParam
   )
   > &RigUnit_ModifyTransforms(
    ItemToModify="((Item=(Type=Null,Name=\"Hair Ba C 02\"),Transform=(Rotation=(X=0,Y=0,Z=0,W=1),Translation=(X=3.4345,Y=0,Z=17.604),Scale3D=(X=1,Y=1,Z=1))),(Item=(Type=Null,Name=\"Hair Ba C 03\"),Transform=(Rotation=(X=0,Y=0,Z=0,W=-1),Translation=(X=6.1375,Y=0,Z=7.7302),Scale3 ...(truncated)"
    WeightMinimum=0
    Weight=@n14.OutputParam
   )
   n15 = RigVMVariableNode() {variable=Fr L Nulls}
   RigVMVariableNode(
    Value=/Script/ControlRigSpline.RigUnit_ControlRigSplineFromTransforms(
     Compression=0
     SamplesPerSegment=8
     SplineMode=Hermite
     Stretch=0
     bClosed=False
     Transforms=/Script/ControlRig.RigUnit_GetTransformItemArray(Space=GlobalSpace, bInitial=False, Items=@n15.Value) {is=struct RigVMUnitNode}.Transforms
    ) {is=struct RigVMUnitNode}.Spline
    ExecuteContext=@RigVMFunction_Sequence_1.H
   ) {variable=Fr L Tick Spline}
   n16 = RigVMVariableNode() {variable=Fr R Nulls}
   > RigVMVariableNode(Value=&RigUnit_ControlRigSplineFromTransforms(
     Compression=0
     SamplesPerSegment=8
     SplineMode=Hermite
     Stretch=0
     bClosed=False
     Transforms=&RigUnit_GetTransformItemArray(Space=GlobalSpace, bInitial=False, Items=@n16.Value).Transforms
    ).Spline) {variable=Fr R Tick Spline}
   n17 = RigVMVariableNode() {variable=Ba L Nulls}
   > RigVMVariableNode(Value=&RigUnit_ControlRigSplineFromTransforms(
     Compression=0
     SamplesPerSegment=8
     SplineMode=Hermite
     Stretch=0
     bClosed=False
     Transforms=&RigUnit_GetTransformItemArray(Space=GlobalSpace, bInitial=False, Items=@n17.Value).Transforms
    ).Spline) {variable="Ba L Tick Spline "}
   n18 = RigVMVariableNode() {variable=Ba C Nulls}
   > RigVMVariableNode(Value=&RigUnit_ControlRigSplineFromTransforms(
     Compression=0
     SamplesPerSegment=8
     SplineMode=Hermite
     Stretch=0
     bClosed=False
     Transforms=&RigUnit_GetTransformItemArray(Space=GlobalSpace, bInitial=False, Items=@n18.Value).Transforms
    ).Spline) {variable=Ba C Tick Spline}
   n19 = RigVMVariableNode() {variable=Ba R Nulls}
   > n20 = RigVMVariableNode(Value=&RigUnit_ControlRigSplineFromTransforms(
     Compression=0
     SamplesPerSegment=8
     SplineMode=Hermite
     Stretch=0
     bClosed=False
     Transforms=&RigUnit_GetTransformItemArray(Space=GlobalSpace, bInitial=False, Items=@n19.Value).Transforms
    ).Spline) {variable=Ba R Tick Spline}
   n21 = &RigVMFunction_ControlFlowBranch(ExecuteContext=@n20.ExecuteContext, Condition=RigVMVariableNode() {variable=display splines}.Value)
   Make_Array = RigVMDispatchNode(
    Values=((),(()),(()),(()),(()))
    Values.0=RigVMVariableNode() {variable=Fr L Tick Spline}.Value
    Values.1=RigVMVariableNode() {variable=Fr R Tick Spline}.Value
    Values.2=RigVMVariableNode() {variable="Ba L Tick Spline "}.Value
    Values.3=RigVMVariableNode() {variable=Ba C Tick Spline}.Value
    Values.4=RigVMVariableNode() {variable=Ba R Tick Spline}.Value
   ) {template="DISPATCH_RigVMDispatch_ArrayMake(in Values,out Array)"}
   For_Each = RigVMDispatchNode(Array=@Make_Array.Array, ExecuteContext=@n21.True) {template="DISPATCH_RigVMDispatch_ArrayIterator(in Array,out Element,out Index,out Count,out Ratio)"}
   > /Script/ControlRigSpline.RigUnit_DrawControlRigSpline(Color=(R=1,G=0,B=0,A=1), Detail=16, Thickness=0.15, Spline=@For_Each.Element) {is=struct RigVMUnitNode}
   Parent Constraint Splines = RigVMFunctionReferenceNode(
    In Transforms=RigVMVariableNode() {variable=Fr L Transforms}.Value
    In Weights=RigVMVariableNode() {variable=L Front Constraint Wgts}.Value
    Display Transforms=RigVMVariableNode() {variable=display spline transforms}.Value
    ExecuteContext=@RigVMFunction_Sequence_1.B
   )
   > /Script/ControlRig.RigUnit_SetTransformItemArray(
    Space=GlobalSpace
    Weight=1
    bInitial=False
    bPropagateToChildren=true
    Items=RigVMVariableNode() {variable=Fr L Nulls}.Value
    Transforms=@"Parent Constraint Splines".Out Transforms
   ) {is=struct RigVMUnitNode}
   > Parent Constraint Splines_1 = RigVMFunctionReferenceNode(
    In Weights=RigVMVariableNode() {variable=R Front Constraint Wgts}.Value
    In Transforms=RigVMVariableNode() {variable=Fr R Transforms}.Value
    Display Transforms=RigVMVariableNode() {variable=display spline transforms}.Value
   )
   > &RigUnit_SetTransformItemArray(
    Space=GlobalSpace
    Weight=1
    bInitial=False
    bPropagateToChildren=false
    Items=RigVMVariableNode() {variable=Fr R Nulls}.Value
    Transforms=@"Parent Constraint Splines_1".Out Transforms
   )
   > Parent Constraint Splines_3 = RigVMFunctionReferenceNode(
    In Transforms=RigVMVariableNode() {variable=Ba C Transforms}.Value
    In Weights=RigVMVariableNode() {variable=C Back Constraint Wgts}.Value
    Display Transforms=RigVMVariableNode() {variable=display spline transforms}.Value
   )
   > &RigUnit_SetTransformItemArray(
    Space=GlobalSpace
    Weight=1
    bInitial=False
    bPropagateToChildren=True
    Items=RigVMVariableNode() {variable=Ba C Nulls}.Value
    Transforms=@"Parent Constraint Splines_3".Out Transforms
   )
   > Parent Constraint Splines_2 = RigVMFunctionReferenceNode(
    In Weights=RigVMVariableNode() {variable=R Back Constraint Wgts}.Value
    Display Transforms=RigVMVariableNode() {variable=display spline transforms}.Value
    In Transforms=RigVMVariableNode() {variable=Ba R Transforms}.Value
   )
   > &RigUnit_SetTransformItemArray(
    Space=GlobalSpace
    Weight=1
    bInitial=False
    bPropagateToChildren=True
    Items=RigVMVariableNode() {variable=Ba R Nulls}.Value
    Transforms=@"Parent Constraint Splines_2".Out Transforms
   )
   > Parent Constraint Splines_4 = RigVMFunctionReferenceNode(
    Display Transforms=RigVMVariableNode() {variable=display spline transforms}.Value
    In Transforms=RigVMVariableNode() {variable=Ba L Transforms}.Value
    In Weights=RigVMVariableNode() {variable=L Back Constraint Wgts}.Value
   )
   > &RigUnit_SetTransformItemArray(
    Space=GlobalSpace
    Weight=1
    bInitial=False
    bPropagateToChildren=True
    Items=RigVMVariableNode() {variable=Ba L Nulls}.Value
    Transforms=@"Parent Constraint Splines_4".Out Transforms
   )
   RigVMFunction_ControlFlowBranch = &RigVMFunction_ControlFlowBranch(Condition=RigVMVariableNode() {variable=Debug Controls On}.Value, ExecuteContext=@RigVMFunction_Sequence_1.A)
   For_Each_2 = RigVMDispatchNode(ExecuteContext=@RigVMFunction_ControlFlowBranch.True, Array=RigVMVariableNode() {variable=Debug Ctrls}.Value) {template="DISPATCH_RigVMDispatch_ArrayIterator(in Array,out Element,out Index,out Count,out Ratio)"}
   > &RigUnit_SetTransform(
    Weight=1
    bInitial=False
    Item=RigVMDispatchNode(Array=RigVMVariableNode() {variable=Debug Bones}.Value, Index=@For_Each_2.Index) {template="DISPATCH_RigVMDispatch_ArrayGetAtIndex(in Array,in Index,out Element)"}.Element
    Value=&RigUnit_GetTransform(bInitial=False, Item=@For_Each_2.Element).Transform
   )
   n22 = &RigUnit_SphericalPoseReader(
    ActiveRegionScaleFactors=(PositiveWidth=1,NegativeWidth=1,PositiveHeight=1,NegativeHeight=1)
    ActiveRegionSize=0.001
    Debug=(bDrawDebug=false,bDraw2D=False,bDrawLocalAxes=False,DebugScale=25,DebugSegments=20,DebugThickness=0.05)
    DriverAxis=(X=1,Y=0,Z=0)
    DriverItem=(Type=Bone,Name="head")
    ExecutePin=()
    FalloffRegionScaleFactors=(PositiveWidth=1,NegativeWidth=1,PositiveHeight=1,NegativeHeight=1)
    FalloffSize=0.5
    FlipHeightScaling=False
    FlipWidthScaling=False
    OptionalParentItem=(Type=Bone,Name="spine_05")
    RotationOffset=(X=180,Y=0,Z=0)
   )
   > &RigUnit_ModifyTransforms(
    ItemToModify="((Item=(Type=Null,Name=\"Hair Fr L 00\"),Transform=(Rotation=(X=0,Y=0,Z=0,W=-1),Translation=(X=-3.0596,Y=0,Z=2.0485),Scale3D=(X=1,Y=1,Z=1))),(Item=(Type=Null,Name=\"Hair Fr L 01\"),Transform=(Rotation=(X=0,Y=0,Z=0,W=1),Translation=(X=-0.2795,Y=1.103,Z=4.4621), ...(truncated)"
    WeightMinimum=0
    Weight=@n22.OutputParam
   )
   n23 = &RigUnit_SphericalPoseReader(
    ActiveRegionScaleFactors=(PositiveWidth=1,NegativeWidth=1,PositiveHeight=1,NegativeHeight=1)
    ActiveRegionSize=0.001
    Debug=(bDrawDebug=false,bDraw2D=False,bDrawLocalAxes=False,DebugScale=25,DebugSegments=20,DebugThickness=0.075)
    DriverAxis=(X=1,Y=0,Z=0)
    DriverItem=(Type=Bone,Name="clavicle_l")
    ExecutePin=()
    FalloffRegionScaleFactors=(PositiveWidth=1,NegativeWidth=1,PositiveHeight=1,NegativeHeight=1)
    FalloffSize=0.5
    FlipHeightScaling=False
    FlipWidthScaling=False
    OptionalParentItem=(Type=Bone,Name="None")
    RotationOffset=(X=0,Y=0,Z=0)
   )
   > &RigUnit_ModifyTransforms(
    ItemToModify=((Item=(Type=Bone,Name="None"),Transform=(Translation=(X=0,Y=0,Z=0),Scale3D=(X=1,Y=1,Z=1))))
    WeightMinimum=0
    Weight=@n23.OutputParam
   )
   > n24 = &RigUnit_SphericalPoseReader(
    ActiveRegionScaleFactors=(PositiveWidth=1,NegativeWidth=1,PositiveHeight=1,NegativeHeight=1)
    ActiveRegionSize=0.001
    Debug=(bDrawDebug=false,bDraw2D=False,bDrawLocalAxes=False,DebugScale=25,DebugSegments=20,DebugThickness=0.075)
    DriverAxis=(X=1,Y=0,Z=0)
    DriverItem=(Type=Bone,Name="clavicle_r")
    FalloffRegionScaleFactors=(PositiveWidth=1,NegativeWidth=1,PositiveHeight=1,NegativeHeight=1)
    FalloffSize=0.5
    FlipHeightScaling=False
    FlipWidthScaling=False
    OptionalParentItem=(Type=Bone,Name="None")
    RotationOffset=(X=0,Y=0,Z=0)
   )
   > &RigUnit_ModifyTransforms(
    ItemToModify=((Item=(Type=Bone,Name="None"),Transform=(Translation=(X=0,Y=0,Z=0),Scale3D=(X=1,Y=1,Z=1))))
    WeightMinimum=0
    Weight=@n24.OutputParam
   )
   &RigUnit_ModifyTransforms(
    ExecutePin=()
    ItemToModify="((Item=(Type=Null,Name=\"Hair Ba C 01\"),Transform=(Rotation=(X=0,Y=0,Z=0,W=1),Translation=(X=-0.925,Y=0,Z=10.1856),Scale3D=(X=1,Y=1,Z=1))),(Item=(Type=Null,Name=\"Hair Ba C 00\"),Transform=(Rotation=(X=0,Y=0,Z=0,W=1),Translation=(X=-6.3404,Y=0,Z=2.8519),Scale ...(truncated)"
    WeightMinimum=0
    Weight=@n4.OutputParam
   )
   &RigUnit_ModifyTransforms(
    ExecutePin=()
    ItemToModify="((Item=(Type=Null,Name=\"Hair Ba C 01\"),Transform=(Rotation=(X=0,Y=0,Z=0,W=1),Translation=(X=-0.925,Y=0,Z=10.1856),Scale3D=(X=1,Y=1,Z=1))),(Item=(Type=Null,Name=\"Hair Ba C 00\"),Transform=(Rotation=(X=0,Y=0,Z=0,W=1),Translation=(X=-6.3404,Y=0,Z=2.8519),Scale ...(truncated)"
    WeightMinimum=0
    Weight=@n5.OutputParam
   )
   &RigVMFunction_ControlFlowBranch(ExecuteContext=(), Condition=RigVMVariableNode() {variable=BackPack On}.Value)
  ]}
 RigVMModel Rig {nodes [
   n1 = &RigUnit_PrepareForExecution()
   RigVMFunction_Sequence = RigVMAggregateNode(ExecuteContext=@n1.ExecutePin)
   ItemArray_11 = RigVMDispatchNode(Value=((Type=Bone,Name="base_hair_fr_l"),(Type=Bone,Name="dyn_hair_fr_l_02"),(Type=Bone,Name="dyn_hair_fr_l_03"),(Type=Bone,Name="dyn_hair_fr_l_04"))) {template="DISPATCH_RigVMDispatch_Constant(io Value)"}
   RigVMVariableNode(Value=@ItemArray_11.Value, ExecuteContext=@RigVMFunction_Sequence.B) {variable=Fr L Bones}
   ItemArray_2 = RigVMDispatchNode(Value=((Type=Bone,Name="base_hair_fr_r"),(Type=Bone,Name="dyn_hair_fr_r_02"),(Type=Bone,Name="dyn_hair_fr_r_03"))) {template="DISPATCH_RigVMDispatch_Constant(io Value)"}
   > RigVMVariableNode(Value=@ItemArray_2.Value) {variable=Fr R Bones}
   ItemArray_1 = RigVMDispatchNode(Value=((Type=Bone,Name="base_hair_ba_l"),(Type=Bone,Name="dyn_hair_ba_l_02"),(Type=Bone,Name="base_hair_ba_l_03"),(Type=Bone,Name="dyn_hair_ba_l_04"),(Type=Bone,Name="dyn_hair_ba_l_05"),(Type=Bone,Name="dyn_hair_ba_l_06"),(Type=Bone,Name="dyn_hair_ba_l_07"))) {template="DISPATCH_RigVMDispatch_Constant(io Value)"}
   > RigVMVariableNode(Value=@ItemArray_1.Value) {variable=Ba L Bones}
   ItemArray_6 = RigVMDispatchNode(Value=((Type=Bone,Name="base_hair_ct"),(Type=Bone,Name="dyn_hair_ct_02"),(Type=Bone,Name="base_hair_ct_03"),(Type=Bone,Name="dyn_hair_ct_04"),(Type=Bone,Name="dyn_hair_ct_05"),(Type=Bone,Name="dyn_hair_ct_06"),(Type=Bone,Name="dyn_hair_ct_07"))) {template="DISPATCH_RigVMDispatch_Constant(io Value)"}
   > RigVMVariableNode(Value=@ItemArray_6.Value) {variable=Ba C Bones}
   ItemArray_12 = RigVMDispatchNode(Value=((Type=Bone,Name="base_hair_ba_r"),(Type=Bone,Name="dyn_hair_ba_r_02"),(Type=Bone,Name="base_hair_ba_r_03"),(Type=Bone,Name="dyn_hair_ba_r_04"),(Type=Bone,Name="dyn_hair_ba_r_05"),(Type=Bone,Name="dyn_hair_ba_r_06"),(Type=Bone,Name="dyn_hair_ba_r_07"))) {template="DISPATCH_RigVMDispatch_Constant(io Value)"}
   > RigVMVariableNode(Value=@ItemArray_12.Value) {variable=Ba R Bones}
   n2 = &RigUnit_ControlRigSplineFromTransforms(
    Compression=0
    SplineMode=Hermite
    Stretch=0
    bClosed=False
    Transforms=RigVMVariableNode() {variable=Fr L Transforms}.Value
    SamplesPerSegment=RigVMVariableNode() {variable=Samples Per Segment}.Value
   )
   RigVMVariableNode(Value=@n2.Spline, ExecuteContext=@RigVMFunction_Sequence.E) {variable=Fr L Init Spline}
   > RigVMVariableNode(Value=&RigUnit_GetLengthControlRigSpline(Spline=@n2.Spline).Length) {variable=Fr L Orig Length}
   > Spline Initial Percents And Offsets = RigVMFunctionReferenceNode(Driven Bones=RigVMVariableNode() {variable=Fr L Bones}.Value, In Spline=@n2.Spline)
   > RigVMVariableNode(Value=@"Spline Initial Percents And Offsets".Out Percents) {variable=Fr L Orig Percents}
   > RigVMVariableNode(Value=@"Spline Initial Percents And Offsets".Out Local Offsets) {variable=Fr L Local Offsets}
   n3 = &RigUnit_ControlRigSplineFromTransforms(
    Compression=0
    SamplesPerSegment=8
    SplineMode=Hermite
    Stretch=0
    bClosed=False
    Transforms=RigVMVariableNode() {variable=Fr R Transforms}.Value
   )
   > RigVMVariableNode(Value=@n3.Spline) {variable=Fr R Init Spline}
   > RigVMVariableNode(Value=&RigUnit_GetLengthControlRigSpline(Spline=@n3.Spline).Length) {variable=Fr R Orig Length}
   > Spline Initial Percents And Offsets_1 = RigVMFunctionReferenceNode(Driven Bones=RigVMVariableNode() {variable=Fr R Bones}.Value, In Spline=@n3.Spline)
   > RigVMVariableNode(Value=@"Spline Initial Percents And Offsets_1".Out Percents) {variable=Fr R Orig Percents}
   > RigVMVariableNode(Value=@"Spline Initial Percents And Offsets_1".Out Local Offsets) {variable=Fr R Local Offsets}
   n4 = &RigUnit_ControlRigSplineFromTransforms(
    Compression=0
    SamplesPerSegment=8
    SplineMode=Hermite
    Stretch=0
    bClosed=False
    Transforms=RigVMVariableNode() {variable=Ba L Transforms}.Value
   )
   > RigVMVariableNode(Value=@n4.Spline) {variable=Ba L Init Spline}
   > RigVMVariableNode(Value=&RigUnit_GetLengthControlRigSpline(Spline=@n4.Spline).Length) {variable=Ba L Orig Length}
   > Spline Initial Percents And Offsets_2 = RigVMFunctionReferenceNode(Driven Bones=RigVMVariableNode() {variable=Ba L Bones}.Value, In Spline=@n4.Spline)
   > RigVMVariableNode(Value=@"Spline Initial Percents And Offsets_2".Out Percents) {variable=Ba L Orig Percents}
   > RigVMVariableNode(Value=@"Spline Initial Percents And Offsets_2".Out Local Offsets) {variable=Ba L Local Offsets}
   n5 = &RigUnit_ControlRigSplineFromTransforms(
    Compression=0
    SamplesPerSegment=8
    SplineMode=Hermite
    Stretch=0
    bClosed=False
    Transforms=RigVMVariableNode() {variable=Ba C Transforms}.Value
   )
   > RigVMVariableNode(Value=@n5.Spline) {variable=Ba C Init Spline}
   > RigVMVariableNode(Value=&RigUnit_GetLengthControlRigSpline(Spline=@n5.Spline).Length) {variable=Ba C Orig Length}
   > Spline Initial Percents And Offsets_3 = RigVMFunctionReferenceNode(Driven Bones=RigVMVariableNode() {variable=Ba C Bones}.Value, In Spline=@n5.Spline)
   > RigVMVariableNode(Value=@"Spline Initial Percents And Offsets_3".Out Percents) {variable=Ba C Orig Percents}
   > RigVMVariableNode(Value=@"Spline Initial Percents And Offsets_3".Out Local Offsets) {variable=Ba C Local Offsets}
   n6 = &RigUnit_ControlRigSplineFromTransforms(
    Compression=0
    SamplesPerSegment=8
    SplineMode=Hermite
    Stretch=0
    bClosed=False
    Transforms=RigVMVariableNode() {variable=Ba R Transforms}.Value
   )
   > RigVMVariableNode(Value=@n6.Spline) {variable=Ba R Init Spline}
   > RigVMVariableNode(Value=&RigUnit_GetLengthControlRigSpline(Spline=@n6.Spline).Length) {variable=Ba R Orig Length}
   > Spline Initial Percents And Offsets_4 = RigVMFunctionReferenceNode(Driven Bones=RigVMVariableNode() {variable=Ba R Bones}.Value, In Spline=@n6.Spline)
   > RigVMVariableNode(Value=@"Spline Initial Percents And Offsets_4".Out Percents) {variable=Ba R Orig Percents}
   > RigVMVariableNode(Value=@"Spline Initial Percents And Offsets_4".Out Local Offsets) {variable=Ba R Local Offsets}
   n7 = &RigUnit_HierarchyImportFromSkeleton(NameSpace=None, bIncludeCurves=False, bIncludeMeshSockets=False, bIncludeVirtualBones=True, ExecutePin=@RigVMFunction_Sequence.A)
   RigVMFunction_ControlFlowBranch = &RigVMFunction_ControlFlowBranch(ExecuteContext=@n7.ExecutePin, Condition=RigVMVariableNode() {variable=Debug Controls On}.Value)
   RigVMDispatchNode(ExecuteContext=@RigVMFunction_ControlFlowBranch.True, Array=RigVMVariableNode() {variable=Debug Ctrls}.Value) {template="DISPATCH_RigVMDispatch_ArrayReset(io Array)"}
   ItemArray_5 = RigVMDispatchNode(Value="((Type=Bone,Name=\"spine_03\"),(Type=Bone,Name=\"spine_04\"),(Type=Bone,Name=\"spine_05\"),(Type=Bone,Name=\"neck_01\"),(Type=Bone,Name=\"neck_02\"),(Type=Bone,Name=\"head\"),(Type=Bone,Name=\"clavicle_l\"),(Type=Bone,Name=\"clavicle_r\"),(Type=Bone,Name=\"upperarm_l\"),(Ty ...(truncated)") {template="DISPATCH_RigVMDispatch_Constant(io Value)"}
   > RigVMVariableNode(Value=@ItemArray_5.Value) {variable=Debug Bones}
   > For_Each = RigVMDispatchNode(Array=@ItemArray_5.Value) {template="DISPATCH_RigVMDispatch_ArrayIterator(in Array,out Element,out Index,out Count,out Ratio)"}
   Concat = &RigVMFunction_NameConcat(B=" Ctrl", A=@For_Each.Element.Name) {template="Concat::Execute(in A,in B,out Result)"}
   > n8 = /Script/ControlRig.RigUnit_HierarchyAddControlTransform(
    InitialValue=(Rotation=(X=0,Y=0,Z=0,W=1),Translation=(X=0,Y=0,Z=0),Scale3D=(X=1,Y=1,Z=1))
    OffsetSpace=GlobalSpace
    Settings="(InitialSpace=LocalSpace,bUsePreferredRotationOrder=False,PreferredRotationOrder=YZX,Limits=(LimitTranslationX=(bMinimum=False,bMaximum=False),LimitTranslationY=(bMinimum=False,bMaximum=False),LimitTranslationZ=(bMinimum=False,bMaximum=False),LimitPitch=(b ...(truncated)"
    Name=@Concat.Result
    Parent=&RigUnit_HierarchyGetParent(bDefaultParent=True, Child=@For_Each.Element).Parent
    OffsetTransform=&RigUnit_GetTransform(bInitial=False, Item=@For_Each.Element).Transform
    Settings.Shape=RigVMVariableNode() {variable=Debug Shape}.Value
   ) {
    is=struct RigVMUnitNode
    template="SpawnControl::Execute(in InitialValue,in Settings,in OffsetTransform,in OffsetSpace,in Parent,in Name,out Item)"
   }
   > RigVMDispatchNode(Element=@n8.Item, Array=RigVMVariableNode() {variable=Debug Ctrls}.Value) {template="DISPATCH_RigVMDispatch_ArrayAdd(io Array,in Element,out Index)"}
   Make_Array_4 = RigVMDispatchNode(Values="((ConstraintParents_3_#1=((Item=(Type=Bone,Name=\"head\"),Weight=1))),((),ConstraintParents_3_#1=((Item=(Type=Bone,Name=\"spine_05\"),Weight=0.8),(Item=(Type=Bone,Name=\"head\"),Weight=0.2))),((),Constr ...(truncated)") {template="DISPATCH_RigVMDispatch_ArrayMake(in Values,out Array)"}
   RigVMVariableNode(Value=@Make_Array_4.Array, ExecuteContext=@RigVMFunction_Sequence.F) {variable=L Front Constraint Wgts}
   Make_Array_5 = RigVMDispatchNode(Values="((ConstraintParents_3_#1=((Item=(Type=Bone,Name=\"head\"),Weight=1))),((),ConstraintParents_3_#1=((Item=(Type=Bone,Name=\"spine_05\"),Weight=1),(Item=(Type=Bone,Name=\"head\"),Weight=0.1))),((),Constrai ...(truncated)") {template="DISPATCH_RigVMDispatch_ArrayMake(in Values,out Array)"}
   > RigVMVariableNode(Value=@Make_Array_5.Array) {variable=R Front Constraint Wgts}
   Make_Array_6 = RigVMDispatchNode(Values="((ConstraintParents_3_#1=((Item=(Type=Bone,Name=\"head\"),Weight=1))),((),ConstraintParents_3_#1=((Item=(Type=Bone,Name=\"spine_05\"),Weight=1),(Item=(Type=Bone,Name=\"head\"),Weight=1))),((),Constraint ...(truncated)") {template="DISPATCH_RigVMDispatch_ArrayMake(in Values,out Array)"}
   > RigVMVariableNode(Value=@Make_Array_6.Array) {variable=L Back Constraint Wgts}
   Make_Array_1 = RigVMDispatchNode(Values="((ConstraintParents_3_#1=((Item=(Type=Bone,Name=\"head\"),Weight=1))),((),ConstraintParents_3_#1=((Item=(Type=Bone,Name=\"spine_05\"),Weight=1),(Item=(Type=Bone,Name=\"head\"),Weight=1))),((),Constraint ...(truncated)") {template="DISPATCH_RigVMDispatch_ArrayMake(in Values,out Array)"}
   > RigVMVariableNode(Value=@Make_Array_1.Array) {variable=C Back Constraint Wgts}
   Make_Array_3 = RigVMDispatchNode(Values="((ConstraintParents_3_#1=((Item=(Type=Bone,Name=\"head\"),Weight=1))),((),ConstraintParents_3_#1=((Item=(Type=Bone,Name=\"spine_05\"),Weight=1),(Item=(Type=Bone,Name=\"head\"),Weight=1))),((),Constraint ...(truncated)") {template="DISPATCH_RigVMDispatch_ArrayMake(in Values,out Array)"}
   > RigVMVariableNode(Value=@Make_Array_3.Array) {variable=R Back Constraint Wgts}
   Create Spline Nulls = RigVMFunctionReferenceNode(In Name="Hair Fr L ", In Transforms=RigVMVariableNode() {variable=Fr L Transforms}.Value, ExecuteContext=@RigVMFunction_Sequence.D)
   > RigVMVariableNode(Value=@"Create Spline Nulls".Out Nulls) {variable=Fr L Nulls}
   > Create Spline Nulls_1 = RigVMFunctionReferenceNode(In Name="Hair Fr R ", In Transforms=RigVMVariableNode() {variable=Fr R Transforms}.Value)
   > RigVMVariableNode(Value=@"Create Spline Nulls_1".Out Nulls) {variable=Fr R Nulls}
   > Create Spline Nulls_2 = RigVMFunctionReferenceNode(In Name="Hair Ba L ", In Transforms=RigVMVariableNode() {variable=Ba L Transforms}.Value)
   > RigVMVariableNode(Value=@"Create Spline Nulls_2".Out Nulls) {variable=Ba L Nulls}
   > Create Spline Nulls_3 = RigVMFunctionReferenceNode(In Name="Hair Ba C ", In Transforms=RigVMVariableNode() {variable=Ba C Transforms}.Value)
   > RigVMVariableNode(Value=@"Create Spline Nulls_3".Out Nulls) {variable=Ba C Nulls}
   > Create Spline Nulls_4 = RigVMFunctionReferenceNode(In Name="Hair Ba R ", In Transforms=RigVMVariableNode() {variable=Ba R Transforms}.Value)
   > RigVMVariableNode(Value=@"Create Spline Nulls_4".Out Nulls) {variable=Ba R Nulls}
   > RigVMVariableNode(Value=6) {variable=Samples Per Segment}
   n9 = &RigUnit_AimBoneMath(
    DebugSettings=(bEnabled=False,Scale=10,WorldOffset=(Rotation=(X=0,Y=0,Z=0,W=1),Translation=(X=0,Y=0,Z=0),Scale3D=(X=1,Y=1,Z=1)))
    InputTransform=(Rotation=(X=0,Y=0,Z=0,W=1),Translation=(X=0,Y=0,Z=0),Scale3D=(X=1,Y=1,Z=1))
    Primary=(Weight=1,Axis=(X=1,Y=0,Z=0),Target=(X=0,Y=0,Z=-1),Kind=Direction,Space=(Type=Bone,Name="None"))
    Secondary=(Weight=1,Axis=(X=0,Y=1,Z=0),Target=(X=0,Y=1,Z=0),Kind=Direction,Space=(Type=Bone,Name="None"))
    Weight=1
   )
   ItemArray_3 = RigVMDispatchNode(Value=((Type=Bone,Name="base_hair_fr_l"),(Type=Bone,Name="dyn_hair_fr_l_03"))) {template="DISPATCH_RigVMDispatch_Constant(io Value)"}
   RigUnit_GetTransform = &RigUnit_GetTransform(Item=(Type=Bone,Name="dyn_hair_fr_l_04"), bInitial=False)
   n10 = /Script/RigVM.RigVMFunction_MathVectorSub(B=(X=-0.75,Y=-2,Z=13), A=@RigUnit_GetTransform.Transform.Translation) {is=struct RigVMUnitNode, template="Subtract::Execute(in A,in B,out Result)"}
   n11 = RigVMDispatchNode(
    Other=((Rotation=(X=0,Y=0,Z=0,W=1),Translation=(X=0,Y=0,Z=0),Scale3D=(X=1,Y=1,Z=1)),(Rotation=(X=0,Y=0,Z=0,W=1),Translation=(X=0,Y=0,Z=0),Scale3D=(X=1,Y=1,Z=1)))
    Other.0.Rotation=@n9.Result.Rotation
    Other.1.Rotation=@n9.Result.Rotation
    Array=&RigUnit_GetTransformItemArray(Space=GlobalSpace, bInitial=true, Items=@ItemArray_3.Value).Transforms
    Other.0.Translation=@n10.Result
    Other.1.Translation=&RigVMFunction_MathVectorSub(B=(X=0,Y=0,Z=24.5), A=@n10.Result) {template="Subtract::Execute(in A,in B,out Result)"}.Result
    ExecuteContext=@RigVMFunction_Sequence.C
   ) {template="DISPATCH_RigVMDispatch_ArrayAppend(io Array,in Other)"}
   > RigVMVariableNode(Value=@n11.Array) {variable=Fr L Transforms}
   n12 = &RigUnit_AimBoneMath(
    DebugSettings=(bEnabled=False,Scale=10,WorldOffset=(Rotation=(X=0,Y=0,Z=0,W=1),Translation=(X=0,Y=0,Z=0),Scale3D=(X=1,Y=1,Z=1)))
    InputTransform=(Rotation=(X=0,Y=0,Z=0,W=1),Translation=(X=0,Y=0,Z=0),Scale3D=(X=1,Y=1,Z=1))
    Primary=(Weight=1,Axis=(X=1,Y=0,Z=0),Target=(X=0,Y=0,Z=-1),Kind=Direction,Space=(Type=Bone,Name="None"))
    Secondary=(Weight=1,Axis=(X=0,Y=1,Z=0),Target=(X=0,Y=-1,Z=0),Kind=Direction,Space=(Type=Bone,Name="None"))
    Weight=1
   )
   Interpolate = /Script/RigVM.RigVMFunction_MathTransformLerp(
    T=0.5
    A=&RigUnit_GetTransform(Item=(Type=Bone,Name="dyn_hair_fr_r_02"), bInitial=False).Transform
    B=&RigUnit_GetTransform(Item=(Type=Bone,Name="dyn_hair_fr_r_03"), bInitial=False).Transform
   ) {is=struct RigVMUnitNode, template="Interpolate::Execute(in A,in B,in T,out Result)"}
   n13 = &RigVMFunction_MathVectorSub(
    B=(X=-0.5,Y=-2,Z=10)
    A=&RigUnit_GetTransform(Item=(Type=Bone,Name="dyn_hair_fr_r_04"), bInitial=False).Transform.Translation
   ) {template="Subtract::Execute(in A,in B,out Result)"}
   Make_Array = RigVMDispatchNode(
    Values="((Rotation=(X=0,Y=0,Z=0,W=1),Translation=(X=0,Y=0,Z=0),Scale3D=(X=1,Y=1,Z=1)),((Rotation=(X=0,Y=0,Z=0,W=1),Translation=(X=0,Y=0,Z=0),Scale3D=(X=1,Y=1,Z=1)),Rotation=(X=0,Y=0,Z=0,W=1),Translation=(X=0,Y=0,Z=0),Scale3D=(X=1,Y=1,Z=1)),((Rotation=(X=0,Y=0,Z=0, ...(truncated)"
    Values.2.Rotation=@n12.Result.Rotation
    Values.3.Rotation=@n12.Result.Rotation
    Values.1=@Interpolate.Result
    Values.0=&RigUnit_GetTransform(Item=(Type=Bone,Name="base_hair_fr_r"), bInitial=False).Transform
    Values.2.Translation=@n13.Result
    Values.3.Translation=&RigVMFunction_MathVectorSub(B=(X=0,Y=0,Z=25), A=@n13.Result) {template="Subtract::Execute(in A,in B,out Result)"}.Result
   ) {template="DISPATCH_RigVMDispatch_ArrayMake(in Values,out Array)"}
   > RigVMVariableNode(Value=@Make_Array.Array) {variable=Fr R Transforms}
   ItemArray_8 = RigVMDispatchNode(Value=((Type=Bone,Name="base_hair_ba_l"),(Type=Bone,Name="dyn_hair_ba_l_02"),(Type=Bone,Name="dyn_hair_ba_l_04"),(Type=Bone,Name="spl_hair_ba_l_07"),(Type=Bone,Name="spl_hair_ba_l_09"))) {template="DISPATCH_RigVMDispatch_Constant(io Value)"}
   > RigVMVariableNode(Value=&RigUnit_GetTransformItemArray(Space=GlobalSpace, bInitial=true, Items=@ItemArray_8.Value).Transforms) {variable=Ba L Transforms}
   ItemArray_4 = RigVMDispatchNode(Value=((Type=Bone,Name="base_hair_ct"),(Type=Bone,Name="dyn_hair_ct_02"),(Type=Bone,Name="dyn_hair_ct_04"),(Type=Bone,Name="spl_hair_ct_07"),(Type=Bone,Name="spl_hair_ct_09"))) {template="DISPATCH_RigVMDispatch_Constant(io Value)"}
   > RigVMVariableNode(Value=&RigUnit_GetTransformItemArray(Space=GlobalSpace, bInitial=true, Items=@ItemArray_4.Value).Transforms) {variable=Ba C Transforms}
   ItemArray_10 = RigVMDispatchNode(Value=((Type=Bone,Name="base_hair_ba_r"),(Type=Bone,Name="dyn_hair_ba_r_02"),(Type=Bone,Name="dyn_hair_ba_r_04"),(Type=Bone,Name="spl_hair_ba_r_07"),(Type=Bone,Name="spl_hair_ba_r_09"))) {template="DISPATCH_RigVMDispatch_Constant(io Value)"}
   > RigVMVariableNode(Value=&RigUnit_GetTransformItemArray(Space=GlobalSpace, bInitial=true, Items=@ItemArray_10.Value).Transforms) {variable=Ba R Transforms}
  ]}
]
variables [
 Ba C Bones {
  type="TArray<FRigElementKey>"
  default=((Type=Bone,Name="dyn_hair_ba_ct_01"),(Type=Bone,Name="dyn_hair_ba_ct_02"),(Type=Bone,Name="dyn_hair_ba_ct_03_base"),(Type=Bone,Name="dyn_hair_ba_ct_04"),(Type=Bone,Name="dyn_hair_ba_ct_05"),(Type=Bone,Name="dyn_hair_ba_ct_06"))
 }
 Ba C Init Spline {type=FControlRigSpline, default=()}
 Ba C Local Offsets {type="TArray<FTransform>", default=((),(),(),(),(),())}
 Ba C Nulls {
  type="TArray<FRigElementKey>"
  default=((Type=Bone,Name="dyn_hair_fr_lf_01"),(Type=Bone,Name="dyn_hair_fr_lf_02"),(Type=Bone,Name="dyn_hair_fr_lf_03_base"),(Type=Bone,Name="dyn_hair_fr_lf_04"),(Type=Bone,Name="dyn_hair_fr_lf_05"))
 }
 Ba C Orig Length {type=float, default=0}
 Ba C Orig Percents {type="TArray<float>"}
 Ba C Points {type="TArray<FVector>", default=((X=0,Y=0,Z=0),(X=0,Y=0,Z=0),(X=0,Y=0,Z=0),(X=0,Y=0,Z=0),(X=0,Y=0,Z=0),(X=0,Y=0,Z=0))}
 Ba C Spline Wgts {type="TArray<FUserDefinedStruct_#2>", default=((),(),(),(),(),(),())}
 Ba C Tick Spline {type=FControlRigSpline, default=()}
 Ba C Transforms {type="TArray<FTransform>", default=((),(),(),(),(),())}
 Ba L Bones {
  type="TArray<FRigElementKey>"
  default=((Type=Bone,Name="dyn_hair_ba_lf_01"),(Type=Bone,Name="dyn_hair_ba_lf_02"),(Type=Bone,Name="dyn_hair_ba_lf_03_base"),(Type=Bone,Name="dyn_hair_ba_lf_04"),(Type=Bone,Name="dyn_hair_ba_lf_05"))
 }
 Ba L Init Spline {type=FControlRigSpline, default=()}
 Ba L Local Offsets {type="TArray<FTransform>", default=((),(),(),(),())}
 Ba L Nulls {
  type="TArray<FRigElementKey>"
  default=((Type=Bone,Name="dyn_hair_fr_lf_01"),(Type=Bone,Name="dyn_hair_fr_lf_02"),(Type=Bone,Name="dyn_hair_fr_lf_03_base"),(Type=Bone,Name="dyn_hair_fr_lf_04"),(Type=Bone,Name="dyn_hair_fr_lf_05"))
 }
 Ba L Orig Length {type=float, default=0}
 Ba L Orig Percents {type="TArray<float>"}
 Ba L Points {type="TArray<FVector>", default=((X=0,Y=0,Z=0),(X=0,Y=0,Z=0),(X=0,Y=0,Z=0),(X=0,Y=0,Z=0),(X=0,Y=0,Z=0))}
 Ba L Spline Wgts {type="TArray<FUserDefinedStruct_#2>", default=((),(),(),(),(),(),())}
 "Ba L Tick Spline " {type=FControlRigSpline, default=()}
 Ba L Transforms {type="TArray<FTransform>", default=((),(),(),(),())}
 Ba R Bones {
  type="TArray<FRigElementKey>"
  default=((Type=Bone,Name="dyn_hair_ba_rt_01"),(Type=Bone,Name="dyn_hair_ba_rt_02"),(Type=Bone,Name="dyn_hair_ba_rt_03_base"),(Type=Bone,Name="dyn_hair_ba_rt_04"),(Type=Bone,Name="dyn_hair_ba_rt_05"))
 }
 Ba R Init Spline {type=FControlRigSpline, default=()}
 Ba R Local Offsets {type="TArray<FTransform>", default=((),(),(),(),())}
 Ba R Nulls {
  type="TArray<FRigElementKey>"
  default=((Type=Bone,Name="dyn_hair_fr_lf_01"),(Type=Bone,Name="dyn_hair_fr_lf_02"),(Type=Bone,Name="dyn_hair_fr_lf_03_base"),(Type=Bone,Name="dyn_hair_fr_lf_04"),(Type=Bone,Name="dyn_hair_fr_lf_05"))
 }
 Ba R Orig Length {type=float, default=0}
 Ba R Orig Percents {type="TArray<float>"}
 Ba R Points {type="TArray<FVector>", default=((X=0,Y=0,Z=0),(X=0,Y=0,Z=0),(X=0,Y=0,Z=0),(X=0,Y=0,Z=0),(X=0,Y=0,Z=0))}
 Ba R Spline Wgts {type="TArray<FUserDefinedStruct_#2>", default=((),(),(),(),(),(),())}
 Ba R Tick Spline {type=FControlRigSpline, default=()}
 Ba R Transforms {type="TArray<FTransform>", default=((),(),(),(),())}
 BackPack On {type=bool, default=False}
 C Back Constraint Wgts {type="TArray<FUserDefinedStruct_#3>"}
 Debug Bones {type="TArray<FRigElementKey>", default=((Type=Bone,Name="calf_l"),(Type=Bone,Name="calf_r"))}
 Debug Controls On {type=bool, default=False}
 Debug Ctrls {type="TArray<FRigElementKey>", default=(())}
 Debug Shape {
  type=FRigUnit_HierarchyAddControl_ShapeSettings
  default=(bVisible=True,Name="Circle_Thin",Color=(R=1,G=0.9718,B=0,A=1),Transform=(Rotation=(X=0,Y=-0.7071,Z=0,W=0.7071),Translation=(X=0,Y=0,Z=0),Scale3D=(X=2,Y=2,Z=2)))
 }
 Fr L Bones {
  type="TArray<FRigElementKey>"
  default=((Type=Bone,Name="dyn_hair_fr_lf_01"),(Type=Bone,Name="dyn_hair_fr_lf_02"),(Type=Bone,Name="dyn_hair_fr_lf_03_base"),(Type=Bone,Name="dyn_hair_fr_lf_04"),(Type=Bone,Name="dyn_hair_fr_lf_05"))
 }
 Fr L Init Spline {type=FControlRigSpline, default=()}
 Fr L Local Offsets {type="TArray<FTransform>", default=((),(),(),(),())}
 Fr L Nulls {
  type="TArray<FRigElementKey>"
  default=((Type=Bone,Name="dyn_hair_fr_lf_01"),(Type=Bone,Name="dyn_hair_fr_lf_02"),(Type=Bone,Name="dyn_hair_fr_lf_03_base"),(Type=Bone,Name="dyn_hair_fr_lf_04"),(Type=Bone,Name="dyn_hair_fr_lf_05"))
 }
 Fr L Orig Length {type=float, default=0}
 Fr L Orig Percents {type="TArray<float>"}
 Fr L Points {type="TArray<FVector>", default=((X=0,Y=0,Z=0),(X=0,Y=0,Z=0),(X=0,Y=0,Z=0),(X=0,Y=0,Z=0),(X=0,Y=0,Z=0))}
 Fr L Spline Wgts {type="TArray<FUserDefinedStruct_#2>", default=((),(),(),(),(),(),())}
 Fr L Tick Spline {type=FControlRigSpline, default=()}
 Fr L Transforms {type="TArray<FTransform>", default=((),(),(),(),())}
 Fr R Bones {
  type="TArray<FRigElementKey>"
  default=((Type=Bone,Name="dyn_hair_fr_rt_01"),(Type=Bone,Name="dyn_hair_fr_rt_02_base"),(Type=Bone,Name="dyn_hair_fr_rt_03"))
 }
 Fr R Init Spline {type=FControlRigSpline, default=()}
 Fr R Local Offsets {type="TArray<FTransform>", default=((),(),())}
 Fr R Nulls {
  type="TArray<FRigElementKey>"
  default=((Type=Bone,Name="dyn_hair_fr_lf_01"),(Type=Bone,Name="dyn_hair_fr_lf_02"),(Type=Bone,Name="dyn_hair_fr_lf_03_base"),(Type=Bone,Name="dyn_hair_fr_lf_04"),(Type=Bone,Name="dyn_hair_fr_lf_05"))
 }
 Fr R Orig Length {type=float, default=0}
 Fr R Orig Percents {type="TArray<float>"}
 Fr R Points {type="TArray<FVector>", default=((X=0,Y=0,Z=0),(X=0,Y=0,Z=0),(X=0,Y=0,Z=0))}
 Fr R Spline Wgts {type="TArray<FUserDefinedStruct_#2>", default=((),(),(),(),(),(),())}
 Fr R Tick Spline {type=FControlRigSpline, default=()}
 Fr R Transforms {type="TArray<FTransform>", default=((),(),())}
 HeadUp {type=double, default=0}
 L Back Constraint Wgts {type="TArray<FUserDefinedStruct_#3>"}
 L Front Constraint Wgts {type="TArray<FUserDefinedStruct_#3>"}
 Look Left {type=double, default=0}
 Look Right {type=double, default=0}
 R Back Constraint Wgts {type="TArray<FUserDefinedStruct_#3>"}
 R Front Constraint Wgts {type="TArray<FUserDefinedStruct_#3>"}
 Samples Per Segment {type=int32, default=4}
 display spline transforms {type=bool, default=False}
 display splines {type=bool, default=True}
]
