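# Anim Blueprint for the OutletVerve female face accessory, adding hair dynamics over the parent mesh pose with a variant fitted for back bling.
# The copied pose runs through a Control Rig, then feeds two spine_04-space RigidBody sims with different physics assets.
# The back-bling sim gets additive offsets on hair bones before a bool blend against the other sim.
# Events compute gravity override from DrapeUltra parameters, skydive alpha, and back-bling state.
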AnimBlueprint /BRCosmetics/Characters/Player/Female/Medium/Bodies/F_MED_OutletVerve/Meshes/Parts/F_MED_OutletVerve_FaceAcc_AnimBP
target_skeleton=&/F_MED_OutletVerve_FaceAcc_Skeleton
graphs [
 AnimGraph {nodes [
   n1 = RigidBody!(ComponentPose=Local To Component(LocalPose=Use cached pose 'startCache'()), Alpha, OverrideWorldGravity:Vector="0,0,0") -> (Pose) {
    bClampLinearTranslationLimitToRefPose=true
    bOverrideWorldGravity=true
    overridePhysicsAsset=&/F_MED_OutletVerve_FaceAcc_BackBling_Physics
    simulationSpace=BaseBoneSpace
    alphaBoolBlend {blendOption=Linear}
    baseBoneRef {boneName=spine_04}
    overrideWorldGravity {z=-980}
    simSpaceSettings {
     dampingAlpha=2
     maxAngularAcceleration=50
     maxAngularVelocity=15
     maxLinearAcceleration=4000
     maxLinearVelocity=800
     velocityScaleZ=0.4
     worldAlpha=0.7
     externalLinearDragV=<1.5 1.5 2>
    }
   }
   n2 = "Transform (Modify) Bone - Bone: dyn_hair_ba_r_05"(ComponentPose=@n1, Alpha=1, Translation="0,0,-5.5", Rotation="0,10,0", Scale="0.85,0.85,1") {
    rotationMode=BMM_Additive
    rotationSpace=BCS_BoneSpace
    scaleMode=BMM_Additive
    scaleSpace=BCS_BoneSpace
    translationMode=BMM_Additive
    translationSpace=BCS_BoneSpace
    boneToModify {boneName=dyn_hair_ba_r_05}
    rotation {yaw=10}
    scale=<0.85 0.85>
    translation {z=-5.5}
   }
   n3 = "Transform (Modify) Bone - Bone: dyn_hair_ct_05"(ComponentPose=@n2, Alpha=1, Translation="0,0,-5.5", Rotation="0,0,0", Scale="0.85,0.85,1") {
    rotationMode=BMM_Additive
    rotationSpace=BCS_BoneSpace
    scaleMode=BMM_Additive
    scaleSpace=BCS_BoneSpace
    translationMode=BMM_Additive
    translationSpace=BCS_BoneSpace
    boneToModify {boneName=dyn_hair_ct_05}
    scale=<0.85 0.85>
    translation {z=-5.5}
   }
   n4 = "Transform (Modify) Bone - Bone: dyn_hair_ba_l_05"(ComponentPose=@n3, Alpha=1, Translation="0,0,-5.5", Rotation="0,-10,0", Scale="0.85,0.85,1") {
    rotationMode=BMM_Additive
    rotationSpace=BCS_BoneSpace
    scaleMode=BMM_Additive
    scaleSpace=BCS_BoneSpace
    translationMode=BMM_Additive
    translationSpace=BCS_BoneSpace
    boneToModify {boneName=dyn_hair_ba_l_05}
    rotation {yaw=-10}
    scale=<0.85 0.85>
    translation {z=-5.5}
   }
   n5 = "Transform (Modify) Bone - Bone: dyn_hair_ba_r_05"(ComponentPose=@n4, Alpha=1, Translation="4,1,2", Rotation="0,0,0", Scale="1,0.8,1") {
    rotationMode=BMM_Additive
    rotationSpace=BCS_BoneSpace
    scaleMode=BMM_Additive
    scaleSpace=BCS_BoneSpace
    translationMode=BMM_Additive
    translationSpace=BCS_BoneSpace
    boneToModify {boneName=dyn_hair_ba_r_05}
    scale {y=0.8}
    translation=<4 1 2>
   }
   n6 = "Transform (Modify) Bone - Bone: dyn_hair_ba_l_05"(ComponentPose=@n5, Alpha=1, Translation="2,-1,3.5", Rotation="0,0,0", Scale="1,0.8,1") {
    rotationMode=BMM_Additive
    rotationSpace=BCS_BoneSpace
    scaleMode=BMM_Additive
    scaleSpace=BCS_BoneSpace
    translationMode=BMM_Additive
    translationSpace=BCS_BoneSpace
    boneToModify {boneName=dyn_hair_ba_l_05}
    scale {y=0.8}
    translation=<2 -1 3.5>
   }
   n7 = "Transform (Modify) Bone - Bone: dyn_hair_ba_r_04"(ComponentPose=@n6, Alpha=1, Translation="9,0,-5", Rotation="0,0,0", Scale="0.7,1,1") {
    rotationMode=BMM_Additive
    rotationSpace=BCS_BoneSpace
    scaleMode=BMM_Additive
    scaleSpace=BCS_BoneSpace
    translationMode=BMM_Additive
    translationSpace=BCS_BoneSpace
    boneToModify {boneName=dyn_hair_ba_r_04}
    scale {x=0.7}
    translation {x=9, z=-5}
   }
   n8 = "Transform (Modify) Bone - Bone: dyn_hair_ct_04"(ComponentPose=@n7, Alpha=1, Translation="0,0,-5.5", Rotation="0,0,0", Scale="0.7,1,1") {
    rotationMode=BMM_Additive
    rotationSpace=BCS_BoneSpace
    scaleMode=BMM_Additive
    scaleSpace=BCS_BoneSpace
    translationMode=BMM_Additive
    translationSpace=BCS_BoneSpace
    boneToModify {boneName=dyn_hair_ct_04}
    scale {x=0.7}
    translation {z=-5.5}
   }
   n9 = "Transform (Modify) Bone - Bone: dyn_hair_ba_l_04"(ComponentPose=@n8, Alpha=1, Translation="0,0,-5.5", Rotation="0,0,0", Scale="0.7,1,1") {
    rotationMode=BMM_Additive
    rotationSpace=BCS_BoneSpace
    scaleMode=BMM_Additive
    scaleSpace=BCS_BoneSpace
    translationMode=BMM_Additive
    translationSpace=BCS_BoneSpace
    boneToModify {boneName=dyn_hair_ba_l_04}
    scale {x=0.7}
    translation {z=-5.5}
   }
   Output Pose(Result=Blend Poses by bool(BlendPose_0=Component To Local(ComponentPose=@n9), BlendPose_1=Use cached pose 'dynamicsCache'()) {blendType=Linear}) {layerGroup=None}
   n10 = RigidBody(ComponentPose=Local To Component(LocalPose=Use cached pose 'startCache'()), OverrideWorldGravity="0,0,0") {
    bClampLinearTranslationLimitToRefPose=true
    bOverrideWorldGravity=true
    overridePhysicsAsset=&/F_MED_OutletVerve_FaceAcc_Physics_B1
    simulationSpace=BaseBoneSpace
    alphaBoolBlend {blendOption=Linear}
    baseBoneRef {boneName=spine_04}
    overrideWorldGravity {z=-980}
    simSpaceSettings {
     dampingAlpha=2
     maxAngularAcceleration=50
     maxAngularVelocity=15
     maxLinearAcceleration=4000
     maxLinearVelocity=800
     velocityScaleZ=0.4
     worldAlpha=0.7
     externalLinearDragV=<1.5 1.5 2>
    }
   }
   Save cached pose 'dynamicsCache'(Pose=Component To Local(ComponentPose=@n10)) {cacheName=dynamicsCache}
   n11 = Control Rig(Source=Copy Pose From Mesh() {bUseAttachedParent=true}) {controlRigAssetReference {blueprintRigClass=&/F_MED_OutletVerve_Spl_CtrlRig.F_MED_OutletVerve_Spl_CtrlRig_C}}
   Save cached pose 'startCache'(Pose=@n11) {cacheName=startCache}
  ]}
 EventGraph {nodes [
   Event BlueprintUpdateAnimation()
   > n1 = Sequence()
   > n2 = "Gravity Override State Function (Parameter Asset) (Nativized)"(
    CharacterPartAnimInstance=Self-Reference()
    GameTimeInSeconds=GetGameTimeInSeconds()
    DebugJointName=neck_02
    Parameters=/BRCosmetics/Characters/Player/Female/Medium/Bodies/F_MED_DrapeUltra/Meshes/Parts/F_MED_DrapeUltra_Parameters
   )
   > Set GravityOverride(GravityOverride=@n2)
   Set skydiveAlpha(execute=@n1.then_1, skydiveAlpha:double=SelectFloat(A=0.6, B=1, bPickA=Get bIsSkydiveFloating())) -> (Output_Get:double)
   Event OnCharacterCustomizationLoaded()
   n3 = CheckForBackbling()
   > Set isWearingBackbling(isWearingBackbling=OR Boolean!(A=@n3, B=@n3.bIsBackblingACape, C:bool=@n3.bIsBackblingLong) -> (ReturnValue))
  ]}
]
variables [
 Gravity Override Params {
  type=FortGravityOverrideParameterSet
  default=(Skydive_GoingUp=(GravityOverride=()),Skydive_FloatingDown=(GravityOverride=()),Skydive_DivingDown=(GravityOverride=()),Skydive_Parachuting=(GravityOverride=()),OnGround_Standing=(GravityOverride=()),OnGround_Moving=(GravityOverride=()),OnGround_Moving_HighRotationRate=(GravityOverride=()),OnGround_Crouch_Moving=(GravityOverride=()),EmoteOrMelee=(GravityOverride=()),Swimming=(GravityOverride=()),Front_End=(GravityOverride=()),DBNO=(GravityOverride=()),DBNO_Carried=(GravityOverride=()),OnGround_Falling=(GravityOverride=()),Ziplining=(GravityOverride=()),Skydive_Surfing=(GravityOverride=()),Riding_Standing=(GravityOverride=()),Riding_Moving=(GravityOverride=()),Riding_Moving_Targeting=(GravityOverride=()),Tactical_Sprint=(GravityOverride=()),Rail_Grinding=(GravityOverride=()),)
 }
 GravityOverride {type=Vector}
 isWearingBackbling {type=boolean}
 RBAN Settings Out {type=FortRigidBodyAnimNodeParameters}
 skydiveAlpha {type=number}
]
