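# Control Rig for the face accessory part of the F_MED_OutletVerve female medium player cosmetic, currently holding only an imported skeleton with no rig logic.
# The hierarchy runs from pelvis through head, with branching dyn_ chains under base_ parents for back and front hair, curls, tips and strands.
# It also carries two star bones on the head, clavicles and upper arms, and three cloth-volume chains on the pelvis.
# The graph has just a Begin Execution node, so nothing drives these bones here.
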ControlRigBlueprint /BRCosmetics/Characters/Player/Female/Medium/Bodies/F_MED_OutletVerve/Meshes/Parts/F_MED_OutletVerve_FaceAcc_CtrlRig
graphs [
 RigVMFunctionLibrary {links [], nodes []}
 RigVMModel {nodes [RigUnit_BeginExecution = &RigUnit_BeginExecution()]}
]
hierarchy [
 root {loc=<0 0 0>, rot=<0 0 0 1>}
 pelvis {parent=root, loc=<0 -0.5301 98.0571>, rot=<0.0127 0.707 -0.0127 -0.707>}
 spine_01 {parent=pelvis, loc=<6.1208 -0.9328 0>, rot=<0 0 0.034 -0.9994>}
 spine_02 {parent=spine_01, loc=<6.5997 -0.2095 0>, rot=<0 0 0.0238 -0.9997>}
 spine_03 {parent=spine_02, loc=<6.4914 -0.1237 0>, rot=<0 0 0.0149 -0.9999>}
 spine_04 {parent=spine_03, loc=<6.9598 0.0794 0>, rot=<0 0 -0.0132 -0.9999>}
 spine_05 {parent=spine_04, loc=<6.9987 0 0>, rot=<0 0 -0.0015 -1>}
 neck_01 {parent=spine_05, loc=<11.3138 0.1529 0>, rot=<0 0 -0.2003 -0.9797>}
 neck_02 {parent=neck_01, loc=<4.5627 0 0>, rot=<0 0 0.0504 -0.9987>}
 head {parent=neck_02, loc=<5.3914 0 0>, rot=<0 0 0.115 -0.9934>}
 base_hair_ba {parent=head, loc=<7.1777 -6.5587 0>, rot=<0.0622 -0.7044 0.7044 -0.0622>}
 dyn_hair_ba_01 {parent=base_hair_ba, loc=<0 0 0>, rot=<0 0 0 1>}
 dyn_hair_ba_02 {parent=dyn_hair_ba_01, loc=<11.6898 0 0>, rot=<0 0 0 1>}
 dyn_hair_ba_03 {parent=dyn_hair_ba_02, loc=<11.6898 0 0>, rot=<0 0 0 1>}
 dyn_hair_ba_04 {parent=dyn_hair_ba_03, loc=<11.6898 0 0>, rot=<0 0 0 1>}
 dyn_hair_ba_05 {parent=dyn_hair_ba_04, loc=<7.5296 0 0>, rot=<0 0 0 1>}
 base_hair_ba_r {parent=dyn_hair_ba_04, loc=<7.1777 -6.5587 6.812>, rot=<0.0622 -0.7044 0.7044 -0.0622>}
 dyn_hair_ba_r_01 {parent=base_hair_ba_r, loc=<0 0 0>, rot=<0 0 0 1>}
 dyn_hair_ba_r_02 {parent=dyn_hair_ba_r_01, loc=<0 0 0>, rot=<0 0 0 1>}
 base_curl_a_r {parent=dyn_hair_ba_r_02, loc=<1.9997 3.1457 1.9139>, rot=<-0.1908 -0.1241 -0.0821 -0.9703>}
 dyn_curl_a_r_01 {parent=base_curl_a_r, loc=<0 0 0>, rot=<0 0 0 1>}
 dyn_curl_a_r_02 {parent=dyn_curl_a_r_01, loc=<4.4087 0 0>, rot=<0 -0.1812 0 -0.9834>}
 base_tip_r_a_l {parent=dyn_curl_a_r_02, loc=<3.3537 2.2197 -0.5027>, rot=<0.1252 -0.2402 -0.3391 -0.9009>}
 dyn_tip_a_r_01 {parent=base_tip_r_a_l, loc=<0 0 0>, rot=<0 0 0 1>}
 dyn_tip_a_r_02 {parent=dyn_tip_a_r_01, loc=<3.1331 0 0>, rot=<0 0 0 1>}
 base_tip_r_b {parent=dyn_curl_a_r_02, loc=<4.3038 -0.1204 -0.0723>, rot=<0.1017 -0.2072 -0.4059 -0.8843>}
 dyn_tip_r_b_01 {parent=base_tip_r_b, loc=<0 0 0>, rot=<0 0 0 1>}
 dyn_tip_r_b_02 {parent=dyn_tip_r_b_01, loc=<2.4769 0 0>, rot=<0 0 0 1>}
 base_curl_b_r {parent=dyn_hair_ba_r_02, loc=<2.5145 -2.453 -0.4949>, rot=<-0.1956 -0.1011 0.0701 -0.9729>}
 dyn_curl_b_r_01 {parent=base_curl_b_r, loc=<0 0 0>, rot=<0 0 0 1>}
 dyn_curl_b_r_02 {parent=dyn_curl_b_r_01, loc=<4.7428 0 0>, rot=<0 0 -0.0985 -0.9951>}
 base_tip_r_c {parent=dyn_curl_b_r_02, loc=<0.9223 1.0771 -0.2317>, rot=<0.0463 -0.1675 -0.2624 -0.9492>}
 dyn_tip_r_c_01 {parent=base_tip_r_c, loc=<0 0 0>, rot=<0 0 0 1>}
 dyn_tip_r_c_02 {parent=dyn_tip_r_c_01, loc=<3.336 0 0>, rot=<0.027 -0.1959 -0.1913 -0.9614>}
 dyn_tip_r_c_03 {parent=dyn_tip_r_c_02, loc=<3.3106 0 0>, rot=<0 0 0 1>}
 base_tip_r_cl {parent=dyn_curl_b_r_02, loc=<1.4405 -1.2354 -0.5715>, rot=<0.023 -0.2626 -0.0841 -0.961>}
 dyn_tip_r_cl_01 {parent=base_tip_r_cl, loc=<0 0 0>, rot=<0 0 0 1>}
 dyn_tip_r_d_02 {parent=dyn_tip_r_cl_01, loc=<1.7631 0 0>, rot=<0.0958 -0.1239 -0.0894 -0.9836>}
 dyn_tip_r_d_03 {parent=dyn_tip_r_d_02, loc=<1.3711 0 0>, rot=<0 0 0 1>}
 base_curl_c_r {parent=dyn_hair_ba_r_02, loc=<1.0728 -6.7282 -3.9942>, rot=<-0.1817 -0.0833 0.2171 -0.9555>}
 dyn_curl_c_r_01 {parent=base_curl_c_r, loc=<0 0 0>, rot=<0 0 0 1>}
 base_tip_r_f {parent=dyn_curl_c_r_01, loc=<1.2428 -1.3227 -0.6347>, rot=<-0.0163 -0.1299 0.1236 -0.9837>}
 dyn_tip_r_f_01 {parent=base_tip_r_f, loc=<0 0 0>, rot=<0 0 0 1>}
 dyn_tip_r_f_02 {parent=dyn_tip_r_f_01, loc=<3.1148 0 0>, rot=<-0.0059 -0.0206 0.1498 -0.9885>}
 dyn_tip_r_f_03 {parent=dyn_tip_r_f_02, loc=<4.3586 0 0>, rot=<0 0 0 1>}
 base_tip_r_e {parent=dyn_curl_c_r_01, loc=<0.683 1.0863 0.6347>, rot=<-0.0034 -0.0994 0.0341 -0.9945>}
 dyn_tip_r_e_01 {parent=base_tip_r_e, loc=<0 0 0>, rot=<0 0 0 1>}
 dyn_tip_r_e_02 {parent=dyn_tip_r_e_01, loc=<4.2044 0 0>, rot=<0.0116 -0.0393 -0.1072 -0.9934>}
 dyn_tip_r_e_03 {parent=dyn_tip_r_e_02, loc=<4.7013 0 0>, rot=<0 0 0 1>}
 base_hair_ba_l {parent=dyn_hair_ba_04, loc=<7.1777 -6.5587 -6.8121>, rot=<0.0622 -0.7044 0.7044 -0.0622>}
 dyn_hair_ba_l_01 {parent=base_hair_ba_l, loc=<0 0 0>, rot=<0 0 0 1>}
 dyn_hair_ba_l_02 {parent=dyn_hair_ba_l_01, loc=<0 0 0>, rot=<0 0 0 1>}
 base_curl_a_l {parent=dyn_hair_ba_l_02, loc=<1.9999 -3.1457 1.9139>, rot=<0.1908 -0.1241 0.0821 -0.9703>}
 dyn_curl_a_l_01 {parent=base_curl_a_l, loc=<0 0 0>, rot=<0 0 0 1>}
 dyn_curl_a_l_02 {parent=dyn_curl_a_l_01, loc=<4.409 0 0>, rot=<0 -0.1812 0 -0.9834>}
 base_tip_l_a_l {parent=dyn_curl_a_l_02, loc=<3.3534 -2.2196 -0.5029>, rot=<-0.1252 -0.2402 0.3391 -0.9009>}
 dyn_tip_l_a_01 {parent=base_tip_l_a_l, loc=<0 0 0>, rot=<0 0 0 1>}
 dyn_tip_l_a_02 {parent=dyn_tip_l_a_01, loc=<3.1331 0 0>, rot=<0 0 0 1>}
 base_tip_l_b_l {parent=dyn_curl_a_l_02, loc=<4.3038 0.1204 -0.0723>, rot=<-0.1017 -0.2072 0.4059 -0.8843>}
 dyn_tip_l_b_01 {parent=base_tip_l_b_l, loc=<0 0 0>, rot=<0 0 0 1>}
 dyn_tip_l_b_02 {parent=dyn_tip_l_b_01, loc=<2.4768 -0.0002 -0.0002>, rot=<0 0 0 1>}
 base_curl_b_l {parent=dyn_hair_ba_l_02, loc=<2.5149 2.4528 -0.495>, rot=<0.1956 -0.1011 -0.0701 -0.9729>}
 dyn_curl_b_l_01 {parent=base_curl_b_l, loc=<0 0 0>, rot=<0 0 0 1>}
 dyn_curl_b_l_02 {parent=dyn_curl_b_l_01, loc=<4.742 0.0004 0.0001>, rot=<0 0 0.0985 -0.9951>}
 base_tip_l_cl {parent=dyn_curl_b_l_02, loc=<1.441 1.2353 -0.5715>, rot=<-0.023 -0.2626 0.0841 -0.961>}
 dyn_tip_l_cl_01 {parent=base_tip_l_cl, loc=<0 0 0>, rot=<0 0 0 1>}
 dyn_tip_l_d_02 {parent=dyn_tip_l_cl_01, loc=<1.7631 0 0>, rot=<-0.0958 -0.1239 0.0894 -0.9836>}
 dyn_tip_l_d_03 {parent=dyn_tip_l_d_02, loc=<1.371 -0.0001 -0.0001>, rot=<0 0 0 1>}
 base_tip_l_c_l {parent=dyn_curl_b_l_02, loc=<0.9226 -1.0772 -0.2318>, rot=<-0.0463 -0.1675 0.2624 -0.9492>}
 dyn_tip_l_c_01 {parent=base_tip_l_c_l, loc=<0 0 0>, rot=<0 0 0 1>}
 dyn_tip_l_c_02 {parent=dyn_tip_l_c_01, loc=<3.336 0.0001 0>, rot=<-0.027 -0.1959 0.1913 -0.9614>}
 dyn_tip_l_c_03 {parent=dyn_tip_l_c_02, loc=<2.4052 0 0>, rot=<0 0 0 1>}
 base_curl_c_l {parent=dyn_hair_ba_l_02, loc=<1.0727 6.7283 -3.9942>, rot=<0.1817 -0.0833 -0.2171 -0.9555>}
 dyn_curl_c_l_01 {parent=base_curl_c_l, loc=<0 0 0>, rot=<0 0 0 1>}
 base_tip_l_f {parent=dyn_curl_c_l_01, loc=<1.2429 1.3225 -0.6347>, rot=<0.0163 -0.1299 -0.1236 -0.9837>}
 dyn_tip_l_f_01 {parent=base_tip_l_f, loc=<0 0 0>, rot=<0 0 0 1>}
 dyn_tip_l_f_02 {parent=dyn_tip_l_f_01, loc=<3.115 -0.0003 -0.0001>, rot=<0.0059 -0.0206 -0.1498 -0.9885>}
 dyn_tip_l_f_03 {parent=dyn_tip_l_f_02, loc=<3.1531 0.0008 0>, rot=<0 0 0 1>}
 base_tip_l_e {parent=dyn_curl_c_l_01, loc=<0.683 -1.0863 0.6347>, rot=<0.0034 -0.0994 -0.0341 -0.9945>}
 dyn_tip_l_e_01 {parent=base_tip_l_e, loc=<0 0 0>, rot=<0 0 0 1>}
 dyn_tip_l_e_02 {parent=dyn_tip_l_e_01, loc=<4.2044 0.0001 0.0001>, rot=<-0.0116 -0.0393 0.1072 -0.9934>}
 dyn_tip_l_e_03 {parent=dyn_tip_l_e_02, loc=<3.3064 0.0001 0>, rot=<0 0 0 1>}
 base_strand_ba {parent=dyn_hair_ba_03, loc=<9.3383 1.9722 1.2358>, rot=<0.3163 0.3582 -0.2832 -0.8316>}
 dyn_strand_ba_01 {parent=base_strand_ba, loc=<0 0 0>, rot=<0 0 0 1>}
 dyn_strand_ba_02 {parent=dyn_strand_ba_01, loc=<2.192 0 0>, rot=<0 0 0 1>}
 base_hair_fr_r {parent=head, loc=<0.3188 3.2049 7.8734>, rot=<0.0112 0.086 0.9958 0.0284>}
 dyn_hair_fr_r_01 {parent=base_hair_fr_r, loc=<0 0 0>, rot=<0 0 0 1>}
 dyn_hair_fr_r_02 {parent=dyn_hair_fr_r_01, loc=<8.4677 0 0>, rot=<-0.0413 -0.0067 0.073 -0.9965>}
 dyn_hair_fr_r_03 {parent=dyn_hair_fr_r_02, loc=<7.4063 0 0>, rot=<-0.0951 0.0101 0.1708 -0.9806>}
 dyn_hair_fr_r_04 {parent=dyn_hair_fr_r_03, loc=<5.6659 0 0>, rot=<-0.054 -0.0346 0.2417 -0.9682>}
 dyn_hair_fr_r_05 {parent=dyn_hair_fr_r_04, loc=<4.7479 0 0>, rot=<0 0 0 1>}
 base_strand_fr_r {parent=dyn_hair_fr_r_01, loc=<3.3358 2.112 1.465>, rot=<-0.0363 0.18 0.0252 -0.9827>}
 dyn_strand_fr_r_01 {parent=base_strand_fr_r, loc=<0 0 0>, rot=<0 0 0 1>}
 dyn_strand_fr_r_02 {parent=dyn_strand_fr_r_01, loc=<1.2508 0 0>, rot=<0.0314 0.1459 0.0689 -0.9864>}
 dyn_strand_fr_r_03 {parent=dyn_strand_fr_r_02, loc=<1.3724 0 0>, rot=<0 0 0 1>}
 base_lock_fr_r {parent=head, loc=<5.3389 7.9889 6.4124>, rot=<-0.0033 0.1283 0.9897 0.0642>}
 dyn_lock_fr_r_01 {parent=base_lock_fr_r, loc=<0 0 0>, rot=<0 0 0 1>}
 dyn_lock_fr_r_02 {parent=dyn_lock_fr_r_01, loc=<4.9932 0 0>, rot=<0.0099 -0.0632 -0.0297 -0.9975>}
 dyn_lock_fr_r_03 {parent=dyn_lock_fr_r_02, loc=<4.6455 0 0>, rot=<-0.0047 -0.1791 -0.026 -0.9835>}
 dyn_lock_fr_r_04 {parent=dyn_lock_fr_r_03, loc=<6.0611 0 0>, rot=<0 0 0 1>}
 star_a {parent=head, loc=<6.2691 4.6809 -7.105>, rot=<-0.1301 -0.6993 -0.1376 -0.6892>}
 star_b {parent=head, loc=<3.4424 3.1307 -8.0111>, rot=<-0.1686 -0.6823 -0.172 -0.6903>}
 base_hair_fr_l {parent=head, loc=<1.0444 2.1447 -7.5078>, rot=<0.003 -0.995 -0.094 0.0336>}
 dyn_hair_fr_l_01 {parent=base_hair_fr_l, loc=<0 0 0>, rot=<0 0 0 1>}
 dyn_hair_fr_l_02 {parent=dyn_hair_fr_l_01, loc=<5.9706 0 0>, rot=<-0.0468 0.0086 -0.0608 -0.997>}
 dyn_hair_fr_l_03 {parent=dyn_hair_fr_l_02, loc=<6 0 0>, rot=<0.0047 -0.0236 -0.1361 -0.9904>}
 dyn_hair_fr_l_04 {parent=dyn_hair_fr_l_03, loc=<6.9978 0 0>, rot=<0 0 0 1>}
 base_strand_fr_b_l {parent=dyn_hair_fr_l_01, loc=<2.4257 -2.2424 0.6564>, rot=<-0.102 0.1628 0.2739 -0.9424>}
 dyn_strand_fr_b_l_01 {parent=base_strand_fr_b_l, loc=<0 0 0>, rot=<0 0 0 1>}
 dyn_strand_fr_b_l_02 {parent=dyn_strand_fr_b_l_01, loc=<1.3675 0 0>, rot=<-0.0849 0.187 0.0879 -0.9747>}
 dyn_strand_fr_b_l_03 {parent=dyn_strand_fr_b_l_02, loc=<1.2872 0 0>, rot=<0 0 0 1>}
 base_strand_fr_a_l {parent=dyn_hair_fr_l_01, loc=<2.6142 2.6318 -0.6268>, rot=<0.2628 -0.0799 -0.2915 -0.9163>}
 dyn_strand_fr_a_l_01 {parent=base_strand_fr_a_l, loc=<0 0 0>, rot=<0 0 0 1>}
 dyn_strand_fr_a_l_02 {parent=dyn_strand_fr_a_l_01, loc=<1.8668 0 0>, rot=<0.044 0.0186 -0.2239 -0.9734>}
 dyn_strand_fr_a_l_03 {parent=dyn_strand_fr_a_l_02, loc=<2.3114 0 0>, rot=<0 0 0 1>}
 clavicle_l {parent=spine_05, loc=<5.4831 -0.1053 -3.2946>, rot=<0.0465 -0.7289 -0.009 -0.6829>}
 upperarm_l {parent=clavicle_l, loc=<9.8658 -1.2136 -0.4105>, rot=<0.0164 -0.3245 -0.0269 -0.9454>}
 clavicle_r {parent=spine_05, loc=<5.4831 -0.1053 3.2946>, rot=<0.7289 0.0465 -0.6829 0.009>}
 upperarm_r {parent=clavicle_r, loc=<-9.8658 1.2136 0.4105>, rot=<0.0164 -0.3245 -0.0269 -0.9454>}
 dyn_Cloth_volume_bk {parent=pelvis, loc=<5.1712 -9.396 0.2721>, rot=<0 0 -0.0741 -0.9973>}
 dyn_Cloth_volume_bk01 {parent=dyn_Cloth_volume_bk, loc=<-4.7247 -0.97 0>, rot=<0 0 -0.2665 -0.9638>}
 dyn_Cloth_volume_l_C {parent=pelvis, loc=<5.5576 -7.461 -8.659>, rot=<-0.3232 0.0011 -0.0701 -0.9437>}
 dyn_Cloth_volume_l_C01 {parent=dyn_Cloth_volume_l_C, loc=<-4.5252 -1.5667 -0.6841>, rot=<0 0 -0.2665 -0.9638>}
 dyn_Cloth_volume_r_C {parent=pelvis, loc=<5.5576 -7.461 8.6589>, rot=<0.3228 -0.0004 -0.083 -0.9428>}
 dyn_Cloth_volume_r_C01 {parent=dyn_Cloth_volume_r_C, loc=<-4.5599 -1.133 0.6841>, rot=<0 0 -0.2665 -0.9638>}
]
