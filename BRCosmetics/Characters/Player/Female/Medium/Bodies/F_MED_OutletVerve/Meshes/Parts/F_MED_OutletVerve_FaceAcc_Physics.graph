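# Physics asset for the FaceAcc mesh part of the female medium OutletVerve character, covering two left-side hair chains and upper-body collision shapes.
# Capsules, plus a sphere on the head, approximate pelvis, spine, neck, clavicles and upper arms.
# The front chain hangs from base_hair_fr_l with tight swing limits and locked twist.
# The back chain links segments 02 to 06 with wider swing on one axis and no constraint to the body.
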PhysicsAsset /BRCosmetics/Characters/Player/Female/Medium/Bodies/F_MED_OutletVerve/Meshes/Parts/F_MED_OutletVerve_FaceAcc_Physics
bodies {
 base_hair_fr_l [{shapeType=Capsule, length=0.505, radius=0.505, shapeName=base_hair_fr_l_capsule, rotation=<rot 0 -90 0>}]
 clavicle_l [{shapeType=Capsule, length=3.2614, radius=4.3194, shapeName=clavicle_l_capsule, center=<6.7502 -0.2631 -1.0593>, rotation=<rot 79.0876 0 0>}]
 clavicle_r [{shapeType=Capsule, length=3.2614, radius=4.3194, shapeName=clavicle_l_capsule, center=<-6.7502 0.2631 1.0593>, rotation=<rot 79.0876 0 0>}]
 dyn_hair_ba_l_02 [{shapeType=Capsule, length=12.107, radius=3.5, shapeName=dyn_hair_02_rt_capsule, rotation=<rot 90 0 0>}]
 dyn_hair_ba_l_03 [{shapeType=Capsule, length=6.385, radius=3.5, shapeName=dyn_hair_04_cr_capsule, center=<3.1925 0 0>, rotation=<rot 90 0 0>}]
 dyn_hair_ba_l_04 [{shapeType=Capsule, length=7.286, radius=3.5, shapeName=dyn_hair_05_cr_capsule, center=<3.643 0 0>, rotation=<rot 90 0 0>}]
 dyn_hair_ba_l_05 [{shapeType=Capsule, length=7.286, radius=3.5, shapeName=dyn_hair_05_cr_capsule, center=<3.643 0 0>, rotation=<rot 90 0 0>}]
 dyn_hair_ba_l_06 [{shapeType=Capsule, length=3.349, radius=3.5, shapeName=dyn_hair_05_cr_capsule, center=<1.6745 0 0>, rotation=<rot 90 0 0>}]
 dyn_hair_fr_l_01 [{shapeType=Capsule, length=4.8259, radius=1.557, shapeName=dyn_hair_06_cr_capsule, center=<2.6495 0.0365 0.5786>, rotation=<rot 82.1876 0 0>}]
 dyn_hair_fr_l_02 [{shapeType=Capsule, length=5.4697, radius=1.5263, shapeName=dyn_hair_06_cr_capsule, center=<3.004 0 0>, rotation=<rot 90 0 0>}]
 dyn_hair_fr_l_03 [{shapeType=Capsule, length=5.2524, radius=1.2949, shapeName=dyn_hair_06_cr_capsule, center=<4.0232 0 0>, rotation=<rot 90 0 0>}]
 head [
  {shapeType=Sphere, radius=7.8541, shapeName=head_sphere, center=<4.6224 0.3144 0>}
  {shapeType=Capsule, length=5.6665, radius=5.881, shapeName=head_capsule, center=<1.2587 1.6335 0>, rotation=<rot 0 -110.4617 90>}
 ]
 neck_01 [{shapeType=Capsule, length=2.6659, radius=3.4708, shapeName=neck_01_capsule, center=<1.8614 1.4972 0>, rotation=<rot 90 -1.7439 0>}]
 neck_02 [{shapeType=Capsule, length=2.6659, radius=3.4708, shapeName=neck_01_capsule, center=<1.8636 1.3451 0>, rotation=<rot 90 0.8385 0>}]
 pelvis [{shapeType=Capsule, length=1.2338, radius=13.9109, shapeName=spine_01_capsule, center=<-3.9777 -0.1407 0>, rotation=<rot 90 23.6731 0>}]
 spine_01 [{shapeType=Capsule, length=1.2338, radius=13.9109, shapeName=spine_01_capsule, center=<-3.9777 -0.1407 0>, rotation=<rot 90 23.6731 0>}]
 spine_02 [{shapeType=Capsule, length=0.5807, radius=10.0802, shapeName=spine_02_capsule, center=<-1.5174 1.2345 0>, rotation=<rot 90 11.5144 0>}]
 spine_03 [{shapeType=Capsule, length=1.3045, radius=10.1618, shapeName=spine_02_capsule, center=<-1.5174 1.2345 0>, rotation=<rot 90 11.5144 0>}]
 spine_04 [
  {
   shapeType=Capsule
   length=14.0675
   radius=5.9493
   shapeName=spine_02_capsule
   center=<2.9775 2.1363 -2.7563>
   rotation=<rot 81.5541 -11.7356 4.713>
  }
  {shapeType=Capsule, length=12.2909, radius=5.967, shapeName=spine_04_capsule, center=<8.7004 0.5299 0>}
  {
   shapeType=Capsule
   length=12.858
   radius=7.1587
   shapeName=spine_02_capsule
   center=<2.1795 1.9019 2.9695>
   rotation=<rot 77.8057 168.2644 -176.3544>
  }
 ]
 spine_05 [{shapeType=Capsule, length=17.5974, radius=5.8447, shapeName=spine_05_capsule, center=<3.0108 -0.2884 0>, rotation=<rot 0 -90 0>}]
 upperarm_l [{shapeType=Capsule, length=26.1842, radius=4.0658, shapeName=upperarm_l_capsule, center=<13 -0.0532 -1.1454>, rotation=<rot 90 -3.8402 0>}]
 upperarm_r [{shapeType=Capsule, length=26.1842, radius=4.0658, shapeName=upperarm_l_capsule, center=<-13 0.0532 1.1454>, rotation=<rot 90 -3.8402 0>}]
}
constraints [
 {bone1Name=dyn_hair_fr_l_03, bone2Name=dyn_hair_fr_l_02, swing1LimitDegrees=15, swing2LimitDegrees=6, twistLimitDegrees=3, twistMotion=Locked}
 {bone1Name=dyn_hair_fr_l_02, bone2Name=dyn_hair_fr_l_01, swing1LimitDegrees=15, swing2LimitDegrees=6, twistLimitDegrees=3, twistMotion=Locked}
 {bone1Name=dyn_hair_fr_l_01, bone2Name=base_hair_fr_l, swing1LimitDegrees=15, swing2LimitDegrees=6, twistLimitDegrees=3, twistMotion=Locked}
 {bone1Name=dyn_hair_ba_l_03, bone2Name=dyn_hair_ba_l_02, swing1LimitDegrees=9, swing2LimitDegrees=45, twistLimitDegrees=0}
 {bone1Name=dyn_hair_ba_l_04, bone2Name=dyn_hair_ba_l_03, swing1LimitDegrees=9, swing2LimitDegrees=45, twistLimitDegrees=0}
 {bone1Name=dyn_hair_ba_l_05, bone2Name=dyn_hair_ba_l_04, swing1LimitDegrees=9, swing2LimitDegrees=45, twistLimitDegrees=0}
 {bone1Name=dyn_hair_ba_l_06, bone2Name=dyn_hair_ba_l_05, swing1LimitDegrees=9, swing2LimitDegrees=45, twistLimitDegrees=0}
]
