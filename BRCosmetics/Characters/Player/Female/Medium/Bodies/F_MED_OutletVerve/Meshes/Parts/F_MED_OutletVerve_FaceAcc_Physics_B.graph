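# Physics asset for the hair of the OutletVerve medium female character's face accessory part, with capsule chains for back left, right and center, front left and right, and a front-right lock.
# Head, neck, spine, clavicle and upper-arm bodies supply collision shapes around the hair.
# Each chain uses limited-swing joints with little or no twist, and front base bodies are locked angularly to the head.
# Extra angularly free constraints cross-link lower back-chain bodies.
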PhysicsAsset /BRCosmetics/Characters/Player/Female/Medium/Bodies/F_MED_OutletVerve/Meshes/Parts/F_MED_OutletVerve_FaceAcc_Physics_B
bodies {
 base_hair_ba_l_03 [{shapeType=Capsule, length=0.505, radius=0.505, shapeName=base_hair_ba_l_03_capsule, rotation=<rot 0 -90 0>}]
 base_hair_ba_r_03 [{shapeType=Capsule, length=0.505, radius=0.505, shapeName=base_hair_ba_r_03_capsule, rotation=<rot 0 -90 0>}]
 base_hair_ct_03 [{shapeType=Capsule, length=0.505, radius=0.505, shapeName=base_hair_ba_r_03_capsule, rotation=<rot 0 -90 0>}]
 base_hair_fr_l [{shapeType=Capsule, length=0.505, radius=0.505, shapeName=base_hair_fr_l_capsule, rotation=<rot 0 -90 0>}]
 base_hair_fr_r [{shapeType=Capsule, length=0.505, radius=0.505, shapeName=base_hair_fr_r_capsule, rotation=<rot 0 -90 0>}]
 base_lock_fr_r [{shapeType=Capsule, length=0.505, radius=0.505, shapeName=base_lock_fr_r_capsule, rotation=<rot 0 -90 0>}]
 clavicle_l [{shapeType=Capsule, length=3.2614, radius=4.3194, shapeName=clavicle_l_capsule, center=<6.7502 -0.5 -1.0593>, rotation=<rot 79.0876 0 0>}]
 clavicle_r [{shapeType=Capsule, length=6.2614, radius=4.3194, shapeName=clavicle_l_capsule, center=<-4.8319 0.5 0.6895>, rotation=<rot 79.0876 0 0>}]
 dyn_hair_ba_l_03 [{shapeType=Capsule, length=5, radius=3, shapeName=dyn_hair_04_cr_capsule, center=<4 0 -3.5>, rotation=<rot 0 0 90>}]
 dyn_hair_ba_l_04 [{shapeType=Capsule, length=4.5, radius=3, shapeName=dyn_hair_05_cr_capsule, center=<4 0 -1.5>, rotation=<rot 0 0 90>}]
 dyn_hair_ba_l_05 [{shapeType=Capsule, length=4.5, radius=3, shapeName=dyn_hair_05_cr_capsule, center=<4 0 0>, rotation=<rot 0 0 90>}]
 dyn_hair_ba_l_06 [{shapeType=Capsule, length=5, radius=3, shapeName=dyn_hair_05_cr_capsule, center=<4 0 0>, rotation=<rot 0 0 90>}]
 dyn_hair_ba_l_07 [{shapeType=Capsule, length=5, radius=3, shapeName=dyn_hair_05_cr_capsule, center=<5 0 0>, rotation=<rot 0 0 90>}]
 dyn_hair_ba_r_03 [{shapeType=Capsule, length=5, radius=3, shapeName=dyn_hair_04_cr_capsule, center=<4 0 -3.5>, rotation=<rot 0 0 90>}]
 dyn_hair_ba_r_04 [{shapeType=Capsule, length=4.5, radius=3, shapeName=dyn_hair_05_cr_capsule, center=<4 0 -1.5>, rotation=<rot 82.723 -180 -180>}]
 dyn_hair_ba_r_05 [{shapeType=Capsule, length=4.5, radius=3, shapeName=dyn_hair_05_cr_capsule, center=<4 0 0>, rotation=<rot 0 0 90>}]
 dyn_hair_ba_r_06 [{shapeType=Capsule, length=5, radius=3, shapeName=dyn_hair_05_cr_capsule, center=<4 0 0>, rotation=<rot 0 0 90>}]
 dyn_hair_ba_r_07 [
  {shapeType=Capsule, length=5, radius=3, shapeName=dyn_hair_05_cr_capsule, center=<5 0 0>, rotation=<rot 0 0 90>}
  {shapeType=Capsule, length=2.5, radius=2.5, shapeName=dyn_hair_ba_r_07_capsule, center=<4.25 -8.7 -4.3>, rotation=<rot 0 0 90>}
 ]
 dyn_hair_ct_03 [{shapeType=Capsule, length=5, radius=3, shapeName=dyn_hair_04_cr_capsule, center=<4 0 -2>, rotation=<rot 0 0 90>}]
 dyn_hair_ct_04 [{shapeType=Capsule, length=4.5, radius=3, shapeName=dyn_hair_05_cr_capsule, center=<4 0 0>, rotation=<rot 0 0 90>}]
 dyn_hair_ct_05 [{shapeType=Capsule, length=4, radius=3, shapeName=dyn_hair_05_cr_capsule, center=<4 0 0>, rotation=<rot 0 0 90>}]
 dyn_hair_ct_06 [{shapeType=Capsule, length=4.5, radius=3, shapeName=dyn_hair_05_cr_capsule, center=<4 0 0>, rotation=<rot 0 0 90>}]
 dyn_hair_ct_07 [{shapeType=Capsule, length=4.5, radius=3, shapeName=dyn_hair_05_cr_capsule, center=<5 0 0>, rotation=<rot 0 0 90>}]
 dyn_hair_fr_l_01 [{shapeType=Capsule, length=2.5, radius=1.5, shapeName=dyn_hair_fr_l_01_capsule, center=<3 0 0.1>, rotation=<rot 0 0 -90>}]
 dyn_hair_fr_l_02 [{shapeType=Capsule, length=5, radius=1.5, shapeName=dyn_hair_fr_l_02_capsule, center=<0.75 0 0>, rotation=<rot -90 0 0>}]
 dyn_hair_fr_l_03 [{shapeType=Capsule, length=6, radius=1.5, shapeName=dyn_hair_fr_l_03_capsule, center=<3.5 0 0>, rotation=<rot -90 0 0>}]
 dyn_hair_fr_r_01 [{shapeType=Capsule, length=3, radius=2.5, shapeName=dyn_hair_fr_r_01_capsule, center=<1.5 0 0.6>, rotation=<rot 0 0 -90>}]
 dyn_hair_fr_r_02 [{shapeType=Capsule, length=3, radius=2, shapeName=dyn_hair_fr_r_02_capsule, center=<-0.5 0 -0.7535>, rotation=<rot -90 0 0>}]
 dyn_hair_fr_r_03 [{shapeType=Capsule, length=6, radius=1.5, shapeName=dyn_hair_fr_r_03_capsule, center=<0 0 -0.3877>, rotation=<rot -90 9.5 0>}]
 dyn_hair_fr_r_04 [{shapeType=Capsule, length=5, radius=1.25, shapeName=dyn_hair_fr_r_04_capsule, center=<2.25 0 -0.5>, rotation=<rot -90 180 -180>}]
 dyn_lock_fr_r_01 [{shapeType=Capsule, length=3, radius=1.5, shapeName=dyn_lock_fr_r_01_capsule, rotation=<rot -90 0 0>}]
 dyn_lock_fr_r_02 [{shapeType=Capsule, length=3, radius=1.5, shapeName=dyn_lock_fr_r_02_capsule, center=<1 0 0>, rotation=<rot -90 0 0>}]
 dyn_lock_fr_r_03 [{shapeType=Capsule, length=4, radius=1.5, shapeName=dyn_lock_fr_r_03_capsule, center=<3 0 0>, rotation=<rot -90 0 0>}]
 head [
  {shapeType=Sphere, radius=7.3541, shapeName=head_sphere, center=<4.6224 0.3144 0>}
  {shapeType=Capsule, length=5.6665, radius=5.881, shapeName=head_capsule, center=<1.2587 1.6335 0>, rotation=<rot 0 -110.4617 90>}
 ]
 neck_01 [{shapeType=Capsule, length=2.6659, radius=3.4708, shapeName=neck_01_capsule, center=<1.8614 1.4972 0>, rotation=<rot 90 -1.7439 0>}]
 neck_02 [{shapeType=Capsule, length=2.6659, radius=3.4708, shapeName=neck_01_capsule, center=<1.8636 1.3451 0>, rotation=<rot 90 0.8385 0>}]
 spine_01 [{shapeType=Capsule, length=13.3473, radius=9.0175, shapeName=spine_01_capsule, center=<-3.6048 0.2887 0>}]
 spine_02 [{shapeType=Capsule, length=16.7037, radius=6.6418, shapeName=spine_02_capsule, center=<0.0877 2.0158 0>}]
 spine_03 [{shapeType=Capsule, length=16.3143, radius=8.4491, shapeName=spine_02_capsule, center=<6.229 4.6277 0>}]
 spine_04 [{shapeType=Capsule, length=19.5, radius=7, shapeName=spine_04_capsule, center=<5.6529 0.7219 0>}]
 spine_05 [{shapeType=Capsule, length=21.0956, radius=5.1878, shapeName=spine_05_capsule, center=<3.6294 0.3262 0>, rotation=<rot 0 -90 0>}]
 upperarm_l [{shapeType=Capsule, length=26.1842, radius=4.0658, shapeName=upperarm_l_capsule, center=<13 -0.0532 -1.1454>, rotation=<rot 90 -3.8402 0>}]
 upperarm_r [{shapeType=Capsule, length=27.1842, radius=4.0658, shapeName=upperarm_l_capsule, center=<-13 0.0532 1.1454>, rotation=<rot 90 -3.8402 0>}]
}
constraints [
 {bone1Name=dyn_hair_ba_l_03, bone2Name=base_hair_ba_l_03, swing1LimitDegrees=15, swing2LimitDegrees=45, twistLimitDegrees=0}
 {bone1Name=dyn_hair_ba_l_04, bone2Name=dyn_hair_ba_l_03, swing1LimitDegrees=15, swing2LimitDegrees=45, twistLimitDegrees=0}
 {bone1Name=dyn_hair_ba_l_05, bone2Name=dyn_hair_ba_l_04, swing1LimitDegrees=8, swing2LimitDegrees=45, twistLimitDegrees=0}
 {bone1Name=dyn_hair_ba_l_06, bone2Name=dyn_hair_ba_l_05, swing1LimitDegrees=15, swing2LimitDegrees=45, twistLimitDegrees=0}
 {bone1Name=dyn_hair_ba_r_03, bone2Name=base_hair_ba_r_03, swing1LimitDegrees=15, swing2LimitDegrees=45, twistLimitDegrees=0}
 {bone1Name=dyn_hair_ba_r_04, bone2Name=dyn_hair_ba_r_03, swing1LimitDegrees=15, swing2LimitDegrees=45, twistLimitDegrees=0}
 {bone1Name=dyn_hair_ba_r_05, bone2Name=dyn_hair_ba_r_04, swing1LimitDegrees=15, swing2LimitDegrees=45, twistLimitDegrees=0}
 {bone1Name=dyn_hair_ba_r_06, bone2Name=dyn_hair_ba_r_05, swing1LimitDegrees=15, swing2LimitDegrees=45, twistLimitDegrees=0}
 {bone1Name=dyn_hair_ba_l_07, bone2Name=dyn_hair_ba_l_06, swing1LimitDegrees=15, swing2LimitDegrees=45, twistLimitDegrees=0}
 {bone1Name=dyn_hair_ba_r_07, bone2Name=dyn_hair_ba_r_06, swing1LimitDegrees=15, swing2LimitDegrees=45, twistLimitDegrees=0}
 {bone1Name=dyn_hair_ct_03, bone2Name=base_hair_ct_03, swing1LimitDegrees=15, swing2LimitDegrees=45, twistLimitDegrees=0}
 {bone1Name=dyn_hair_ct_04, bone2Name=dyn_hair_ct_03, swing1LimitDegrees=15, swing2LimitDegrees=45, twistLimitDegrees=0}
 {bone1Name=dyn_hair_ct_05, bone2Name=dyn_hair_ct_04, swing1LimitDegrees=15, swing2LimitDegrees=45, twistLimitDegrees=0}
 {bone1Name=dyn_hair_ct_06, bone2Name=dyn_hair_ct_05, swing1LimitDegrees=15, swing2LimitDegrees=45, twistLimitDegrees=0}
 {bone1Name=dyn_hair_ct_07, bone2Name=dyn_hair_ct_06, swing1LimitDegrees=15, swing2LimitDegrees=45, twistLimitDegrees=0}
 {
  bone1Name=dyn_hair_ba_l_06
  bone2Name=dyn_hair_ct_06
  swing1LimitDegrees=45
  swing1Motion=Free
  swing2LimitDegrees=45
  swing2Motion=Free
  twistLimitDegrees=45
  twistMotion=Free
 }
 {
  bone1Name=dyn_hair_ba_r_06
  bone2Name=dyn_hair_ct_06
  swing1LimitDegrees=45
  swing1Motion=Free
  swing2LimitDegrees=45
  swing2Motion=Free
  twistLimitDegrees=45
  twistMotion=Free
 }
 {
  bone1Name=base_hair_fr_l
  bone2Name=head
  swing1LimitDegrees=45
  swing1Motion=Locked
  swing2LimitDegrees=45
  swing2Motion=Locked
  twistLimitDegrees=45
  twistMotion=Locked
 }
 {bone1Name=dyn_hair_fr_l_01, bone2Name=base_hair_fr_l, swing1LimitDegrees=15, swing2LimitDegrees=6, twistLimitDegrees=2}
 {bone1Name=dyn_hair_fr_l_02, bone2Name=dyn_hair_fr_l_01, swing1LimitDegrees=15, swing2LimitDegrees=6, twistLimitDegrees=2}
 {bone1Name=dyn_hair_fr_l_03, bone2Name=dyn_hair_fr_l_02, swing1LimitDegrees=15, swing2LimitDegrees=6, twistLimitDegrees=2}
 {bone1Name=dyn_hair_fr_r_01, bone2Name=base_hair_fr_r, swing1LimitDegrees=15, swing2LimitDegrees=6, twistLimitDegrees=2}
 {bone1Name=dyn_hair_fr_r_03, bone2Name=dyn_hair_fr_r_02, swing1LimitDegrees=15, swing2LimitDegrees=6, twistLimitDegrees=2}
 {bone1Name=dyn_hair_fr_r_04, bone2Name=dyn_hair_fr_r_03, swing1LimitDegrees=15, swing2LimitDegrees=6, twistLimitDegrees=2}
 {
  bone1Name=base_lock_fr_r
  bone2Name=head
  swing1LimitDegrees=45
  swing1Motion=Locked
  swing2LimitDegrees=45
  swing2Motion=Locked
  twistLimitDegrees=45
  twistMotion=Locked
 }
 {bone1Name=dyn_lock_fr_r_01, bone2Name=base_lock_fr_r, swing1LimitDegrees=15, swing2LimitDegrees=6, twistLimitDegrees=2}
 {bone1Name=dyn_lock_fr_r_02, bone2Name=dyn_lock_fr_r_01, swing1LimitDegrees=15, swing2LimitDegrees=6, twistLimitDegrees=2}
 {bone1Name=dyn_lock_fr_r_03, bone2Name=dyn_lock_fr_r_02, swing1LimitDegrees=15, swing2LimitDegrees=6, twistLimitDegrees=2}
 {
  bone1Name=base_hair_fr_r
  bone2Name=head
  swing1LimitDegrees=45
  swing1Motion=Locked
  swing2LimitDegrees=45
  swing2Motion=Locked
  twistLimitDegrees=45
  twistMotion=Locked
 }
 {bone1Name=dyn_hair_fr_r_02, bone2Name=dyn_hair_fr_r_01, swing1LimitDegrees=15, swing2LimitDegrees=6, twistLimitDegrees=2}
 {
  bone1Name=dyn_hair_ba_l_05
  bone2Name=dyn_hair_ba_r_07
  swing1LimitDegrees=45
  swing1Motion=Free
  swing2LimitDegrees=45
  swing2Motion=Free
  twistLimitDegrees=45
  twistMotion=Free
 }
 {
  bone1Name=dyn_hair_ba_l_06
  bone2Name=dyn_hair_ba_r_07
  swing1LimitDegrees=45
  swing1Motion=Free
  swing2LimitDegrees=45
  swing2Motion=Free
  twistLimitDegrees=45
  twistMotion=Free
 }
 {
  bone1Name=dyn_hair_ba_l_07
  bone2Name=dyn_hair_ba_r_07
  swing1LimitDegrees=45
  swing1Motion=Free
  swing2LimitDegrees=45
  swing2Motion=Free
  twistLimitDegrees=45
  twistMotion=Free
 }
]
