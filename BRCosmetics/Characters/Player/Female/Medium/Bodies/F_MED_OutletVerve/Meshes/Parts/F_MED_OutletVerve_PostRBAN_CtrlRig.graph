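# Post-RBAN Control Rig for the F_MED_OutletVerve character that orients three hair aim bones (back right, back left, center) toward their dynamic hair bones.
# A loop pairs each dyn_hair "_02" bone with its aim_hair "_01" bone by array index and applies aim math.
# The head bone's rotated -Y axis supplies the secondary direction to stabilize roll.
# Debug axes drawing is left enabled for the aimed transform.
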ControlRigBlueprint /BRCosmetics/Characters/Player/Female/Medium/Bodies/F_MED_OutletVerve/Meshes/Parts/F_MED_OutletVerve_PostRBAN_CtrlRig
graphs [
 RigVMFunctionLibrary {links [], nodes []}
 RigVMModel {nodes [
   RigUnit_BeginExecution = &RigUnit_BeginExecution()
   ItemArray = RigVMDispatchNode(Value=((Type=Bone,Name="dyn_hair_ba_r_02"),(Type=Bone,Name="dyn_hair_ba_l_02"),(Type=Bone,Name="dyn_hair_ct_02"))) {template="DISPATCH_RigVMDispatch_Constant(io Value)"}
   For_Each_1 = RigVMDispatchNode(Array=@ItemArray.Value, ExecuteContext=@RigUnit_BeginExecution.ExecutePin) {template="DISPATCH_RigVMDispatch_ArrayIterator(in Array,out Element,out Index,out Count,out Ratio)"}
   ItemArray_1 = RigVMDispatchNode(Value=((Type=Bone,Name="aim_hair_ba_r_01"),(Type=Bone,Name="aim_hair_ba_l_01"),(Type=Bone,Name="aim_hair_ct_01"))) {template="DISPATCH_RigVMDispatch_Constant(io Value)"}
   n1 = RigVMDispatchNode(Index=@For_Each_1.Index, Array=@ItemArray_1.Value) {template="DISPATCH_RigVMDispatch_ArrayGetAtIndex(in Array,in Index,out Element)"}
   RigUnit_GetTransform = &RigUnit_GetTransform(Item=(Type=Bone,Name="head"), bInitial=False)
   n2 = /Script/ControlRig.RigUnit_AimBoneMath(
    DebugSettings=(bEnabled=False,Scale=10,WorldOffset=(Rotation=(X=0,Y=0,Z=0,W=1),Translation=(X=0,Y=0,Z=0),Scale3D=(X=1,Y=1,Z=1)))
    Primary=(Weight=1,Axis=(X=1,Y=0,Z=0),Target=(X=1,Y=0,Z=0),Kind=Location,Space=(Type=Bone,Name="None"))
    Secondary=(Weight=1,Axis=(X=0,Y=0,Z=1),Target=(X=1,Y=0,Z=0),Kind=Direction,Space=(Type=Bone,Name="None"))
    Weight=1
    InputTransform=&RigUnit_GetTransform(bInitial=False, Item=@n1.Element).Transform
    Primary.Target=&RigUnit_GetTransform(bInitial=False, Item=@For_Each_1.Element).Transform.Translation
    Secondary.Target=/Script/RigVM.RigVMFunction_MathTransformRotateVector(Vector=(X=0,Y=-1,Z=0), Transform=@RigUnit_GetTransform.Transform) {is=struct RigVMUnitNode, template="Rotate Vector::Execute(in Transform,in Vector,out Result)"}.Result
   ) {is=struct RigVMUnitNode}
   > &RigUnit_SetTransform(Weight=1, bInitial=False, Value=@n2.Result, Item=@n1.Element)
   > RigVMFunction_DebugTransformMutableNoSpace = /Script/RigVM.RigVMFunction_DebugTransformMutableNoSpace(
    Color=(R=1,G=0,B=0,A=1)
    DebugDrawSettings=(DepthPriority=SDPG_Foreground,Lifetime=-1)
    Mode=Axes
    Scale=10
    Thickness=0
    WorldOffset=(Rotation=(X=0,Y=0,Z=0,W=1),Translation=(X=0,Y=0,Z=0),Scale3D=(X=1,Y=1,Z=1))
    bEnabled=True
    Transform=@n2.Result
   ) {is=struct RigVMUnitNode}
  ]}
 RigVMModel Rig {nodes [
   n1 = &RigUnit_PrepareForExecution()
   &RigUnit_HierarchyImportFromSkeleton(NameSpace=None, bIncludeCurves=False, bIncludeMeshSockets=False, bIncludeVirtualBones=True, ExecutePin=@n1.ExecutePin)
  ]}
]
